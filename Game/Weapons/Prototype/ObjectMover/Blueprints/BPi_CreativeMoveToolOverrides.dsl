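; BPi_CreativeMoveToolOverrides is a Blueprint interface, in the ObjectMover prototype weapon folder, for overriding values used by the creative move tool.
; It declares getters for a creative actor's bounds, origin and relative transform, plus a grid axis guide attachment point.
; Each function holds only default false, zero-vector or empty returns, with no logic of its own.
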(fn GetCreativeActorBounds
  (return false "0, 0, 0"))

(fn GetCreativeActorOrigin
  (return false "0, 0, 0"))

(fn GetCreativeActorRelativeTransform
  (return false 0))

(fn GetGridAxisGuideAttachmentPoint
  (return false false "0, 0, 0"))
