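; Blueprint interface for interaction with the prototype ObjectMover weapon tool.
; It declares one function, SetPostProcess, which takes a Boolean On flag.
; The name suggests implementers switch a post-process effect on or off, but the file holds only the signature, with no implementation.
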(fn SetPostProcess
  (params (On :type Boolean)))
