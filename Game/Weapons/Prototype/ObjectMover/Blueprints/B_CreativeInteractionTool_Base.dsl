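; Base Blueprint for the Creative mode move tool, a handheld phone that selects, moves and deletes objects, providing its visuals and audio.
; The visual state sets the phone's dynamic material color and icon, transition speed and beam colors.
; Tick updates pooled lightbeam meshes, which rotate and scale toward target bounds.
; Equip and visibility events toggle highlight post-process, outlines and weapon state, while soft-referenced sounds alternate between two audio components with fade-out.
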(event Custom|UpdateBuildingPieceLocationAndRadius
  (params (Vec4 :type Vector%204%20Structure))
  ;[1 1 1] Server calls to replicate necessary variables for phone visuals to work across all clients (to be able to see other players' interact states)
  (Variables|Default|SetBuildingPieceCenterAndRadius Vec4))

(event Custom|UpdateBuildingPreviewTransform
  (params (Transform :type Transform))
  ;[1 1 1] Server calls to replicate necessary variables for phone visuals to work across all clients (to be able to see other players' interact states)
  (Variables|Default|SetBlueprintWorldTransform Transform))

(event EventPlayScreenFlashForObjectDuplication
  ;[1 1 1] Flash Screen On Duplicate
  (Variables|Default|SetWasTransitionTimlinePlaying (Components|Timeline|IsPlaying (Variables|B_CreativeInteractionTool_Base|GetTransitionTL)))
  (Components|Timeline|Stop (Variables|B_CreativeInteractionTool_Base|GetTransitionTL))
  (for _array_element (Variables|FortCreativeMoveTool|GetActivePreviewActors)
    (Utilities|IsValid _array_element
      (:Is%20Valid
        (Class|CreativeMoveToolBlueprintPreview|PrepareDuplicateMaterialTimeline _array_element))
      (:Is%20Not%20Valid
        (bind (_direction _lerp) (Timeline|ScreenFlashTL PlayFromStart)
          (:Update
            (Rendering|Material|SetVectorParameterValue (Variables|PhoneVisuals|GetPhoneMID) "ScreenColor" (Math|Color|Lerp(LinearColor) (Variables|PhoneVisuals|GetDesiredColor) (Math|Color|LinearColor*Float (Variables|PhoneVisuals|GetDesiredColor) 1e+01) _lerp) :!class /Script/Engine.MaterialInstanceDynamic))
          (:Finished
            (if (Variables|Default|GetWasTransitionTimlinePlaying)
              (Components|Timeline|Play (Variables|B_CreativeInteractionTool_Base|GetTransitionTL)))))))))

(event Custom|PlayMoveToolSoundBySoftObjectReference
  (params (Sound%20To%20Play :type Sound%20Base%20Soft%20Object%20Reference))
  (bind _move_tool_sound_a (Variables|Audio|GetMove%20Tool%20Sound%20A))
  (bind _move_tool_sound_a_2 (Variables|Audio|GetMove%20Tool%20Sound%20A))
  (bind _currentmusic (Variables|Audio|GetCurrentMusic))
  (bind _move_tool_sound_b (Variables|Default|GetMove%20Tool%20Sound%20B))
  (bind _move_tool_sound_b_2 (Variables|Default|GetMove%20Tool%20Sound%20B))
  ;[1 1 1] Loads a sound if it isn't loaded, and then plays one
  (bind _assound_base (Utilities|SoundObjectfromReference Sound%20To%20Play)
    (:Completed
      (Variables|Audio|SetCurrentMusic _assound_base)
      (Utilities|FlowControl|FlipFlop
        (:A
          (Audio|Components|Audio|SetSound _move_tool_sound_a_2 _currentmusic)
          (Audio|Components|Audio|Play _move_tool_sound_a_2 0.0)
          (Utilities|IsValid _move_tool_sound_b_2
            (:Is%20Valid
              (if (Audio|Components|Audio|IsPlaying _move_tool_sound_b_2)
                (Audio|Components|Audio|FadeOut _move_tool_sound_b_2 0.5 0.0 "Linear" :!class /Script/Engine.AudioComponent)))))
        (:B
          (Audio|Components|Audio|SetSound _move_tool_sound_b _currentmusic)
          (Audio|Components|Audio|Play _move_tool_sound_b 0.0)
          (Utilities|IsValid _move_tool_sound_a
            (:Is%20Valid
              (if (Audio|Components|Audio|IsPlaying _move_tool_sound_a)
                (Audio|Components|Audio|FadeOut _move_tool_sound_a 0.5 0.0 "Linear" :!class /Script/Engine.AudioComponent)))))))))

(event EventTick
  (params (DeltaSeconds :type Float%20%28single-precision%29))
  ;[1 1 1] Update Our Previews to the corresponding actors in the selected actors set
  (|CallParentFunction DeltaSeconds)
  (Networking|CanExecuteCosmeticEvents
    (:True
      (Utilities|UpdateLightbeam))))

(event FortWeapon|EventOnEquippedAndAttached
  ;[1 1 1] Enable post process on equip
  ;;[1 1 1] On eqiup
  (|CallParentFunction)
  (Networking|CanExecuteCosmeticEvents
    (:True
      (if (CreativeMoveTool|IsRunningonOwningClient)
        (CallFunction|ServerSetPhoneMaterialState "Deselected")
        (else
          _))
      (if (CallFunction|IsLocallyControlled)
        (bind _assound_base (Utilities|SoundObjectfromReference "/Game/Sounds/Creative/Gadgets/MoveTool/MoveTool_Raise_01_Cue.MoveTool_Raise_01_Cue")
          (:Synchronous
            (Variables|Default|SetCanPostProcessBeEnabled true)
            (CallFunction|SetEnableTODMPostProcess false))
          (:Completed
            (Audio|PlaySoundatLocation _assound_base (Transformation|GetActorLocation))))))))

(event EventOnWeaponVisibilityChanged
  (params (bVisible :type Boolean) (bSetForLocalControllerOnly :type Boolean))
  (bind _returnvalue (not bSetForLocalControllerOnly))
  ;[1 1 1] Handle weapon leaving the player's hand
  (|CallParentFunction bVisible bSetForLocalControllerOnly)
  (Networking|CanExecuteCosmeticEvents
    (:True
      (CallFunction|ServeronVisibilityChanged bVisible _returnvalue)
      (CallFunction|VisibilityChangedEvent bVisible _returnvalue))) ; In some instances, we only update visibility on the client side.\nIn such cases, it's important to also handle the corresponding server call.
)

(event Custom|ServerOnVisibilityChanged
  (params (Visible :type Boolean) (LocalControllerOnly :type Boolean))
  ;[1 1 1] Handle weapon leaving the player's hand
  (CallFunction|VisibilityChangedEvent Visible LocalControllerOnly))

(event Custom|VisibilityChangedEvent
  (params (Visible :type Boolean) (LocalControllerOnly :type Boolean))
  (bind _result (CallFunction|ShouldRespondtoVisilityChange))
  (bind _returnvalue (and LocalControllerOnly _result))
  ;[1 1 1] Handle weapon leaving the player's hand
  (if Visible
    (if _returnvalue
      (CallFunction|UnholsterWeapon))
    (elif _returnvalue
      (CallFunction|DeactiveWeaponandState)
      (Networking|CanExecuteCosmeticEvents
        (:False
          (CreativeMoveTool|CleanupPreviewActors)))
      (else
        _))))

(event FortWeapon|EventOnLocalUnEquip
  ;[1 1 1] Enable post process on equip
  ;;[1 1 1] On Unequip
  (|CallParentFunction)
  (Networking|CanExecuteCosmeticEvents
    (:True
      (if (CallFunction|IsLocallyControlled)
        (bind _assound_base (Utilities|SoundObjectfromReference "/Game/Sounds/Creative/Gadgets/MoveTool/MoveTool_Lower_01_Cue.MoveTool_Lower_01_Cue")
          (:Synchronous
            (Variables|Default|SetCanPostProcessBeEnabled false)
            (CallFunction|SetEnableTODMPostProcess true)
            (Utilities|Array|Clear (Variables|FortCreativeMoveTool|GetActivePreviewActors)))
          (:Completed
            (Audio|PlaySoundatLocation _assound_base (Transformation|GetActorLocation))))))))

(event CreativeMoveTool|EventPlayHoverSound
  (bind _move_tool_highlight_sound (Variables|Default|GetMove%20Tool%20Highlight%20Sound))
  ;[1 1 1] Audio
  (bind _sound (CallFunction|GetHighlightSound))
  (bind _assound_base (Utilities|SoundObjectfromReference _sound)
    (:Completed
      (Audio|Components|Audio|SetSound _move_tool_highlight_sound _assound_base)
      (Audio|Components|Audio|Play _move_tool_highlight_sound 0.0))))

(event MoveToolFunctions|EventSetPostProcess
  (params (On :type Boolean))
  ;[0 0.1364 1 0.379] Interface call for updating post process state. Used for devices that want to update post process without loading the tool (if it's not loaded already. e.g. play mode)
  (CreativeMoveTool|SetHoverPostProcessEnabled On))

(event CreativeMoveTool|EventPlayMultiSelectAddedSound
  ;[1 1 1] Audio
  (CallFunction|PlayMoveToolSoundbySoftObjectReference "/Game/Sounds/Creative/Gadgets/MoveTool/MoveTool_MultiSelect_Select_Cue.MoveTool_MultiSelect_Select_Cue"))

(event CreativeMoveTool|EventPlayMultiSelectRemovedSound
  ;[1 1 1] Audio
  (CallFunction|PlayMoveToolSoundbySoftObjectReference "/Game/Sounds/Creative/Gadgets/MoveTool/MoveTool_MultiSelect_Cancel_Cue.MoveTool_MultiSelect_Cancel_Cue"))

(event EventHideTractorBeams
  (bind _lightbeammeshpool (Variables|Default|GetLightbeamMeshPool))
  ;[1 1 1] Hide Lightbeams
  (for _array_element _lightbeammeshpool
    (Utilities|IsValid _array_element
      (:Is%20Valid
        (Components|DestroyComponent _array_element))))
  (Utilities|Array|Clear _lightbeammeshpool))

(event CreativeMoveTool|EventPlayObjectPlacedSound
  ;[1 1 1] Audio
  (CallFunction|PlayMoveToolSoundbySoftObjectReference "/Game/Sounds/Creative/Gadgets/MoveTool/MoveTool_Place_01_Cue.MoveTool_Place_01_Cue"))

(event CreativeMoveTool|EventStartInteractionAudio
  (bind _movetoolloopingsoundcomponent (Variables|Audio|GetMoveToolLoopingSoundComponent))
  ;[1 1 1] Audio
  (bind _assound_base (Utilities|SoundObjectfromReference "/Game/Sounds/Creative/Gadgets/MoveTool/MoveTool_SelectedLoop_01_Cue.MoveTool_SelectedLoop_01_Cue")
    (:Completed
      (Audio|Components|Audio|SetSound _movetoolloopingsoundcomponent _assound_base)
      (Audio|Components|Audio|Play _movetoolloopingsoundcomponent 0.0))))

(event EventPlayTransitionTimeline
  ;[1 1 1] Play Transition timeline
  (Components|Timeline|Play (Variables|B_CreativeInteractionTool_Base|GetTransitionTL)))

(event EventStopScaleAudio
  ;[1 1 1] Audio
  (Audio|Components|Audio|Stop (Variables|Audio|GetMove%20Tool%20Scale%20Up)))

(event EventHandlePhoneMaterialStateUpdate
  (bind _transitiontl (Variables|B_CreativeInteractionTool_Base|GetTransitionTL))
  (bind _transitiontl_2 (Variables|B_CreativeInteractionTool_Base|GetTransitionTL))
  ;[1 1 1] MaterialState Update
  (Variables|PhoneVisuals|SetDesiredColor (CallFunction|GetPhoneScreenBGColor))
  (Variables|Default|SetPreviousColor (Variables|PhoneVisuals|GetCurrentColor))
  (Variables|Default|SetPreviousTexture (Variables|Default|GetDesiredTexture))
  (Variables|Default|SetDesiredTexture (CallFunction|GetPhoneScreenIcon))
  (Rendering|Material|SetTextureParameterValue (Variables|PhoneVisuals|GetPhoneMID) "IconB" (Variables|Default|GetPreviousTexture))
  (Rendering|Material|SetTextureParameterValue (Variables|PhoneVisuals|GetPhoneMID) "IconB" (Variables|Default|GetDesiredTexture))
  (Components|Timeline|SetPlayRate _transitiontl (CallFunction|GetTransitionPlayRate))
  (if (Components|Timeline|IsPlaying _transitiontl)
    (Components|Timeline|Stop _transitiontl_2)
    (Components|Timeline|SetPlaybackPosition _transitiontl_2 0.0 false true)
    (else
      _))
  (CallFunction|UpdateMaterialValiditiesforPreviews)
  (for (_array_element _array_index) (Variables|Default|GetLightbeamMeshPool)
    (Utilities|IsValid _array_element
      (:Is%20Valid
        (CallFunction|SetRayMIDParams (Utilities|Array|Get(acopy) (Variables|Default|GetLightbeamMIDPool) _array_index)))))
  (if (<= (Utilities|Array|Length (Variables|FortCreativeMoveTool|GetActivePreviewActors)) 0)
    (CallFunction|HideTractorBeams)))

(fn CanUtilizeMoveToolSpawningFunctionality
  (bind _returnvalue (|CallParentFunction))
  (return (and _returnvalue (FortWeapon|IsEquipped))))

(fn CheckForMultiSelection
  (return (and (> 0 1) false)))

(fn CleanupErrantRays
  (bind _possessedpreviewactors (Variables|FortCreativeMoveTool|GetPossessedPreviewActors))
  (bind _returnvalue (Utilities|Array|Length _possessedpreviewactors))
  (bind _lightbeammeshpool (Variables|Default|GetLightbeamMeshPool))
  (bind _lightbeammeshpool_2 (Variables|Default|GetLightbeamMeshPool))
  ;[1 1 1] Sometimes replication latency causes there to be mismatches in the amount of rays, this does a simple check to see if the arrays are of different lengths and destroys meshes if needed
  (if (> (Utilities|Array|Length _lightbeammeshpool_2) _returnvalue)
    (for _index (range _returnvalue (+ (Utilities|Array|LastIndex _lightbeammeshpool_2) 1))
      (Utilities|IsValid (Utilities|Array|Get(acopy) _lightbeammeshpool _index)
        (:Is%20Valid
          (Components|DestroyComponent (Utilities|Array|Get(acopy) _lightbeammeshpool _index)))
        (:Is%20Not%20Valid
          _))
      (Utilities|Array|RemoveIndex (Variables|Default|GetLightbeamMeshPool) _index))))

(fn FindTransitionPlayPosition
  (params (Duration :type Float%20%28double-precision%29))
  (return (Utilities|Reciprocal (- Duration (Math|Float|Clamp(Float) (- 1.0 (- (Components|Timeline|GetTimelineLength (Variables|B_CreativeInteractionTool_Base|GetTransitionTL)) (Components|Timeline|GetPlaybackPosition (Variables|B_CreativeInteractionTool_Base|GetTransitionTL)))))))))

(fn GetFirstActiveBlueprintPreviewersFirstMeshComponent
  (bind _activepreviewactors (Variables|FortCreativeMoveTool|GetActivePreviewActors))
  (bind _output (Utilities|Array|Get(acopy) _activepreviewactors))
  (if (Utilities|Array|IsValidIndex _activepreviewactors)
    (Utilities|IsValid _output
      (:Is%20Valid
        (bind _ascreative_move_tool_preview_actor (Utilities|Casting|CastToCreativeMoveToolPreviewActor _output)
          (:then
            (if (Utilities|Array|IsValidIndex (Class|CreativeMoveToolPreviewActor|GetPreviewMeshComponents _ascreative_move_tool_preview_actor))
              (bind _output_get (LocalVariables|SetFirstBlueprintMeshComponent (Utilities|Array|Get(acopy) (Class|CreativeMoveToolPreviewActor|GetPreviewMeshComponents _ascreative_move_tool_preview_actor))))
              (Utilities|IsValid _output_get
                (:Is%20Valid
                  (return (LocalVariables|GetFirstBlueprintMeshComponent)))
                (:Is%20Not%20Valid
                  _))
              (else
                _))
            (return 0))
          (:CastFailed
            _)))
      (:Is%20Not%20Valid
        _))
    (else
      _)))

(fn GetHighlightSound
  (return "/Game/Sounds/Creative/Gadgets/MoveTool/MoveTool_Highlight_01_Cue.MoveTool_Highlight_01_Cue"))

(fn GetInteractionActive
  (params (Index :type ECreativeMoveToolVisualState%20Enum :default "ValidPlacement"))
  (return (Utilities|Select :Index Index :ValidPlacement true :InvalidPlacement true :Delete true :Deselected false :Multiselect false :AddedToQuickBar false)))

(fn GetLightbeamMesh
  (params (ActorIndex :type Integer))
  (bind _returnvalue (CreativeMoveTool|GetMultiselectLimit))
  (bind _actorindex (Variables|GetActorIndex))
  (bind _returnvalue_2 (Math|Integer|%25%28Integer%29 _actorindex _returnvalue))
  (Utilities|FlowControl|Sequence
    (:then_0
      (if (Utilities|Array|IsValidIndex (Variables|Default|GetLightbeamMeshPool) _returnvalue_2)
        (Utilities|IsValid (Utilities|Array|Get(acopy) (Variables|Default|GetLightbeamMeshPool) _returnvalue_2)
          (:Is%20Not%20Valid
            _))
        (else
          _))
      (bind _returnvalue_3 (AddComponent|NODE_AddStaticMeshComponent-2))
      (Utilities|Array|SetArrayElem (Variables|Default|GetLightbeamMeshPool) _returnvalue_2 _returnvalue_3 true)
      (bind _returnvalue_4 (Rendering|Material|CreateDynamicMaterialInstance _returnvalue_3 0 0 "None" :!class /Script/Engine.PrimitiveComponent))
      (Utilities|Array|SetArrayElem (Variables|Default|GetLightbeamMIDPool) _returnvalue_2 _returnvalue_4 true)
      (CallFunction|SetRayMIDParams _returnvalue_4))
    (:then_1
      (return (Utilities|Array|Get(acopy) (Variables|Default|GetLightbeamMeshPool) _returnvalue_2) (Utilities|Array|Get(acopy) (Variables|Default|GetLightbeamMIDPool) _returnvalue_2)))))

(fn GetMeshComponentsFromActor
  (params (Actor :type Actor%20Object%20Reference))
  (bind _ascreative_move_tool_overrides (Utilities|Casting|CastToCreativeMoveToolOverrides Actor)
    (:then
      (if (> (Utilities|Array|Length (Class|CreativeMoveToolOverrides|GetMeshComponents _ascreative_move_tool_overrides)) 0)
        (return (Class|CreativeMoveToolOverrides|GetMeshComponents _ascreative_move_tool_overrides))
        (else
          _)))
    (:CastFailed
      _))
  (return (Actor|GetComponentsByClass Actor "/Script/Engine.MeshComponent")))

(fn GetPhoneScreenBGColor
  (return (Utilities|Select :Index (Variables|FortCreativeMoveTool|GetActivePhoneVisualState) :ValidPlacement (Variables|PhoneVisuals|GetScreenSelectedColor) :InvalidPlacement (Variables|Default|GetScreenInvalidColor) :Delete (Variables|Default|GetScreenDeleteColor) :Deselected (Variables|PhoneVisuals|GetScreenDeselectedColor) :Multiselect (Variables|PhoneVisuals|GetScreenMultiSelectColor) :AddedToQuickBar (Variables|PhoneVisuals|GetScreenAddedToQuickbarColor))))

(fn GetPhoneScreenIcon
  (return (Utilities|Select :Index (Variables|FortCreativeMoveTool|GetActivePhoneVisualState) :ValidPlacement (Variables|PhoneVisuals|GetScreenSelectedIcon) :InvalidPlacement (Variables|PhoneVisuals|GetScreenInvalidIcon) :Delete (Variables|PhoneVisuals|GetScreenDeleteIcon) :Deselected (Variables|PhoneVisuals|GetScreenDeselectedIcon) :Multiselect (Variables|PhoneVisuals|GetScreenMultiSelectIcon) :AddedToQuickBar (Variables|PhoneVisuals|GetScreenAddedToQuickbarIcon))))

(fn GetPlayerRotationOffset
  (bind _returnvalue (Game|GetInstigatorController))
  (bind _returnvalue_2 (Pawn|GetControlledPawn _returnvalue))
  (bind _returnvalue_3 (Transformation|GetActorForwardVector _returnvalue_2))
  (bind _x (.x _returnvalue_3))
  (bind _y (.y _returnvalue_3))
  (if (>= (Math|Float|Absolute(Float) _x) (Math|Float|Absolute(Float) _y))
    (if (>= _x 0.0)
      (return "0, 0, 0")
      (else
        (return "0,180.000000, 0")))
    (elif (>= _y 0.0)
      (return "0,90.000000, 0")
      (else
        (return "0,270.000000, 0")))))

(fn GetStaticMeshCenterAndRadius
  (params (StaticMesh :type Static%20Mesh%20Component%20Object%20Reference))
  (bind (_min _max) (Components|StaticMesh|GetLocalBounds StaticMesh))
  (bind (_origin _boxextent _sphereradius) (Collision|GetComponentBounds StaticMesh))
  (return (Math|Transform|TransformLocation (Transformation|GetWorldTransform StaticMesh) (/ (+ _min _max) 2.0)) _sphereradius))

(fn GetTransitionPlayRate
  (return (CallFunction|FindTransitionPlayPosition (Utilities|Select :Index (Variables|FortCreativeMoveTool|GetActivePhoneVisualState) :ValidPlacement 0.2 :InvalidPlacement 0.15 :Delete 0.15 :Deselected 0.35 :Multiselect 0.0 :AddedToQuickBar 0.2))))

(fn GetXYBoundsExtent
  (params (Component :type Scene%20Component%20Object%20Reference))
  (bind (_origin _boxextent) (Collision|GetComponentBounds Component))
  (bind (_returnvalue_x _returnvalue_y) (Math|Vector|BreakVector (* _boxextent (/ "1.000000,1.000000,1.000000" (Variables|/Script/Engine.SceneComponent|GetRelativeScale3D Component)))))
  (return (Math|Float|Max(Float) _returnvalue_x _returnvalue_y)))

(fn IsBuildingPiece
  (bind _returnvalue (CallFunction|GetBoundManipulateInteractBehavior))
  (bind (_asmove_buildings_on_grid _bsuccess) (Utilities|Casting|CastToMoveBuildingsOnGrid _returnvalue))
  (Utilities|FlowControl|Sequence
    (:then_0
      (bind _asfort_player_controller (Utilities|Casting|CastToFortPlayerController (Game|GetInstigatorController (Actor|GetOwner)))
        (:then
          (Utilities|IsValid (Building|GetBuildPreviewMarker _asfort_player_controller)
            (:Is%20Valid
              (return (Building|GetBuildingMeshComponent (Building|GetBuildPreviewMarker _asfort_player_controller)) (or (Creative|IsBuildToolOverridden _asfort_player_controller) _bsuccess)))))
        (:CastFailed
          (return 0 _bsuccess))))
    (:then_1
      (return 0 false))))

(fn IsLocallyControlled
  (Utilities|FlowControl|Sequence
    (:then_0
      (bind _weaponmesh (Variables|Visual|GetWeaponMesh)
        (:then
          (Utilities|IsValid (Components|GetOwner _weaponmesh)
            (:Is%20Valid
              (Utilities|IsValid (Game|GetInstigator (Components|GetOwner _weaponmesh))
                (:Is%20Valid
                  (return (Pawn|IsLocallyControlled (Game|GetInstigator (Components|GetOwner _weaponmesh)))))))))
        (:else)))
    (:then_1
      (return false))))

(fn OrientScaleLightBeamByMesh
  (params (StaticMesh :type Static%20Mesh%20Component%20Object%20Reference) (ActorIndex :type Integer))
  (bind (_centerws _sphericalradius) (CallFunction|GetStaticMeshCenterandRadius StaticMesh))
  (Utilities|IsValid StaticMesh
    (:Is%20Valid
      (if (not (or (Variables|/Script/Engine.ActorComponent|GetbIsEditorOnly StaticMesh) (Variables|/Script/Engine.SceneComponent|GetbHiddenInGame StaticMesh)))
        (bind _distance (Class|MissionBlueprintFunctionLibrary|DistanceBetweenTwoVectors (Variables|Default|GetPreviousLocation) _centerws))
        (if (or (>= _distance 2048.0) (Math|Vector|Equal(Vector) _centerws "0, 0, 0" 0.001))
          (Transformation|SetRelativeScale3D (CallFunction|GetLightbeamMesh (Variables|GetActorIndex)) "0, 0, 0")
          (else
            (CallFunction|ScaleMeshLightRays _centerws _sphericalradius (Variables|GetActorIndex))))
        (Variables|Default|SetPreviousLocation _centerws)))))

(fn OrientScaleLightBeamMeshByInfo
  (params (Location :type Vector) (Radius :type Float%20%28double-precision%29) (ActorIndex :type Integer))
  (bind _distance (Class|MissionBlueprintFunctionLibrary|DistanceBetweenTwoVectors (Variables|Default|GetPreviousLocation) Location))
  (if (or (>= _distance 2048.0) (Math|Vector|Equal(Vector) Location "0, 0, 0" 0.001))
    (Transformation|SetRelativeScale3D (CallFunction|GetLightbeamMesh (Variables|GetActorIndex)) "0, 0, 0")
    (else
      (CallFunction|ScaleMeshLightRays Location Radius (Variables|GetActorIndex))))
  (Variables|Default|SetPreviousLocation Location))

(fn RemoveEditorOnlyItems
  (params (ArrayToClean :type Array%20of%20Mesh%20Component%20Object%20References))
  (bind (_arrayindex _arrayelement) (Utilities|Array|ReverseforEachLoop ArrayToClean)
    (:LoopBody
      (Utilities|IsValid _arrayelement
        (:Is%20Valid
          (if (Variables|/Script/Engine.ActorComponent|GetbIsEditorOnly _arrayelement)
            (Utilities|Array|RemoveItem (Variables|GetArrayToClean) _arrayelement)))))))

(fn ScaleMeshLightRays
  (params (WS%20Center%20Location :type Vector) (Object%20Spherical%20Radius :type Float%20%28double-precision%29) (ActorIndex :type Integer))
  (bind _actorindex (Variables|GetActorIndex))
  (bind _lightbeammesh (CallFunction|GetLightbeamMesh _actorindex))
  (bind _returnvalue (Transformation|GetWorldLocation _lightbeammesh))
  (bind _returnvalue_2 (- WS%20Center%20Location _returnvalue))
  (bind (_lightbeammesh_2 _lightbeammid) (CallFunction|GetLightbeamMesh (Variables|GetActorIndex)))
  (bind _returnvalue_3 (Math|Vector|VectorLength _returnvalue_2))
  (bind _actorindex_2 (Variables|GetActorIndex))
  (bind _lightbeammesh_3 (CallFunction|GetLightbeamMesh _actorindex_2))
  (bind (_lightbeamlargestcomponentscale_x _lightbeamlargestcomponentscale_y _lightbeamlargestcomponentscale_z) (Math|Vector|BreakVector (Variables|GetLightBeamLargestComponentScale_X)))
  (Transformation|SetWorldRotation _lightbeammesh_3 (Math|Rotator|MakeRotfromX _returnvalue_2) false false)
  (bind _output_get (Variables|SetLightBeamLargestComponentScale_X (Math|Vector|MakeVector (Math|Float|Max(Float) _lightbeamlargestcomponentscale_x (/ _returnvalue_3 1e+02)) (Math|Float|Max(Float) (Utilities|ClampMaxFloat (* (/ Object%20Spherical%20Radius 1e+02) (Variables|Default|GetLightBeamWidthMultiplier)) (Variables|Default|GetLightBeamMaxWidth)) _lightbeamlargestcomponentscale_y) (Math|Float|Max(Float) (Utilities|ClampMaxFloat (* (/ Object%20Spherical%20Radius 1e+02) (Variables|Default|GetLightBeamWidthMultiplier)) (Variables|Default|GetLightBeamMaxWidth)) _lightbeamlargestcomponentscale_z))))
  (Transformation|SetRelativeScale3D _lightbeammesh_3 _output_get)
  (Rendering|Material|SetScalarParameterValue _lightbeammid "Length" (Utilities|ClampMaxFloat _returnvalue_3 (Variables|Default|GetLightBeamMaxLength)) :!class /Script/Engine.MaterialInstanceDynamic))

(fn SetEnableTODM_PostProcess
  (params (TODMPPEnabled :type Boolean))
  ;[1 1 1] Temp way of doing this - set basepostprocessmaterial does nothing
  (Rendering|Material|SetScalarParameterValue "/Game/Packages/Fortress_SharedMaterials/GlobalMaterialParameters/FortniteMaterialParameters.FortniteMaterialParameters" "CreativeModeDisablePBWOutlines" (Math|Conversions|ToFloat(Boolean) (not TODMPPEnabled)) :!class /Script/Engine.KismetMaterialLibrary))

(fn SetHoverPostProcessEnabled
  (params (bEnabled :type Boolean))
  (bind _highlightpostprocess (Variables|Default|GetHighlightPostProcess)
    (:then
      (Variables|/Script/Engine.PostProcessComponent|SetbEnabled (and bEnabled (Variables|Default|GetCanPostProcessBeEnabled)) _highlightpostprocess)
      (return))
    (:else)))

(fn SetPostProcess
  (params (ForceEnabled :type Boolean))
  (CreativeMoveTool|SetHoverPostProcessEnabled ForceEnabled))

(fn SetRayMIDParams
  (params (RayMID :type Material%20Instance%20Dynamic%20Object%20Reference))
  (Rendering|Material|SetVectorParameterValue (Variables|GetRayMID) "ColorLow" (Utilities|Select :Index (Variables|FortCreativeMoveTool|GetActivePhoneVisualState) :ValidPlacement (Variables|PhoneVisuals|GetBeamColorLow_Valid) :InvalidPlacement (Variables|PhoneVisuals|GetBeamColorLow_Invalid) :Delete (Variables|PhoneVisuals|GetBeamColorLow_Delete) :Deselected (Variables|PhoneVisuals|GetBeamColorLow_Valid) :Multiselect (Variables|PhoneVisuals|GetBeamColorLow_Valid)) :!class /Script/Engine.MaterialInstanceDynamic)
  (Rendering|Material|SetVectorParameterValue (Variables|GetRayMID) "ColorHigh" (Utilities|Select :Index (Variables|FortCreativeMoveTool|GetActivePhoneVisualState) :ValidPlacement (Variables|PhoneVisuals|GetBeamColorHigh_Valid) :InvalidPlacement (Variables|PhoneVisuals|GetBeamColorHigh_Invalid) :Delete (Variables|PhoneVisuals|GetBeamColorHigh_Delete) :Deselected (Variables|PhoneVisuals|GetBeamColorHigh_Valid) :Multiselect (Variables|PhoneVisuals|GetBeamColorHigh_Valid)) :!class /Script/Engine.MaterialInstanceDynamic))

(fn ShouldRespondToVisilityChange
  (return (Variables|Default|GetbShouldRespondToWeaponVisibilityChanged)))

(fn ConstructionScript
  (Networking|CanExecuteCosmeticEvents
    (:True
      (bind _returnvalue (Rendering|Material|CreateDynamicMaterialInstance (Variables|Visual|GetWeaponMesh) 0 (Variables|Default|GetWeaponMeshMIDSourceMaterial) "" :!class /Script/Engine.PrimitiveComponent))
      (Variables|PhoneVisuals|SetPhoneMID _returnvalue))))
