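; Preview actor for the Creative move tool, animating visual feedback such as delete, duplicate, drop, grid, guide and transition effects on the manipulated object.
; Play events run Blueprint timelines, several with selectable play or reverse, forwarding updates and completion to handler functions.
; The construction script caches those timelines in CreativeMoveToolBlueprintPreview variables.
; Query functions accept interface overrides from the base actor for center, bounds and guide attachment, centering walls on local mesh bounds.
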(event EventPlayDeleteTL
  ;[1 1 1 0.5] Delete Timeline
  (bind (_direction _wpotransition _blueprinttransition _desaturation _relativelocation _lightcolor) (Timeline|DeleteTL PlayFromStart)
    (:Update
      (CallFunction|HandleDeleteTLTick _wpotransition _blueprinttransition _desaturation _relativelocation _lightcolor))
    (:Finished
      (CallFunction|OnDeleteTLComplete))
    (:DisableShadows
      (CallFunction|OnDeleteTLDisableShadows))))

(event EventPlayDropPropTL
  ;[1 1 1 0.5] Drop Prop Timeline
  (bind (_direction _lerp) (Timeline|DropPropTL PlayFromStart)
    (:Update
      (CallFunction|HandleDropPropTLTick _lerp))
    (:Finished
      (CallFunction|OnDropPropTLComplete))))

(event EventPlayFadeRotationGuideTL
  (params (PlaybackType :type EPreviewTimelinePlaybackType%20Enum :default "Play"))
  ;[1 1 1 0.5] Fade Rotation Guide Timeline
  (switch Utilities|FlowControl|Switch|SwitchonEPreviewTimelinePlaybackType PlaybackType
    (:Play
      (bind (_direction _opacity) (Timeline|FadeRotationGuideTL Play)
        (:Update
          (CallFunction|HandleFadeRotationGuideTLTick _direction _opacity))
        (:Finished
          (CallFunction|HandleFadeRotationGuideTLComplete _direction))))
    (:PlayFromStart
      (Timeline|FadeRotationGuideTL PlayFromStart))
    (:Reverse
      (Timeline|FadeRotationGuideTL Reverse))
    (:ReverseFromEnd
      (Timeline|FadeRotationGuideTL ReverseFromEnd))))

(event EventPlayLerpGridTL
  ;[1 1 1 0.5] Lerp Grid Timeline
  (bind (_direction _lerp) (Timeline|LerpGridTL PlayFromStart)
    (:Update
      (CallFunction|HandleLerpGridTick _lerp))
    (:Finished
      (CallFunction|OnLerpGridTLComplete))))

(event EventPlayScaleGridPreviewTL
  (params (PlaybackType :type EPreviewTimelinePlaybackType%20Enum :default "Play"))
  ;[1 1 1 0.5] Scale Grid preview Timeline
  (switch Utilities|FlowControl|Switch|SwitchonEPreviewTimelinePlaybackType PlaybackType
    (:Play
      (bind (_direction _scale) (Timeline|ScaleGridPreviewTL Play)
        (:Update
          (CallFunction|HandleScaleGridPreviewTLTick _direction _scale))
        (:Finished
          (CallFunction|OnScaleGridPreviewTLComplete _direction))))
    (:PlayFromStart
      (Timeline|ScaleGridPreviewTL PlayFromStart))
    (:Reverse
      (Timeline|ScaleGridPreviewTL Reverse))
    (:ReverseFromEnd
      (Timeline|ScaleGridPreviewTL ReverseFromEnd))))

(event EventPlayTransitionToIntermediateTL
  (params (PlaybackType :type EPreviewTimelinePlaybackType%20Enum :default "Play"))
  ;[1 1 1 0.5] Transition To Intermediate Timeline
  (switch Utilities|FlowControl|Switch|SwitchonEPreviewTimelinePlaybackType PlaybackType
    (:Play
      (bind (_direction _lerp) (Timeline|TransitionToIntermediateTL Play)
        (:Update
          (CallFunction|HandleTransitiontoIntermediateTLTick _lerp))
        (:Finished
          (CallFunction|OnTransitiontoIntermediateTLComplete))))
    (:PlayFromStart
      (Timeline|TransitionToIntermediateTL PlayFromStart))
    (:Reverse
      (Timeline|TransitionToIntermediateTL Reverse))
    (:ReverseFromEnd
      (Timeline|TransitionToIntermediateTL ReverseFromEnd))))

(event EventPlayScaleGuideAxisTL
  ;[1 1 1 0.5] Scale Axis GuideTimeline
  (bind (_direction _lerpcolor _opacity) (Timeline|ScaleGuideAxisTL PlayFromStart)
    (:Update
      (CallFunction|HandleScaleGuideAxisTLTick _lerpcolor _opacity))
    (:Finished
      (CallFunction|OnScaleGuideAxisTLComplete))))

(event EventPlayFadeScaleGuideTL
  (params (PlaybackType :type EPreviewTimelinePlaybackType%20Enum :default "Play"))
  ;[1 1 1 0.5] Fade Scale Guide Timeline
  (switch Utilities|FlowControl|Switch|SwitchonEPreviewTimelinePlaybackType PlaybackType
    (:Play
      (bind (_direction _lerp) (Timeline|FadeScaleGuideTL Play)
        (:Update
          (CallFunction|HandleFadeScaleGuideTLTick _lerp))
        (:Finished
          (CallFunction|HandleFadeScaleGuideTLComplete _direction))))
    (:PlayFromStart
      (Timeline|FadeScaleGuideTL PlayFromStart))
    (:Reverse
      (Timeline|FadeScaleGuideTL Reverse))
    (:ReverseFromEnd
      (Timeline|FadeScaleGuideTL ReverseFromEnd))))

(event EventPlayUpdateGridPreviewScaleTL
  ;[1 1 1 0.5] Update Grid Preview Scale Timeline
  (bind (_direction _lerp) (Timeline|UpdateGridPreviewScaleTL PlayFromStart)
    (:Update
      (CallFunction|HandleScaleGridPreviewTLTick _direction _lerp))
    (:Finished
      (CallFunction|OnScaleGridPreviewTLComplete _direction))))

(event EventPlayFadeGravityTL
  (params (PlaybackType :type EPreviewTimelinePlaybackType%20Enum :default "Play"))
  ;[1 1 1 0.5] Fade Gravity Timeline
  (switch Utilities|FlowControl|Switch|SwitchonEPreviewTimelinePlaybackType PlaybackType
    (:Play
      (bind (_direction _lerp) (Timeline|FadeGravityLineTL Play)
        (:Update
          (CallFunction|HandleFadeShadowPlaneTLTick _lerp))
        (:Finished
          (CallFunction|OnFadeShadowPlaneTLComplete _direction))))
    (:PlayFromStart
      (Timeline|FadeGravityLineTL PlayFromStart))
    (:Reverse
      (Timeline|FadeGravityLineTL Reverse))
    (:ReverseFromEnd
      (Timeline|FadeGravityLineTL ReverseFromEnd))))

(event EventPlayRotationGuideOrientationTL
  ;[1 1 1 0.5] Rotation Guide OrientationTimeline
  (bind (_direction _lerp) (Timeline|RotationGuideOrientationTL PlayFromStart)
    (:Update
      (CallFunction|HandleRotationGuideOrientationTLTick _lerp))
    (:Finished
      (CallFunction|OnRotationGuideOrientationTLComplete))))

(event EventPlayDuplicateTL
  ;[1 1 1 0.5] DuplicateTimeline
  (bind (_direction _gridintensity _diffuseadd) (Timeline|DuplicateTL PlayFromStart)
    (:Update
      (CallFunction|HandleDuplicateTLTick _gridintensity _diffuseadd))
    (:Finished
      (CallFunction|OnDuplicateAxisTLComplete))))

(event EventPlayTransitionTL
  (params (PlaybackType :type EPreviewTimelinePlaybackType%20Enum :default "Play"))
  ;[1 1 1 0.5] Transition Timeline
  (switch Utilities|FlowControl|Switch|SwitchonEPreviewTimelinePlaybackType PlaybackType
    (:Play
      (bind (_direction _material_lerp _particlerelativelocation _lightlerp) (Timeline|TransitionTL Play)
        (:Update
          (CallFunction|HandleTransitionTLTick _material_lerp _particlerelativelocation _lightlerp))
        (:Finished
          (CallFunction|OnTransitionTLComplete _direction))))
    (:PlayFromStart
      (Timeline|TransitionTL PlayFromStart))
    (:Reverse
      (Timeline|TransitionTL Reverse))
    (:ReverseFromEnd
      (Timeline|TransitionTL ReverseFromEnd))))

(fn GetGridAxisGuideAttachmentPointOverride
  (bind _returnvalue (CallFunction|GetBaseActor))
  (Utilities|IsValid _returnvalue
    (:Is%20Valid
      (bind (_override _relativetocreativeactororigin _relativeattachlocation) (CreativeMoveTool|GetGridAxisGuideAttachmentPoint(Message) _returnvalue)) ; Check for attachment location override
      (if _override
        (return _relativetocreativeactororigin _relativeattachlocation true)
        (else
          _)))
    (:Is%20Not%20Valid
      _))
  (return false "0, 0, 0" false))

(fn GetObjectCenter
  (bind _returnvalue (CallFunction|GetBaseActor))
  (Utilities|IsValid _returnvalue
    (:Is%20Valid
      (bind (_override _center) (CreativeMoveTool|GetCreativeActorOrigin(Message) _returnvalue)) ; Check for override from interface
      (if _override
        (return _center)
        (else
          (bind _asbuilding_wall (Utilities|Casting|CastToBuildingWall _returnvalue)
            (:then
              (Utilities|IsValid (Variables|/Script/FortniteGame.BuildingSMActor|GetStaticMeshComponent _asbuilding_wall)
                (:Is%20Valid
                  (return (* (+ (Components|StaticMesh|GetLocalBounds (Variables|/Script/FortniteGame.BuildingSMActor|GetStaticMeshComponent _asbuilding_wall)) (.Max (Components|StaticMesh|GetLocalBounds (Variables|/Script/FortniteGame.BuildingSMActor|GetStaticMeshComponent _asbuilding_wall)))) 0.5)))
                (:Is%20Not%20Valid
                  _)))
            (:CastFailed
              (bind _returnvalue_2 (|CallParentFunction))))
          (return _returnvalue_2))))
    (:Is%20Not%20Valid
      (return "0, 0, 0"))))

(fn GetOverrideActorBounds
  (bind _returnvalue (CallFunction|GetBaseActor))
  (Utilities|IsValid _returnvalue
    (:Is%20Valid
      (bind (_override_bounds _bounds) (CreativeMoveTool|GetCreativeActorBounds(Message) _returnvalue))
      (return _bounds _override_bounds))
    (:Is%20Not%20Valid
      (return "0, 0, 0" false))))

(fn ConstructionScript
  ;[1 1 1 0.5] Cache native versions of timelines
  (Variables|CreativeMoveToolBlueprintPreview|SetDeleteTLC (Variables|B_CreativeMoveTool_ActorPreview|GetDeleteTL))
  (Variables|CreativeMoveToolBlueprintPreview|SetTransitionTLC (Variables|B_CreativeMoveTool_ActorPreview|GetTransitionTL))
  (Variables|CreativeMoveToolBlueprintPreview|SetDuplicateTLC (Variables|B_CreativeMoveTool_ActorPreview|GetDuplicateTL))
  (Variables|CreativeMoveToolBlueprintPreview|SetRotationGuideOrientationTLC (Variables|B_CreativeMoveTool_ActorPreview|GetRotationGuideOrientationTL))
  (Variables|CreativeMoveToolBlueprintPreview|SetFadeGravityTLC (Variables|B_CreativeMoveTool_ActorPreview|GetFadeGravityLineTL))
  (Variables|CreativeMoveToolBlueprintPreview|SetUpdateGridPreviewScaleTLC (Variables|B_CreativeMoveTool_ActorPreview|GetUpdateGridPreviewScaleTL))
  (Variables|CreativeMoveToolBlueprintPreview|SetFadeScaleGuideTLC (Variables|B_CreativeMoveTool_ActorPreview|GetFadeScaleGuideTL))
  (Variables|CreativeMoveToolBlueprintPreview|SetTransitionToIntermediateTLC (Variables|B_CreativeMoveTool_ActorPreview|GetTransitionToIntermediateTL))
  (Variables|CreativeMoveToolBlueprintPreview|SetScaleGuideAxisTLC (Variables|B_CreativeMoveTool_ActorPreview|GetScaleGuideAxisTL))
  (Variables|CreativeMoveToolBlueprintPreview|SetScaleGridPreviewTLC (Variables|B_CreativeMoveTool_ActorPreview|GetScaleGridPreviewTL))
  (Variables|CreativeMoveToolBlueprintPreview|SetLerpGridTLC (Variables|B_CreativeMoveTool_ActorPreview|GetLerpGridTL))
  (Variables|CreativeMoveToolBlueprintPreview|SetDropPropTLC (Variables|B_CreativeMoveTool_ActorPreview|GetDropPropTL))
  (Variables|CreativeMoveToolBlueprintPreview|SetFadeRotationGuideTLC (Variables|B_CreativeMoveTool_ActorPreview|GetFadeRotationGuideTL)))
