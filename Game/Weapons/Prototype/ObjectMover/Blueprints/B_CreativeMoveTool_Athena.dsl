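; Creative-mode move tool Blueprint that layers audio, visual feedback and playset selection onto the native tool for moving, rotating, scaling and copying objects.
; Input overrides and bound dispatchers drive preview-actor guides and grid reveal, plus sounds, with a timer fading scale loops at size limits.
; Soft-referenced sounds avoid hard loads.
; Without cosmetics, it spawns and later destroys a phone-tool target mode; selection-limit failures trigger a failed-click ability.
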(event InputEvents|EventHandleRotateSingleClockwisePressed
  ;[1 1 1] Handle Rotation Events
  (|CallParentFunction)
  (CallFunction|HandleClockwiseRotationInputChange true))

(event InputEvents|EventHandleRotateClockwisePressed
  ;[1 1 1] Handle Rotation Events
  (|CallParentFunction)
  (CallFunction|HandleClockwiseRotationInputChange true))

(event InputEvents|EventHandleRotateSingleCounterclockwisePressed
  ;[1 1 1] Handle Rotation Events
  (|CallParentFunction)
  (CallFunction|HandleCounterClockwiseRotationInputChange true))

(event InputEvents|EventHandleRotateCounterclockwisePressed
  ;[1 1 1] Handle Rotation Events
  (|CallParentFunction)
  (CallFunction|HandleCounterClockwiseRotationInputChange true))

(event InputEvents|EventHandleRotateSingleClockwiseReleased
  ;[1 1 1] Handle Rotation Events
  (|CallParentFunction)
  (CallFunction|HandleClockwiseRotationInputChange false))

(event InputEvents|EventHandleRotateClockwiseReleased
  ;[1 1 1] Handle Rotation Events
  (|CallParentFunction)
  (CallFunction|HandleClockwiseRotationInputChange false))

(event InputEvents|EventHandleRotateSingleCounterclockwiseReleased
  ;[1 1 1] Handle Rotation Events
  (|CallParentFunction)
  (CallFunction|HandleCounterClockwiseRotationInputChange false))

(event InputEvents|EventHandleRotateCounterclockwiseReleased
  ;[1 1 1] Handle Rotation Events
  (|CallParentFunction)
  (CallFunction|HandleCounterClockwiseRotationInputChange false))

(event Custom|EventOnCouldNotAddObject
  (params (Object :type Object%20Reference) (NumSelected :type Integer))
  ;[1 1 1] Gameplay Abilities
  (CallFunction|ServerPlayGameplayAbility (CallFunction|GetMoveObjectsFreelyInteractionBehavior) "/Game/Weapons/Prototype/ObjectMover/Blueprints/ObjectInteractionBehaviors/GA_Failed_Click.GA_Failed_Click_C"))

(event Custom|ServerPlayGameplayAbility
  (params (ObjectInteractionBehavior :type Object%20Interaction%20Behavior%20Object%20Reference) (NumSelected :type Fort%20Gameplay%20Ability%20Class%20Reference))
  ;[1 1 1] Gameplay Abilities
  (Class|ObjectInteractionBehavior|PlayGameplayAbility ObjectInteractionBehavior NumSelected))

(event EventBeginPlay
  ;[1 1 1] Set up the playset inventory UI
  (|CallParentFunction)
  (Networking|CanExecuteCosmeticEvents
    (:True
      (EventDispatchers|BindEventtoOnMoveToolRotationAxisChanged (EventDispatchers|CreateEvent :!event UpdateRotationAxis))
      (EventDispatchers|BindEventtoOnMoveToolScaleInsteadOfRotateChanged (EventDispatchers|CreateEvent :!event ScaleInsteadOfRotate))
      (EventDispatchers|BindEventtoOnCouldNotAddReachedSelectionLimit (EventDispatchers|CreateEvent :!event Custom|EventOnCouldNotAddObject))
      (EventDispatchers|BindEventtoOnMoveToolPrecisionChanged (EventDispatchers|CreateEvent :!event PlayPrecisionFX))
      ;[1 1 1] Load Assets so we don't need to hard reference them when including the move tool
      (bind _assound_base (Utilities|SoundObjectfromReference "/Game/Sounds/Creative/Gadgets/MoveTool/MoveTool_ScalingUp_Cue.MoveTool_ScalingUp_Cue")
        (:Synchronous
          (bind _assound_base_2 (Utilities|SoundObjectfromReference "/Game/Sounds/Creative/Gadgets/MoveTool/MoveTool_ScalingDown_Cue.MoveTool_ScalingDown_Cue")
            (:Completed
              (Audio|Components|Audio|SetSound (Variables|Audio|GetMove%20Tool%20Scale%20Down) _assound_base_2))))
        (:Completed
          (Audio|Components|Audio|SetSound (Variables|Audio|GetMove%20Tool%20Scale%20Up) _assound_base))))
    (:False
      (CallFunction|EventInitializeObjectInteractionModeStack))))

(event CreativeMoveTool|EventOnAllowGravityToggleChanged
  (params (bAllowGravity :type Boolean))
  ;[1 1 1] Bind so we can call this in the previewer blueprint and drive visual aids
  (|CallParentFunction bAllowGravity)
  (Default|CallGravityToggleChanged bAllowGravity))

(event Custom|HandlePlaysetSelected
  (params (PlaysetItemEntry :type Fort%20Item%20Entry%20Structure))
  ;[1 1 1] Spawn Playsets
  (CallFunction|SpawnPlaysetonPlayer PlaysetItemEntry))

(event Custom|EventInitializeObjectInteractionModeStack
  ;[1 1 1] Construct ObjectInteractionMode Stack (establishes the priority each mode will be queried when looking for MoveTool target)
  (bind _returnvalue (Game|SpawnActorfromClass "/Game/Weapons/Prototype/ObjectMover/Blueprints/ObjectInteractionModes/PhoneToolActorTargets/CreativeInteractionMode_PhoneToolTargetMode.CreativeInteractionMode_PhoneToolTargetMode_C" (Transformation|GetActorTransform (Actor|GetOwner)) "Undefined" "MultiplyWithRoot" self 0))
  (Variables|Default|SetPhoneToolTargetMode _returnvalue))

(event EventEndPlay
  (params (EndPlayReason :type EEndPlayReason%20Enum :default "Destroyed"))
  ;[1 1 1] Cleanup
  (|CallParentFunction EndPlayReason)
  (Networking|CanExecuteCosmeticEvents
    (:False
      (Actor|DestroyActor (Variables|Default|GetPhoneToolTargetMode)))))

(event InputEvents|EventHandleScaleUpPressed
  ;[1 1 1] Input Sounds / Some FX
  (|CallParentFunction)
  (Audio|Components|Audio|Play (Variables|Audio|GetMove%20Tool%20Scale%20Up) 0.0)
  (for _array_element (Variables|FortCreativeMoveTool|GetActivePreviewActors)
    (Utilities|IsValid _array_element
      (:Is%20Valid
        (Class|CreativeMoveToolBlueprintPreview|FadeScaleGuideIn _array_element true)
        (Class|CreativeMoveToolBlueprintPreview|UpdateGridPreviewScale _array_element))))
  (Utilities|Time|SetTimerbyEvent (EventDispatchers|CreateEvent :!event Custom|Scale%20Up%20Modifier%20Audio) 0.1 true))

(event InputEvents|EventHandlePushPressed
  ;[1 1 1] Input Sounds / Some FX
  (|CallParentFunction)
  (CallFunction|PlayMoveToolSoundbySoftObjectReference "/Game/Sounds/Creative/Gadgets/MoveTool/MoveTool_Push_Cue.MoveTool_Push_Cue"))

(event InputEvents|EventHandlePullPressed
  ;[1 1 1] Input Sounds / Some FX
  (|CallParentFunction)
  (CallFunction|PlayMoveToolSoundbySoftObjectReference "/Game/Sounds/Creative/Gadgets/MoveTool/MoveTool_Pull_Cue.MoveTool_Pull_Cue"))

(event InputEvents|EventHandleRotateSwitchAxisPressed
  ;[1 1 1] Input Sounds / Some FX
  (|CallParentFunction)
  (CallFunction|PlayMoveToolSoundbySoftObjectReference "/Game/Sounds/Creative/Gadgets/MoveTool/MoveTool_Axis_Cue.MoveTool_Axis_Cue"))

(event InputEvents|EventHandleScaleDownPressed
  ;[1 1 1] Input Sounds / Some FX
  (|CallParentFunction)
  (Audio|Components|Audio|Play (Variables|Audio|GetMove%20Tool%20Scale%20Down) 0.0)
  (for _array_element (Variables|FortCreativeMoveTool|GetActivePreviewActors)
    (Utilities|IsValid _array_element
      (:Is%20Valid
        (Class|CreativeMoveToolBlueprintPreview|FadeScaleGuideIn _array_element false)
        (Class|CreativeMoveToolBlueprintPreview|UpdateGridPreviewScale _array_element))))
  (Utilities|Time|SetTimerbyEvent (EventDispatchers|CreateEvent :!event Custom|Scale%20Up%20Modifier%20Audio) 0.1 true))

(event InputEvents|EventHandleResetRotationPressed
  ;[1 1 1] Input Sounds / Some FX
  (|CallParentFunction)
  (CallFunction|PlayMoveToolSoundbySoftObjectReference "/Game/Sounds/Creative/Gadgets/MoveTool/MoveTool_Rotate_Cue.MoveTool_Rotate_Cue"))

(event InputEvents|EventHandleDropToFloorPressed
  ;[1 1 1] Input Sounds / Some FX
  (|CallParentFunction)
  (if (not (CallFunction|IsDropsDisabled))
    (CallFunction|PlayMoveToolSoundbySoftObjectReference "/Game/Sounds/Creative/Gadgets/MoveTool/MoveTool_Axis_Cue.MoveTool_Axis_Cue")))

(event InputEvents|EventHandleScaleUpReleased
  (bind _move_tool_scale_up (Variables|Audio|GetMove%20Tool%20Scale%20Up))
  ;[1 1 1] Input Sounds / Some FX
  (|CallParentFunction)
  (for _array_element (Variables|FortCreativeMoveTool|GetActivePreviewActors)
    (Utilities|IsValid _array_element
      (:Is%20Valid
        (Class|CreativeMoveToolBlueprintPreview|FadeScaleGuideOut _array_element)
        (Class|CreativeMoveToolBlueprintPreview|StopGridPreviewScale _array_element))))
  (if (Audio|Components|Audio|IsPlaying _move_tool_scale_up)
    (Audio|Components|Audio|FadeOut _move_tool_scale_up 0.25 0.0 "Linear" :!class /Script/Engine.AudioComponent)
    (Utilities|Time|ClearandInvalidateTimerbyHandle (Variables|Default|GetHandle%20Scale%20Up%20Modifier))))

(event InputEvents|EventHandleExitHeldDownLongEnough
  (bind _move_tool_scale_down (Variables|Audio|GetMove%20Tool%20Scale%20Down))
  (bind _move_tool_scale_up (Variables|Audio|GetMove%20Tool%20Scale%20Up))
  ;[1 1 1] Input Sounds / Some FX
  (|CallParentFunction)
  (CallFunction|PlayMoveToolSoundbySoftObjectReference "/Game/Sounds/Creative/Gadgets/MoveTool/MoveTool_Axis_Cue.MoveTool_Axis_Cue")
  (Utilities|FlowControl|Sequence
    (:then_0
      (if (Audio|Components|Audio|IsPlaying _move_tool_scale_up)
        (Audio|Components|Audio|FadeOut _move_tool_scale_up 0.25 0.0 "Linear" :!class /Script/Engine.AudioComponent)))
    (:then_1
      (if (Audio|Components|Audio|IsPlaying _move_tool_scale_down)
        (Audio|Components|Audio|FadeOut _move_tool_scale_down 0.25 0.0 "Linear" :!class /Script/Engine.AudioComponent))))
  (Utilities|Time|ClearandInvalidateTimerbyHandle (Variables|Default|GetHandle%20Scale%20Up%20Modifier)))

(event InputEvents|EventHandleResetScalePressed
  ;[1 1 1] Input Sounds / Some FX
  (|CallParentFunction)
  (CallFunction|PlayMoveToolSoundbySoftObjectReference "/Game/Sounds/Creative/Gadgets/MoveTool/MoveTool_Rotate_Cue.MoveTool_Rotate_Cue")
  (for _array_element (Variables|FortCreativeMoveTool|GetActivePreviewActors)
    (Utilities|IsValid _array_element
      (:Is%20Valid
        (Class|CreativeMoveToolBlueprintPreview|UpdateGridPreviewScale _array_element)))))

(event InputEvents|EventHandleScaleDownReleased
  (bind _move_tool_scale_down (Variables|Audio|GetMove%20Tool%20Scale%20Down))
  ;[1 1 1] Input Sounds / Some FX
  (|CallParentFunction)
  (for _array_element (Variables|FortCreativeMoveTool|GetActivePreviewActors)
    (Utilities|IsValid _array_element
      (:Is%20Valid
        (Class|CreativeMoveToolBlueprintPreview|FadeScaleGuideOut _array_element)
        (Class|CreativeMoveToolBlueprintPreview|StopGridPreviewScale _array_element))))
  (if (Audio|Components|Audio|IsPlaying _move_tool_scale_down)
    (Audio|Components|Audio|FadeOut _move_tool_scale_down 0.25 0.0 "Linear" :!class /Script/Engine.AudioComponent)
    (Utilities|Time|ClearandInvalidateTimerbyHandle (Variables|Default|GetHandle%20Scale%20Up%20Modifier))))

(event Custom|Scale%20Up%20Modifier%20Audio
  (bind _returnvalue (CreativeMoveTool|GetAverageScaleModifier))
  ;[1 1 1] Input Sounds / Some FX
  (Utilities|FlowControl|Sequence
    (:then_0
      (if (and (>= _returnvalue 3.0) (Audio|Components|Audio|IsPlaying (Variables|Audio|GetMove%20Tool%20Scale%20Up)))
        (Audio|Components|Audio|FadeOut (Variables|Audio|GetMove%20Tool%20Scale%20Up) 0.25 0.0 "Linear" :!class /Script/Engine.AudioComponent)))
    (:then_1
      (if (and (<= _returnvalue 0.333) (Audio|Components|Audio|IsPlaying (Variables|Audio|GetMove%20Tool%20Scale%20Down)))
        (Audio|Components|Audio|FadeOut (Variables|Audio|GetMove%20Tool%20Scale%20Down) 0.25 0.0 "Linear" :!class /Script/Engine.AudioComponent))))
  (Utilities|Time|ClearandInvalidateTimerbyHandle (Variables|Default|GetHandle%20Scale%20Up%20Modifier)))

(event InputEvents|EventHandleScaleSwitchAxisPressed
  ;[1 1 1] Input Sounds / Some FX
  (|CallParentFunction)
  (CallFunction|PlayMoveToolSoundbySoftObjectReference "/Game/Sounds/Creative/Gadgets/MoveTool/MoveTool_Axis_Cue.MoveTool_Axis_Cue")
  (for _array_element (Variables|FortCreativeMoveTool|GetActivePreviewActors)
    (Utilities|IsValid _array_element
      (:Is%20Valid
        (Class|CreativeMoveToolBlueprintPreview|UpdateScaleGuide _array_element)))))

(event Custom|UpdateRotationAxis
  (params (Rotation%20Axis :type Byte))
  ;[1 1 1] Input Sounds / Some FX
  (for _array_element (Variables|FortCreativeMoveTool|GetActivePreviewActors)
    (Utilities|IsValid _array_element
      (:Is%20Valid
        (Class|CreativeMoveToolBlueprintPreview|UpdateRotationGuide _array_element)))))

(event InputEvents|EventHandleCopyPickupPressed
  ;[1 1 1] Input Sounds / Some FX
  (|CallParentFunction)
  (if (Variables|Default|GetOnHoverChangedCanInteract)
    (CallFunction|PlayMoveToolSoundbySoftObjectReference (select (CallFunction|CheckforMultiSelection) "/Game/Sounds/Creative/Gadgets/MoveTool/MoveTool_MultiSelect_Duplicate_Cue.MoveTool_MultiSelect_Duplicate_Cue" "/Game/Sounds/Creative/Gadgets/MoveTool/MoveTool_Copy_01_Cue.MoveTool_Copy_01_Cue"))
    (Variables|Default|SetMulti%20Selected%20Count 0)))

(event Custom|ScaleInsteadOfRotate
  (params (Scale :type Boolean))
  ;[1 1 1] Input Sounds / Some FX
  (Utilities|FlowControl|Sequence
    (:then_0
      (if (Math|Boolean|NotEqual(Boolean) Scale (Variables|Default|GetbWasScaling))
        (for _array_element (Variables|FortCreativeMoveTool|GetActivePreviewActors)
          (Utilities|IsValid _array_element
            (:Is%20Valid
              (if Scale
                (Class|CreativeMoveToolBlueprintPreview|UpdateScaleGuide _array_element)
                (Class|CreativeMoveToolBlueprintPreview|SetRotationModeActive _array_element false)
                (Class|CreativeMoveToolBlueprintPreview|SetScaleModeActive _array_element true)
                (else
                  (Class|CreativeMoveToolBlueprintPreview|UpdateRotationGuide _array_element)
                  (Class|CreativeMoveToolBlueprintPreview|SetRotationModeActive _array_element true)
                  (Class|CreativeMoveToolBlueprintPreview|SetScaleModeActive _array_element false))))))))
    (:then_1
      (Variables|Default|SetbWasScaling Scale))))

(event InputEvents|EventHandlePrecisionModePressed
  ;[1 1 1] Input Sounds / Some FX
  (|CallParentFunction)
  (Variables|Default|SetHasPlayedHitTraceRuleSound false))

(event InputEvents|EventHandlePrecisionModeReleased
  ;[1 1 1] Input Sounds / Some FX
  (|CallParentFunction)
  (if (not (Variables|Default|GetHasPlayedHitTraceRuleSound))
    (CallFunction|PlayPrecisionFX false)))

(event InputEvents|EventHandleHitTraceRulePressed
  ;[1 1 1] Input Sounds / Some FX
  (|CallParentFunction)
  (Variables|Default|SetHasPlayedHitTraceRuleSound false))

(event InputEvents|EventHandleHitTraceRuleHeldDownLongEnough
  ;[1 1 1] Input Sounds / Some FX
  (|CallParentFunction)
  (if (not (Variables|Default|GetHasPlayedHitTraceRuleSound))
    (CallFunction|PlayMoveToolSoundbySoftObjectReference "/Game/Sounds/Creative/Gadgets/MoveTool/MoveTool_Axis_Cue.MoveTool_Axis_Cue")
    (Variables|Default|SetHasPlayedHitTraceRuleSound true)))

(event InputEvents|EventHandleClearMultiSelectQueuePressed
  ;[1 1 1] Input Sounds / Some FX
  (|CallParentFunction)
  (if (> (Variables|Default|GetMulti%20Selected%20Count) 0)
    (CallFunction|PlayMoveToolSoundbySoftObjectReference "/Game/Sounds/Creative/Gadgets/MoveTool/MoveTool_MultiSelect_Cancel_Cue.MoveTool_MultiSelect_Cancel_Cue")
    (Variables|Default|SetMulti%20Selected%20Count 0)))

(event Custom|HandleAddToQuickBarPressed_1
  ;[1 1 1] Input Sounds / Some FX
  (if (Variables|Default|GetOnHoverChangedCanInteract)
    (bind _assound_base (Utilities|SoundObjectfromReference "/Game/Sounds/Creative/Gadgets/MoveTool/MoveTool_Granted_02_Cue.MoveTool_Granted_02_Cue")
      (:Completed
        (Audio|PlaySoundatLocation _assound_base (Transformation|GetActorLocation))))))

(event Custom|PlayPrecisionFX
  (params (On :type Boolean) (PrecisionLevel :type Byte))
  ;[1 1 1] Input Sounds / Some FX
  (Utilities|FlowControl|Sequence
    (:then_0
      (for _array_element (Variables|FortCreativeMoveTool|GetActivePreviewActors)
        (Utilities|IsValid _array_element
          (:Is%20Valid
            (Class|CreativeMoveToolBlueprintPreview|RevealGrid _array_element (select (CallFunction|GetShouldUsePrecisionGridSnapping) (Math|Float|Truncate (.x (Utilities|Array|Get(aref) (CallFunction|GetGridSnapRatios) (Math|Conversions|ToInteger(Byte) (CallFunction|GetGridSnapIndex))))) 0))))))
    (:then_1
      (if (not (Variables|Default|GetHasPlayedHitTraceRuleSound))
        (CallFunction|PlayMoveToolSoundbySoftObjectReference "/Game/Sounds/Creative/Gadgets/MoveTool/MoveTool_Equip_Cue.MoveTool_Equip_Cue")))))

(fn HandleClockwiseRotationInputChange
  (params (bPressed :type Boolean))
  (for _array_element (Variables|FortCreativeMoveTool|GetActivePreviewActors)
    (Utilities|IsValid _array_element
      (:Is%20Valid
        (Class|CreativeMoveToolBlueprintPreview|HandleRotationClockwiseInputChange _array_element (Variables|GetbPressed))))))

(fn HandleCounterClockwiseRotationInputChange
  (params (bPressed :type Boolean))
  (for _array_element (Variables|FortCreativeMoveTool|GetActivePreviewActors)
    (Utilities|IsValid _array_element
      (:Is%20Valid
        (Class|CreativeMoveToolBlueprintPreview|HandleRotationCounterClockwiseInputChange _array_element (Variables|GetbPressed))))))

(fn SpawnPlaysetOnPlayer
  (params (PlaysetItemEntry :type Fort%20Item%20Entry%20Structure))
  (bind _asfort_weapon_ranged_item_definition (Utilities|Casting|CastToFortWeaponRangedItemDefinition (Class|FortItemFunctionLibrary|BreakFortItemEntry PlaysetItemEntry)))
  (Game|ConstructObjectfromClass (WeaponItemDefinition|GetPrimaryFireAbility _asfort_weapon_ranged_item_definition) self)
  (bind _asfort_playset_grenade_item_definition (Utilities|Casting|CastToFortPlaysetGrenadeItemDefinition _asfort_weapon_ranged_item_definition))
  (CreativeMoveTool|OnPlaysetSelected (Class|FortPlaysetGrenadeItemDefinition|GetPlaysettoSpawn _asfort_playset_grenade_item_definition false :!class /Script/FortniteGame.FortPlaysetGrenadeItemDefinition)))

(fn ConstructionScript
  (|CallParentFunction))
