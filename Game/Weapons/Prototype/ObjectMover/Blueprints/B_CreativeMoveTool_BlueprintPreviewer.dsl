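; Preview actor for the Fortnite Creative move tool, showing a blueprint-style stand-in of the targeted actor with grid, gravity, rotation and scale guides.
; It async-loads replacement materials, then mirrors static, spline, instanced, skeletal and geometry-collection meshes into indexed components with parameter-copying dynamic materials.
; Tick line-traces the gravity guide and shadow plane and updates rotation-locked meshes.
; Ending play can restore original materials, with visual setup limited to cosmetic-capable instances.
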(event EventBeginPlay)

(event EventEndPlay
  (params (EndPlayReason :type EEndPlayReason%20Enum :default "Destroyed"))
  ;[1 1 1] End Play Safety. Make Sure to set the object back to its original size
  (Networking|CanExecuteCosmeticEvents
    (:True
      (if (Variables|Default|GetWantsRestoreOriginalMaterialsWhenDestroy)
        (CallFunction|RestoreOriginalMeshComponentMaterials)))
    (:False
      _))
  (CallFunction|SetEndSafetyParameters)
  (|CallParentFunction EndPlayReason))

(event Custom|DestroyPreviewOnServerRPC
  ;[1 1 1] Destroy This Actor
  (CallFunction|DestroyPreview))

(event Custom|DestroyPreview
  ;[1 1 1] Destroy This Actor
  (Actor|DestroyActor))

(event Custom|ClientMakeDestroyOnServerRequest
  ;[1 1 1] Destroy This Actor
  (if (not (Variables|Default|GetIsAskingForDestruction))
    (Variables|Default|SetIsAskingForDestruction true)
    (CallFunction|DestroyPreviewonServerRPC)))

(event Custom|Event%20Create%20Blueprint%20Preview
  (params (CreativeInteractionTool :type Fort%20Creative%20Move%20Tool%20Object%20Reference) (PropOwningActor :type Actor%20Object%20Reference))
  ;[1 1 1] Create Blueprint - This event is needed every time we create a new preview
  (Networking|CanExecuteCosmeticEvents
    (:True
      (CallFunction|SetCreativeInteractionTool CreativeInteractionTool)
      (Variables|Default|SetPropOwningActor PropOwningActor)
      (Variables|Default|GetBaseActor
        (:then
          (if (not (Utilities|Enum|Equal(Enum) (Variables|Default|GetFXState) "NewEnumerator3"))
            ;[1 1 1] Replacment material loading so we can do things like common parts.
            (bind _softmaterialstoload (CallFunction|GatherReplacementMaterials))
            (if (> (Utilities|Array|Length _softmaterialstoload) 0)
              (Variables|Default|SetPendingLoads (Utilities|Array|Length _softmaterialstoload))
              (for _array_element _softmaterialstoload
                (bind _object (Utilities|AsyncLoadAsset _array_element)
                  (:Completed
                    (bind _asmaterial_instance_constant (Utilities|Casting|CastToMaterialInstanceConstant _object)
                      (:then
                        (Utilities|Array|Add (Variables|Default|GetReplacementMaterialsFromActor) _asmaterial_instance_constant))
                      (:CastFailed
                        _))
                    (bind _output_get (Variables|Default|SetPendingLoads (- (Variables|Default|GetPendingLoads) 1)))
                    (if (== _output_get 0)
                      (CallFunction|CreateBlueprintPreview :!label _label)
                      (CallFunction|GridAxisGuideSetup)
                      (CallFunction|GravityGuideSetup)
                      (CallFunction|ShadowPlaneSetup)
                      (CallFunction|RotationGuideInitialSetup)
                      (CallFunction|ScaleGuideInitialSetup)
                      (CallFunction|ParticlePropHelperSetup)
                      (Default|UpdateRelativeLocationOfPivotPointAndGuides)))))))
          (goto _label))
        (:else)))))

(event EventTick
  (params (DeltaSeconds :type Float%20%28single-precision%29))
  (bind (_min _max) (Components|StaticMesh|GetLocalBounds (Variables|Default|GetGravityGuide)))
  (bind _max_x (Math|Vector|BreakVector _max))
  (bind _gravityguide (Variables|Default|GetGravityGuide))
  (bind _returnvalue (Transformation|GetWorldLocation _gravityguide))
  (bind _shadowplanescalefudgefactor (Variables|GravityGuide|GetShadowPlaneScaleFudgeFactor))
  (bind _maxlocalboundsxy (Variables|Bounds|GetMaxLocalBoundsXY))
  (bind _returnvalue_2 (* _maxlocalboundsxy _shadowplanescalefudgefactor))
  (bind _returnvalue_3 (Math|Conversions|ToVector(Float) _returnvalue_2))
  (bind _shadowplane (Variables|Default|GetShadowPlane))
  (bind _bhiddeningame (Variables|/Script/Engine.SceneComponent|GetbHiddenInGame _shadowplane))
  (bind _gravityguidemintracedistance (Variables|Default|GetGravityGuideMinTraceDistance))
  ;[1 1 1] Gravity Mode Visualizer
  (Utilities|FlowControl|Sequence
    (:then_0
      (if (Variables|GravityGuide|GetGravityModeActive)
        ;; GravityLineTrace
        (Networking|CanExecuteCosmeticEvents
          (:True
            (bind (_outhit _returnvalue_4) (Collision|LineTraceByChannel :Start (Transformation|GetWorldLocation (Variables|Default|GetGravityGuide)) :End (+ (Transformation|GetWorldLocation (Variables|Default|GetGravityGuide)) (Variables|GravityGuide|GetGravityLineTraceBias)) :TraceChannel "TraceTypeQuery10" :bTraceComplex false :DrawDebugType "None" :bIgnoreSelf true :TraceColor "(R=1.000000,G=0.000000,B=0.000000,A=1.000000)" :TraceHitColor "(R=0.000000,G=1.000000,B=0.000000,A=1.000000)" :DrawTime 0.25))
            (bind (_outhit_bblockinghit _outhit_binitialoverlap _outhit_time _outhit_distance _outhit_location) (Collision|BreakHitResult _outhit))
            (if (and (not _bhiddeningame) (or (not _returnvalue_4) (<= (Math|Vector|VectorLength (- _returnvalue _outhit_location)) _gravityguidemintracedistance)))
              (CallFunction|FadeShadowPlaneOut)
              (elif (and _bhiddeningame (not (<= (Math|Vector|VectorLength (- _returnvalue _outhit_location)) _gravityguidemintracedistance)))
                (CallFunction|FadeShadowPlaneIn)))
            (Transformation|SetWorldLocation (Variables|Default|GetShadowPlane) (Math|Vector|vector+vector _outhit_location (Math|Vector|MakeVector 0.0 0.0 (Variables|GravityGuide|GetShadowPlaneZOffset))) false false)
            (Transformation|SetWorldScale3D (Variables|Default|GetGravityGuide) (Math|Vector|MakeVector (/ (Math|Vector|VectorLength (- (Transformation|GetWorldLocation (Variables|Default|GetGravityGuide)) _outhit_location)) _max_x) 1.0 1.0))
            (bind _baseactor (Variables|Default|GetBaseActor)
              (:then
                (bind (_returnvalue_x _returnvalue_y _returnvalue_z) (Math|Vector|BreakVector (Transformation|GetActorRelativeScale3D _baseactor)))
                (Transformation|SetRelativeScale3D (Variables|Default|GetShadowPlane) (* _returnvalue_3 (Math|Float|Max(Float) _returnvalue_x _returnvalue_y _returnvalue_z))))
              (:else
                (Transformation|SetRelativeScale3D (Variables|Default|GetShadowPlane) _returnvalue_3))))
          (:False
            _))))
    (:then_1
      (if (Utilities|IsMapNonEmpty (Variables|Default|GetLockedStaticMeshComponents))
        (CallFunction|UpdateRotationLockedObjects)))))

(event Custom|Server%20SetFXState%20With%20Notify
  (params (New%20FX%20State :type Enum_CreativeMoveTool_PreviewFX%20Enum :default "NewEnumerator0"))
  (Variables|Default|SetFXState New%20FX%20State)
  (CallFunction|UpdateFXState))

(event Custom|ReturnActorToPool
  ;[1 1 1] Return To Pool (not currently working)
  (Rendering|SetVisibility (Variables|Default|GetDefaultSceneRoot) false true)
  (Transformation|DetachFromActor "KeepRelative"))

(fn AttachGridAxisGuide
  (Transformation|AttachComponentToComponent (Variables|Default|GetGridAxisGuide) (Utilities|Array|Get(acopy) (Variables|Default|GetMeshComponents)) "None" "KeepRelative" "KeepRelative" "KeepRelative" true)
  ;[0.494 1 0.2481 0.5] Determine the relative offset of the grid axis guide attachment point, for actors with strange offsets (applies to some props brought over from BR)
  (bind _baseactor (Variables|Default|GetBaseActor)
    (:then
      (bind (_override _relativetocreativeactororigin _relativeattachlocation) (CreativeMoveTool|GetGridAxisGuideAttachmentPoint(Message) _baseactor))
      (if _override
        (if _relativetocreativeactororigin
          (LocalVariables|SetRelativeOffset (+ (CallFunction|GetActorCenter) _relativeattachlocation))
          (else
            (LocalVariables|SetRelativeOffset _relativeattachlocation)))
        (else
          (LocalVariables|SetRelativeOffset (CallFunction|GetActorCenter) :!label _label)))
      (Transformation|SetRelativeTransform (Variables|Default|GetGridAxisGuide) (Math|Transform|MakeTransform (+ (LocalVariables|GetRelativeOffset) "0, 0,2.000000")) false false))
    (:else
      _))
  (goto _label))

(fn AttemptSetTickEnabled
  (params (Enabled :type Boolean))
  (bind _tickactiverequests (Variables|Default|GetTickActiveRequests))
  (if Enabled
    (Math|Integer|IncrementInt _tickactiverequests
      (:%20%20
        (Actor|Tick|SetActorTickEnabled true)))
    (else
      (bind _result (Math|Integer|DecrementInt _tickactiverequests)
        (:%20%20
          (if (== _result 0)
            (Actor|Tick|SetActorTickEnabled false)))))))

(fn CalculateGridPreviewScale
  (bind _gridpreviewboundsscale (Variables|Default|GetGridPreviewBoundsScale))
  (bind _baseactor (Variables|Default|GetBaseActor)
    (:then
      (bind (_override_bounds _bounds) (CreativeMoveTool|GetCreativeActorBounds(Message) _baseactor))
      (if _override_bounds
        (return (* _bounds _gridpreviewboundsscale))
        (else
          (return (* _gridpreviewboundsscale (.OutBoxExtent (Fortnite|GetActorCollisionComponentBounds _baseactor)))))))
    (:else
      (return "0, 0, 0"))))

(fn CheckforDestructionAnimationwhenTimelineCompleted
  (Networking|SwitchHasAuthority
    (:Remote
      (if (Variables|Default|GetPlayTransitionBackwardWhenAnimationEnds)
        (if (not (Components|Timeline|IsPlaying (Variables|B_CreativeMoveTool_BlueprintPreviewer|GetTransitionTL)))
          (CallFunction|PlayTransitionBetweenMeshes "Backward"))))))

(fn CleanUpTemporaryStaticMeshComponentStandIns
  (for _array_element (Variables|Default|GetStandInTempStaticMeshComponents)
    (Components|DestroyComponent _array_element))
  (Utilities|Array|Clear (Variables|Default|GetStandInTempStaticMeshComponents)))

(fn ClearComponentVariablesOnLoop
  (Variables|BlueprintPreview|SetCurrentStaticMeshComponent 0)
  (Variables|Default|SetCurrentSkeletalMeshObject 0)
  (Variables|Default|SetCurrentSkeletalMeshComponent 0)
  (Variables|Default|SetCurrentBlueprintSkeletalMeshComonent 0)
  (Variables|Default|SetCurrentBlueprintStaticMeshComponent 0)
  (Variables|Default|SetCurrentGeometryCollectionComponent 0)
  (Variables|Default|SetCurrentBlueprintGeometryCollectionComponent 0)
  (Variables|BlueprintPreview|SetisStaticMeshComponent false)
  (Variables|BlueprintPreview|SetisSkelatalMeshComponent false)
  (Variables|BlueprintPreview|SetRenderComponent false)
  (Variables|Bounds|SetMeshComponentTransform)
  (Utilities|Array|Clear (Variables|Default|GetStandInTempStaticMeshComponents)))

(fn ClearVariablesOnInitialSetup
  (Utilities|Array|Clear (Variables|Default|GetStaticMeshComponents))
  (Utilities|Map|Clear (Variables|Default|GetLockedStaticMeshComponents))
  (Utilities|Array|Clear (Variables|Default|GetMeshComponentMIDs))
  (Utilities|Array|Clear (Variables|BlueprintPreview|GetBlueprintMIDs))
  (Utilities|Array|Clear (Variables|Default|GetSkeletalMeshObjects))
  (Utilities|Map|Clear (Variables|Default|GetOriginalMeshComponentMaterials))
  (Utilities|Array|Clear (Variables|Default|GetCurrentSkeletalMeshComponents))
  (Utilities|Array|Clear (Variables|Default|GetOriginalMeshComponentMaterialIDs))
  (Variables|Bounds|SetMaxLocalBoundsZ 0.0)
  (Variables|Bounds|SetMaxLocalBoundsXY 0.0)
  (Variables|Bounds|SetMaxComponentRadius 0.0)
  (Variables|Default|SetComponentBoundsMin "999999.000000,999999.000000,999999.000000")
  (Variables|Default|SetComponentBoundsMax "0, 0, 0")
  (Variables|Bounds|SetComponentTrueCenter "0, 0, 0")
  (Variables|Default|SetActorIsProp false))

(fn ConfirmMeshComponents
  ;[1 0.8195 0 0.5] Re-set MeshComponents to catch some things that are incorrectly invalid (can happen with mesh components added during runtime)
  (bind _ascreative_move_tool_overrides (Utilities|Casting|CastToCreativeMoveToolOverrides (Variables|Default|GetBaseActor)))
  (if (Utilities|Array|IsNotEmpty (Class|CreativeMoveToolOverrides|GetMeshComponents _ascreative_move_tool_overrides))
    (Variables|Default|SetMeshComponents (Class|CreativeMoveToolOverrides|GetMeshComponents _ascreative_move_tool_overrides))))

(fn CopyMeshComponentMaterialParametersToBlueprint
  (params (SourcePrimitiveComponent :type Primitive%20Component%20Object%20Reference) (Material%20ID :type Integer) (TargetMID :type Material%20Instance%20Dynamic%20Object%20Reference))
  (Utilities|IsValid (Variables|GetTargetMID)
    (:Is%20Valid
      ;[1 1 1] Diffuse / Base Color
      (bind _success (CallFunction|GetSetTextureParameter (Variables|GetSourcePrimitiveComponent) (Variables|GetMaterial%20ID) "Diffuse" "Diffuse" (Variables|GetTargetMID)))
      (if (not _success)
        (CallFunction|GetSetTextureParameter (Variables|GetSourcePrimitiveComponent) (Variables|GetMaterial%20ID) "Base Color" "Diffuse" (Variables|GetTargetMID)))
      ;[1 1 1] Normal Map
      (CallFunction|GetSetTextureParameter (Variables|GetSourcePrimitiveComponent) (Variables|GetMaterial%20ID) "Normals" "Normals" (Variables|GetTargetMID))
      ;[1 1 1] Opacity / Masked Map
      (bind _success_2 (CallFunction|GetSetTextureParameter (Variables|GetSourcePrimitiveComponent) (Variables|GetMaterial%20ID) "MaskTexture" "MaskTexture" (Variables|GetTargetMID)))
      (if (not _success_2)
        (bind _success_3 (CallFunction|GetSetTextureParameter (Variables|GetSourcePrimitiveComponent) (Variables|GetMaterial%20ID) "OpacityMask" "MaskTexture" (Variables|GetTargetMID)))
        (if (not _success_3)
          (CallFunction|GetSetTextureParameter (Variables|GetSourcePrimitiveComponent) (Variables|GetMaterial%20ID) "MaskTexture_OpaqueCanopy" "MaskTexture" (Variables|GetTargetMID)))))))

(fn CreateBlueprintPreview
  (bind (_buildingmeshcomponent _buildpreviewmarkeractor _bool) (CallFunction|IsBuildingPiece))
  (CallFunction|ClearVariablesonInitialSetup)
  (Initialization|ValidateMeshComponents)
  (Utilities|Casting|CastToBuildingProp (Variables|Default|GetBaseActor)
    (:then
      (Variables|Default|SetActorIsProp true))
    (:CastFailed
      (Variables|Default|SetActorIsProp false)))
  (Variables|Default|SetDropEndTransform (Transformation|GetActorTransform))
  (Rendering|SetVisibility (Variables|Default|GetDefaultSceneRoot) true false)
  (Transformation|AttachActorToActor (select _bool _buildpreviewmarkeractor (Variables|Default|GetBaseActor)))
  (Transformation|ResetRelativeTransform (Variables|Transformation|GetRootComponent))
  (Variables|Default|SetNumMeshesLoadingOnTarget 0)
  (for (_array_element _array_index) (Variables|Default|GetMeshComponents)
    (Variables|Default|SetCurrentIndex _array_index)
    (Utilities|IsValid _array_element
      (:Is%20Valid
        (CallFunction|ClearComponentVariablesonLoop)
        (bind (_success _bdelayuntilassetsloaded) (CallFunction|GetStaticMeshComponent _array_element))
        (if _success
          _
          (elif (not _bdelayuntilassetsloaded)
            (bind _success_2 (CallFunction|GetSkeletalMeshComponent _array_element))
            (if (not _success_2)
              (bind _bsuccessful (CallFunction|GetWidgetComponent _array_element))
              (if (not _bsuccessful)
                (bind _asgeometry_collection_component (Utilities|Casting|CastToGeometryCollectionComponent _array_element))
                (for (_ _array_index_2) (Utilities|Struct|BreakGeometryCollectionProxyMeshData (Variables|/Script/GeometryCollectionEngine.GeometryCollection|GetRootProxyData (Variables|/Script/GeometryCollectionEngine.GeometryCollectionComponent|GetRestCollection _asgeometry_collection_component)))
                  (if (== _array_index_2 0)
                    (bind _success_3 (CallFunction|GetGeometryCollectionComponent _asgeometry_collection_component _array_index_2))
                    (if _success_3
                      (for _array_element_2 (Variables|Default|GetStandInTempStaticMeshComponents)
                        (Variables|BlueprintPreview|SetCurrentMeshComponent _array_element_2)
                        (CallFunction|GetMaterialsandCopyParameters :!label _label))
                      (CallFunction|CleanUpTemporaryStaticMeshComponentStandIns))))))
            (else
              (Math|Integer|IncrementInt (Variables|Default|GetNumMeshesLoadingOnTarget)))))
        (goto _label)))))

(fn CreateMeshComponentMIDs
  (params (SaveOriginalMaterials :type Boolean) (MeshComponent :type Mesh%20Component%20Object%20Reference) (Index :type Integer) (MaterialInterface :type Material%20Interface%20Object%20Reference))
  (bind _meshcomponentmids (Variables|Default|GetMeshComponentMIDs))
  (Utilities|IsValid MeshComponent
    (:Is%20Valid
      (CallFunction|StoreOriginalMeshComponentMaterials (Variables|GetSaveOriginalMaterials) (Variables|GetIndex) (Variables|GetMeshComponent))
      (if (Utilities|Equal(Object) (Game|GetInstigatorController) (Game|GetPlayerController 0))
        ;[1 1 1] For local clients only
        (bind _asmaterial_instance_dynamic (Utilities|Casting|CastToMaterialInstanceDynamic (Variables|GetMaterialInterface)))
        (Utilities|Array|Add (Variables|Default|GetMeshComponentMIDs) _asmaterial_instance_dynamic)
        (else
          (bind _asmaterial_instance_dynamic_2 (Utilities|Casting|CastToMaterialInstanceDynamic (Variables|GetMaterialInterface))
            (:then
              (bind _returnvalue (Rendering|Material|CreateDynamicMaterialInstance (Variables|GetMeshComponent) (Variables|GetIndex) (Variables|/Script/Engine.MaterialInstance|GetParent _asmaterial_instance_dynamic_2) "None" :!class /Script/Engine.PrimitiveComponent))
              (Utilities|Array|Add _meshcomponentmids _returnvalue)
              (Rendering|Material|CopyMaterialInstanceParameters _returnvalue _asmaterial_instance_dynamic_2 true))
            (:CastFailed
              (bind _returnvalue_2 (Rendering|Material|CreateDynamicMaterialInstance (Variables|GetMeshComponent) (Variables|GetIndex) (Variables|GetMaterialInterface) "None" :!class /Script/Engine.PrimitiveComponent))
              (Utilities|Array|Add _meshcomponentmids _returnvalue_2)))))
      (return))))

(fn CreatureVisualsLoaded
  (CallFunction|SetupPreview)
  (bind _asbuilding_prop_creature_placer (Utilities|Casting|CastToBuildingProp_CreaturePlacer (Variables|Default|GetBaseActor)))
  (EventDispatchers|UnbindallEventsfromOnCreativeAssetsLoaded (Actor|GetComponentbyClass _asbuilding_prop_creature_placer "/Script/FortniteGame.CreativeAssetLoadingTrackingComponent")))

(fn DestroyBlueprintComponents
  (for _array_element (Variables|Default|GetBlueprintMeshComponents)
    (Components|DestroyComponent _array_element)))

(fn DetatchActorFromWallProp
  ;[0.8321 0.6907 1] Not sure why, but the build preview marker runs off into the distance when you place a wall (and the blueprint mesh goes with it). Rather than having a coder dig in to try and resolve this, we just detach and keep the world transform
  (Utilities|Casting|CastToMoveBuildingsOnGrid (Variables|Default|GetCreativeInteraction))
  (Transformation|DetachFromActor "KeepWorld" "KeepWorld" "KeepWorld")
  (Transformation|SetWorldTransform (Variables|Default|GetDefaultSceneRoot) (Variables|Default|GetDropEndTransform) false false))

(fn DoNotRestoreOriginalMaterialsWhenDestroy
  (Networking|SwitchHasAuthority
    (:Remote
      (Variables|Default|SetWantsRestoreOriginalMaterialsWhenDestroy false))))

(fn FindComponentOriginWS
  (params (SceneComponent :type Scene%20Component%20Object%20Reference))
  (bind (_origin _boxextent) (Collision|GetComponentBounds SceneComponent))
  (bind (_boxextent_x _boxextent_y _boxextent_z) (Math|Vector|BreakVector _boxextent))
  (return (Math|Transform|InverseTransformLocation (Transformation|GetWorldTransform SceneComponent) (- _origin (Math|Vector|MakeVector 0.0 0.0 _boxextent_z)))))

(fn FindDropStartLocation
  (return false 0.0))

(fn GatherReplacementMaterials
  (Utilities|Array|Clear (Variables|Default|GetReplacementMaterialsFromActor))
  (for _array_element (Actor|GetComponentsByClass (Variables|Default|GetBaseActor) "/Script/GeometryCollectionEngine.GeometryCollectionComponent")
    (for _array_element_2 (Utilities|Struct|BreakGeometryCollectionProxyMeshData (Variables|/Script/GeometryCollectionEngine.GeometryCollection|GetRootProxyData (Variables|/Script/GeometryCollectionEngine.GeometryCollectionComponent|GetRestCollection _array_element)))
      (for _array_element_3 (Variables|/Script/Engine.StaticMesh|GetStaticMaterials _array_element_2)
        (bind _returnvalue (AssetUserData|GetAssetUserDataOfClass (Utilities|Struct|BreakStaticMaterial _array_element_3) "/Script/FortniteGame.FortCreativeMaterialAssetData"))
        (bind _asfort_creative_material_asset_data (Utilities|Casting|CastToFortCreativeMaterialAssetData _returnvalue))
        (if (Utilities|IsValidSoftObjectReference (Variables|/Script/FortniteGame.FortCreativeMaterialAssetData|GetMoverToolMaterial _asfort_creative_material_asset_data))
          (Utilities|Array|AddUnique (LocalVariables|GetFoundSoftObjs) (Variables|/Script/FortniteGame.FortCreativeMaterialAssetData|GetMoverToolMaterial _asfort_creative_material_asset_data))))))
  (return (LocalVariables|GetFoundSoftObjs)))

(fn GetActorCenter
  (bind _baseactor (Variables|Default|GetBaseActor)
    (:then
      (bind (_override _center) (CreativeMoveTool|GetCreativeActorOrigin(Message) _baseactor))
      (if _override
        (return _center)
        (else
          (return (Variables|Bounds|GetComponentTrueCenter)))))
    (:else
      (return "0, 0, 0"))))

(fn GetRotationGuideWorldRotation
  (return "0, 0, 0"))

(fn GetBlueprintBillboardComponent
  (bind (_meshcomponenttransform_location _meshcomponenttransform_rotation) (Math|Transform|BreakTransform (Variables|GetMeshComponentTransform_Location)))
  (bind (_currentbillboardcomponentdrawsize_x _currentbillboardcomponentdrawsize_y) (Math|Vector2D|BreakVector2D (Variables|GetCurrentBillboardComponentDrawSize_X)))
  (Variables|BlueprintPreview|GetCurrentWidgetObject
    (:then
      (bind _returnvalue (AddComponent|NODE_AddStaticMeshComponent-2 :RelativeTransform_Location _meshcomponenttransform_location :RelativeTransform_Rotation _meshcomponenttransform_rotation :RelativeTransform_Scale_X 1.0 :RelativeTransform_Scale_Y _currentbillboardcomponentdrawsize_x :RelativeTransform_Scale_Z _currentbillboardcomponentdrawsize_y)) ; Collision must be off!!
      (bind _output_get (Variables|Default|SetCurrentBlueprintStaticMeshComponent _returnvalue))
      (Utilities|Array|SetArrayElem (Variables|Default|GetBlueprintMeshComponents) (Variables|Default|GetCurrentIndex) _output_get true))
    (:else)))

(fn GetBlueprintGeometryCollectionComponentFromPool
  (params (IndexOfProxyToConvert :type Integer))
  (bind _currentgeometrycollectioncomponent (Variables|Default|GetCurrentGeometryCollectionComponent))
  (bind _restcollection (Variables|/Script/GeometryCollectionEngine.GeometryCollectionComponent|GetRestCollection _currentgeometrycollectioncomponent))
  (bind _rootproxydata (Variables|/Script/GeometryCollectionEngine.GeometryCollection|GetRootProxyData _restcollection))
  (bind _proxymeshes (Utilities|Struct|BreakGeometryCollectionProxyMeshData _rootproxydata))
  (bind _currentindex (Variables|Default|GetCurrentIndex))
  (bind _blueprintmeshcomponents (Variables|Default|GetBlueprintMeshComponents))
  (bind _output (Utilities|Array|Get(acopy) _blueprintmeshcomponents _currentindex))
  (Utilities|IsValid (Variables|Default|GetCurrentGeometryCollectionComponent)
    (:Is%20Valid
      (if (== (Utilities|Array|Length _proxymeshes) 0)
        (if (and (Utilities|IsArrayNonEmpty (Variables|Default|GetBlueprintMeshComponents)) (Utilities|Array|IsValidIndex (Variables|Default|GetBlueprintMeshComponents) (Variables|Default|GetCurrentIndex)))
          (Utilities|IsValid _output
            (:Is%20Valid
              (bind _asgeometry_collection_component (Utilities|Casting|CastToGeometryCollectionComponent _output)
                (:then
                  (Variables|Default|SetCurrentBlueprintGeometryCollectionComponent _asgeometry_collection_component))
                (:CastFailed
                  (bind _returnvalue (AddComponent|NODE_AddGeometryCollectionComponent-0 :!label _label))
                  (bind _output_get (Variables|Default|SetCurrentBlueprintGeometryCollectionComponent _returnvalue))
                  (Utilities|Array|SetArrayElem (Variables|Default|GetBlueprintMeshComponents) (Variables|Default|GetCurrentIndex) _output_get true)))
              (Transformation|SetRelativeTransform (Variables|Default|GetCurrentGeometryCollectionComponent) (Variables|Bounds|GetMeshComponentTransform) false false)
              (ChaosPhysics|SetRestCollection (Variables|Default|GetCurrentBlueprintGeometryCollectionComponent) (Variables|/Script/GeometryCollectionEngine.GeometryCollectionComponent|GetRestCollection (Variables|Default|GetCurrentGeometryCollectionComponent)) true)
              (Rendering|SetVisibility (Variables|Default|GetCurrentBlueprintGeometryCollectionComponent) true false)
              (Transformation|SetAbsolute (Variables|Default|GetCurrentBlueprintGeometryCollectionComponent) false (Variables|/Script/Engine.SceneComponent|GetbAbsoluteRotation (Variables|Default|GetCurrentGeometryCollectionComponent)) (Variables|/Script/Engine.SceneComponent|GetbAbsoluteScale (Variables|Default|GetCurrentGeometryCollectionComponent))))
            (:Is%20Not%20Valid
              _)))
        (goto _label)
        (else
          (for (_array_element _array_index) _proxymeshes
            (if (== _array_index (Variables|GetIndexOfProxyToConvert))
              (bind _returnvalue_2 (AddComponent|NODE_AddStaticMeshComponent-7)) ; Collision must be off, uses absolute scale
              (bind _output_get_2 (Variables|BlueprintPreview|SetCurrentStaticMeshComponent _returnvalue_2))
              (Utilities|Array|Insert (Variables|Default|GetStandInTempStaticMeshComponents) _returnvalue_2)
              (Components|StaticMesh|SetStaticMesh _output_get_2 _array_element)
              (Variables|BlueprintPreview|SetisStaticMeshComponent true)
              (Variables|BlueprintPreview|SetRenderComponent true)
              (CallFunction|GetBlueprintStaticMeshComponentfromPool))))))))

(fn GetBlueprintMaterialByPlatform
  (params (MaterialBeingOverwritten :type Material%20Interface%20Object%20Reference))
  (bind _returnvalue (AssetUserData|GetAssetUserDataOfClass MaterialBeingOverwritten "/Script/FortniteGame.FortCreativeMaterialAssetData"))
  (bind _asfort_creative_material_asset_data (Utilities|Casting|CastToFortCreativeMaterialAssetData _returnvalue)
    (:then
      (if (Variables|/Script/FortniteGame.FortCreativeMaterialAssetData|GetUseSelfForMoverToolMaterial _asfort_creative_material_asset_data)
        (return MaterialBeingOverwritten)
        (elif (Utilities|IsValidSoftObjectReference (Variables|/Script/FortniteGame.FortCreativeMaterialAssetData|GetMoverToolMaterial _asfort_creative_material_asset_data))
          (for _array_element (Variables|Default|GetReplacementMaterialsFromActor)
            (if (== (Utilities|ResolveSoftReference _array_element) (Variables|/Script/FortniteGame.FortCreativeMaterialAssetData|GetMoverToolMaterial _asfort_creative_material_asset_data))
              (return _array_element))))))
    (:CastFailed
      _))
  (return (select (Math|Boolean|ORBoolean (TouchPlatform|IsMobileGame) (TouchPlatform|IsNintendoSwitch) (== 0 (Rendering|Material|GetRenderingMaterialQualityLevel))) "/Game/Weapons/Prototype/ObjectMover/Materials/M_MoveTool_Controlled_LQ.M_MoveTool_Controlled_LQ" "/Game/Weapons/Prototype/ObjectMover/Materials/M_MoveTool_Controlled.M_MoveTool_Controlled")))

(fn GetBlueprintSkeletalMeshComponentFromPool
  (bind _currentindex (Variables|Default|GetCurrentIndex))
  (bind _blueprintmeshcomponents (Variables|Default|GetBlueprintMeshComponents))
  (bind _output (Utilities|Array|Get(acopy) _blueprintmeshcomponents _currentindex))
  (Variables|Default|GetCurrentSkeletalMeshObject
    (:then
      (if (Utilities|IsArrayNonEmpty (Variables|Default|GetBlueprintMeshComponents))
        (Utilities|IsValid _output
          (:Is%20Valid
            (bind _asskeletal_mesh_component (Utilities|Casting|CastToSkeletalMeshComponent _output)
              (:then
                (Variables|Default|SetCurrentBlueprintSkeletalMeshComonent _asskeletal_mesh_component))
              (:CastFailed
                (bind _returnvalue (AddComponent|NODE_AddSkeletalMeshComponent-0 :!label _label)) ; Collision should be off
                (bind _output_get (Variables|Default|SetCurrentBlueprintSkeletalMeshComonent _returnvalue))
                (Utilities|Array|SetArrayElem (Variables|Default|GetBlueprintMeshComponents) (Variables|Default|GetCurrentIndex) _output_get true)
                (Utilities|Array|AddUnique (Variables|Default|GetBlueprintSkeletalMeshComponents) _output_get)))
            (Components|SkinnedMesh|SetSkinnedAssetandUpdate (Variables|Default|GetCurrentBlueprintSkeletalMeshComonent) (Variables|Default|GetCurrentSkeletalMeshObject) false)
            (Rendering|SetVisibility (Variables|Default|GetCurrentBlueprintSkeletalMeshComonent) true false)
            (Transformation|SetRelativeTransform (Variables|Default|GetCurrentBlueprintSkeletalMeshComonent) (Variables|Bounds|GetMeshComponentTransform) false false)
            ;[1 1 1] Once anim data is ready, set animations (if valid)
            (Animation|BindEventtoOnAnimInitialized (Variables|Default|GetCurrentSkeletalMeshComponent) (EventDispatchers|CreateEvent :!event PlaySkeletalMeshAnimation)))
          (:Is%20Not%20Valid
            _)))
      (goto _label))
    (:else)))

(fn GetBlueprintStaticMeshComponentFromPool
  (bind _currentindex (Variables|Default|GetCurrentIndex))
  (bind _blueprintmeshcomponents (Variables|Default|GetBlueprintMeshComponents))
  (bind _output (Utilities|Array|Get(acopy) _blueprintmeshcomponents _currentindex))
  (bind _currentstaticmeshcomponent (Variables|BlueprintPreview|GetCurrentStaticMeshComponent))
  (Variables|BlueprintPreview|GetCurrentStaticMeshComponent
    (:then
      (if (and (Utilities|IsArrayNonEmpty (Variables|Default|GetBlueprintMeshComponents)) (Utilities|Array|IsValidIndex (Variables|Default|GetBlueprintMeshComponents) (Variables|Default|GetCurrentIndex)))
        (Utilities|IsValid _output
          (:Is%20Valid
            (bind _asstatic_mesh_component (Utilities|Casting|CastToStaticMeshComponent _output)
              (:then
                (Variables|Default|SetCurrentBlueprintStaticMeshComponent _asstatic_mesh_component))
              (:CastFailed
                (bind _asspline_mesh_component (Utilities|Casting|CastToSplineMeshComponent (Variables|BlueprintPreview|GetCurrentStaticMeshComponent) :!label _label)
                  (:then
                    (bind _returnvalue (CallFunction|HandleAddSplineMeshComponent _asspline_mesh_component))
                    (Variables|Default|SetCurrentBlueprintStaticMeshComponent _returnvalue))
                  (:CastFailed
                    (bind _asinstanced_static_mesh_component (Utilities|Casting|CastToInstancedStaticMeshComponent (Variables|BlueprintPreview|GetCurrentStaticMeshComponent))
                      (:then
                        (if (> (Components|InstancedStaticMesh|GetInstanceCount _asinstanced_static_mesh_component) 0)
                          (bind _returnvalue_2 (CallFunction|HandleAddInstancedMeshComponent _asinstanced_static_mesh_component (Components|InstancedStaticMesh|GetInstanceCount _asinstanced_static_mesh_component)))
                          (Variables|Default|SetCurrentBlueprintStaticMeshComponent _returnvalue_2)))
                      (:CastFailed
                        (bind _returnvalue_3 (AddComponent|NODE_AddStaticMeshComponent-0)) ; Collision must be off, uses absolute scale
                        (Variables|Default|SetCurrentBlueprintStaticMeshComponent _returnvalue_3)))))
                (Utilities|Array|SetArrayElem (Variables|Default|GetBlueprintMeshComponents) (Variables|Default|GetCurrentIndex) (Variables|Default|GetCurrentBlueprintStaticMeshComponent) true)))
            (Transformation|SetRelativeTransform (Variables|Default|GetCurrentBlueprintStaticMeshComponent) (Variables|Bounds|GetMeshComponentTransform) false false)
            (Components|StaticMesh|SetStaticMesh (Variables|Default|GetCurrentBlueprintStaticMeshComponent) (Variables|/Script/Engine.StaticMeshComponent|GetStaticMesh _currentstaticmeshcomponent))
            (if (Components|ComponentHasTag _currentstaticmeshcomponent "RotationLockedBlueprintPreviewer")
              (Utilities|Map|Add (Variables|Default|GetLockedStaticMeshComponents) (Variables|Default|GetCurrentBlueprintStaticMeshComponent) (Math|Transform|MakeTransform (/ (- (Transformation|GetWorldLocation (Variables|BlueprintPreview|GetCurrentStaticMeshComponent)) (Transformation|GetWorldLocation (Variables|Default|GetDefaultSceneRoot))) (Transformation|GetWorldScale (Variables|Default|GetDefaultSceneRoot))) (Transformation|GetWorldRotation (Variables|BlueprintPreview|GetCurrentStaticMeshComponent))))
              (Utilities|DoOnceThenDo false
                (:DoOnce
                  (CallFunction|AttemptSetTickEnabled true))
                (:ThenDo
                  _)))
            (Rendering|SetVisibility (Variables|Default|GetCurrentBlueprintStaticMeshComponent) true false)
            (Transformation|SetAbsolute (Variables|Default|GetCurrentBlueprintStaticMeshComponent) false (Variables|/Script/Engine.SceneComponent|GetbAbsoluteRotation (Variables|BlueprintPreview|GetCurrentStaticMeshComponent)) (Variables|/Script/Engine.SceneComponent|GetbAbsoluteScale (Variables|BlueprintPreview|GetCurrentStaticMeshComponent))))
          (:Is%20Not%20Valid
            _)))
      (goto _label))
    (:else)))

(fn GetCurrentBlueprintMeshComponent
  (return (Utilities|Array|Get(acopy) (Variables|Default|GetBlueprintMeshComponents) (Utilities|Array|LastIndex (Variables|Default|GetBlueprintMeshComponents)))))

(fn GetFirstActiveBlueprintPreviewForActor
  (params (ContextActor :type Actor%20Object%20Reference%20%28by%20ref%29) (ExcludeThisPreviewInstance :type Boolean :default true))
  (return 0))

(fn GetFXState
  (return (Variables|Default|GetFXState)))

(fn GetGeometryCollectionComponent
  (params (CurrentMeshComponent :type Mesh%20Component%20Object%20Reference) (IndexOfProxyToConvert :type Integer))
  (Utilities|IsValid CurrentMeshComponent
    (:Is%20Valid
      (bind _output_get (Variables|BlueprintPreview|SetCurrentMeshComponent CurrentMeshComponent))
      (bind _asgeometry_collection_component (Utilities|Casting|CastToGeometryCollectionComponent _output_get)
        (:then
          (if (CallFunction|ShouldMeshComponentbeRendered _output_get)
            (if (not (Components|ComponentHasTag _asgeometry_collection_component "IgnoredByBlueprintPreviewer"))
              (Variables|Default|SetCurrentGeometryCollectionComponent _asgeometry_collection_component)
              (CallFunction|SetBoundsVariables _asgeometry_collection_component)
              (Variables|BlueprintPreview|SetRenderComponent true)
              (CallFunction|GetBlueprintGeometryCollectionComponentfromPool (Variables|GetIndexOfProxyToConvert))
              (return true))))
        (:CastFailed
          _))
      (Variables|BlueprintPreview|SetRenderComponent false)
      (return false))
    (:Is%20Not%20Valid
      (return false))))

(fn GetGreatestMIDArrayIndex
  (return (Math|Integer|Max(Integer) (Utilities|Array|LastIndex (Variables|Default|GetMeshComponents)) (Utilities|Array|LastIndex (Variables|BlueprintPreview|GetBlueprintMIDs)))))

(fn GetGridPreviewMID
  (bind _asmaterial_instance_dynamic (Utilities|Casting|CastToMaterialInstanceDynamic (Rendering|Material|GetMaterial (Variables|Default|GetGridAxisGuide) 0))
    (:then
      (Variables|Default|SetGridPreviewMID _asmaterial_instance_dynamic))
    (:CastFailed
      (bind _returnvalue (Rendering|Material|CreateDynamicMaterialInstance (Variables|Default|GetGridAxisGuide) 0 0 "None" :!class /Script/Engine.PrimitiveComponent))
      (Variables|Default|SetGridPreviewMID _returnvalue))))

(fn GetMaterialsAndCopyParameters
  (bind _currentmeshcomponent (Variables|BlueprintPreview|GetCurrentMeshComponent))
  (bind _currentmeshcomponent_2 (Variables|BlueprintPreview|GetCurrentMeshComponent))
  (bind _currentblueprintgeometrycollectioncomponent (Variables|Default|GetCurrentBlueprintGeometryCollectionComponent))
  (if (Components|ComponentHasTag _currentmeshcomponent "MaterialChangeIgnoredByBlueprintPreviewer")
    _
    (elif (Variables|BlueprintPreview|GetRenderComponent)
      (for _index (range (Rendering|Material|GetNumMaterials _currentmeshcomponent))
        (if (not (Components|ComponentHasTag (Variables|BlueprintPreview|GetCurrentMeshComponent) "RenderedByBlueprintPreviewer"))
          (bind _returnvalue (Rendering|Material|GetBaseMaterial (Rendering|Material|GetMaterial _currentmeshcomponent_2 _index)))
          (if (not (ScalableFloat|GetValueasBool (Variables|Default|GetAllowTransparentMaterials)))
            (switch Utilities|FlowControl|Switch|SwitchonEBlendMode (Variables|/Script/Engine.Material|GetBlendMode _returnvalue)
              (:BLEND_Opaque
                _)
              (:BLEND_Masked
                _)
              (:BLEND_Translucent
                ;[1 1 1] In rare instances, we want to render non-opaque materials
                (if (not (Utilities|Array|ContainsItem (Variables|BlueprintPreview|GetNonOpaqueMaterialsToRender) (Utilities|ResolveSoftReference _returnvalue))) :!label _label
                  (Rendering|Material|SetMaterial (Utilities|Array|Get(acopy) (Variables|Default|GetBlueprintMeshComponents) (Variables|Default|GetCurrentIndex)) _index "/Game/Weapons/Prototype/ObjectMover/Materials/M_MoveTool_Invisible.M_MoveTool_Invisible")))
              (:BLEND_Additive
                (goto _label))
              (:BLEND_Modulate
                (goto _label))
              (:BLEND_AlphaComposite
                (goto _label))
              (:BLEND_AlphaHoldout
                (goto _label)))))
        (CallFunction|CreateMeshComponentMIDs true (Variables|BlueprintPreview|GetCurrentMeshComponent) _index (Rendering|Material|GetMaterial _currentmeshcomponent_2 _index))
        (bind _returnvalue_2 (Rendering|Material|CreateDynamicMaterialInstance (CallFunction|GetCurrentBlueprintMeshComponent) _index (CallFunction|GetBlueprintMaterialbyPlatform (Rendering|Material|GetMaterial _currentmeshcomponent_2 _index)) "None" :!class /Script/Engine.PrimitiveComponent))
        (bind _output_get (Variables|BlueprintPreview|SetCurrentBlueprintMID _returnvalue_2))
        (Utilities|Array|Add (Variables|BlueprintPreview|GetBlueprintMIDs) _output_get)
        (if (not (CallFunction|IgnoreCopyingMaterialParameters))
          (CallFunction|CopyMeshComponentMaterialParameterstoBlueprint (Variables|BlueprintPreview|GetCurrentMeshComponent) _index (Variables|BlueprintPreview|GetCurrentBlueprintMID))))
      (else
        ;[1 1 1] If we fail, lets make sure the mesh components get cleared out so we don't see bad meshes
        (bind _currentblueprintstaticmeshcomponent (Variables|Default|GetCurrentBlueprintStaticMeshComponent)
          (:then
            (Components|StaticMesh|SetStaticMesh _currentblueprintstaticmeshcomponent 0))
          (:else
            (bind _currentblueprintskeletalmeshcomonent (Variables|Default|GetCurrentBlueprintSkeletalMeshComonent)
              (:then
                (Components|SkinnedMesh|SetSkinnedAssetandUpdate _currentblueprintskeletalmeshcomonent 0 false))
              (:else
                (Utilities|IsValid _currentblueprintgeometrycollectioncomponent
                  (:Is%20Valid
                    (ChaosPhysics|SetRestCollection _currentblueprintgeometrycollectioncomponent 0 true)))))))))))

(fn GetMeshComponentFromArray
  (params (Index :type Integer) (StaticMeshArray :type Array%20of%20Mesh%20Component%20Object%20References))
  (bind _output (Utilities|Array|Get(acopy) StaticMeshArray Index))
  (if (Utilities|Array|IsValidIndex StaticMeshArray Index)
    (Utilities|IsValid _output
      (:Is%20Valid
        (return _output true))
      (:Is%20Not%20Valid
        _)))
  (return 0 false))

(fn GetMidFromArray
  (params (Index :type Integer) (MIDArray :type Array%20of%20Material%20Instance%20Dynamic%20Object%20References))
  (bind _output (Utilities|Array|Get(acopy) MIDArray Index))
  (if (Utilities|Array|IsValidIndex MIDArray Index)
    (Utilities|IsValid _output
      (:Is%20Valid
        (return _output true))
      (:Is%20Not%20Valid
        _)))
  (return 0 false))

(fn GetRotationGuideColor
  (return 0))

(fn GetScaleGuideColor
  (return 0))

(fn GetSceneComponentRelativeTransform
  (params (InputPin :type Scene%20Component%20Object%20Reference))
  (return (Math|Transform|MakeTransform (Math|Transform|InverseTransformLocation (Transformation|GetActorTransform (Components|GetOwner InputPin)) (Transformation|GetWorldLocation InputPin)) (Math|Transform|InverseTransformRotation (Transformation|GetActorTransform (Components|GetOwner InputPin)) (Transformation|GetWorldRotation InputPin)) (select (not (Variables|/Script/Engine.SceneComponent|GetbAbsoluteScale InputPin)) (/ (Transformation|GetWorldScale InputPin) (Transformation|GetActorScale3D (Components|GetOwner InputPin))) (Transformation|GetWorldScale InputPin)))))

(fn GetSetTextureParameter
  (params (SourcePrimitiveComponent :type Primitive%20Component%20Object%20Reference) (Material%20ID :type Integer) (Get%20Parameter%20Name :type Name) (Set%20Parameter%20Name :type Name) (TargetMID :type Material%20Instance%20Dynamic%20Object%20Reference))
  (bind _returnvalue (Rendering|Material|GetMaterial SourcePrimitiveComponent Material%20ID))
  (bind _asmaterial_instance_dynamic (Utilities|Casting|CastToMaterialInstanceDynamic _returnvalue)
    (:then
      (bind _output_get (LocalVariables|SetLocalMID _asmaterial_instance_dynamic))
      (bind _returnvalue_2 (Rendering|Material|GetTextureParameterValue _output_get (Variables|GetGet%20Parameter%20Name) :!class /Script/Engine.MaterialInstanceDynamic))
      (bind _output_get_2 (LocalVariables|SetTextureObject _returnvalue_2))
      (Utilities|IsValid _output_get_2
        (:Is%20Valid
          (Rendering|Material|SetTextureParameterValue (Variables|GetTargetMID) (Variables|GetSet%20Parameter%20Name) (LocalVariables|GetTextureObject) :!label _label)
          (return true))
        (:Is%20Not%20Valid
          _)))
    (:CastFailed
      (bind _asmaterial_instance_constant (Utilities|Casting|CastToMaterialInstanceConstant _returnvalue)
        (:then
          (bind _output_get_3 (LocalVariables|SetLocalMIC _asmaterial_instance_constant))
          (bind _returnvalue_3 (Rendering|Material|GetTextureParameterValue _output_get_3 (Variables|GetGet%20Parameter%20Name) :!class /Script/Engine.MaterialInstanceConstant))
          (bind _output_get_4 (LocalVariables|SetTextureObject _returnvalue_3))
          (Utilities|IsValid _output_get_4
            (:Is%20Valid
              (goto _label))
            (:Is%20Not%20Valid
              _)))
        (:CastFailed
          _))))
  (return false))

(fn GetSkeletalMeshComponent
  (params (CurrentMeshComponent :type Mesh%20Component%20Object%20Reference))
  (bind _output_get (Variables|BlueprintPreview|SetCurrentMeshComponent CurrentMeshComponent))
  (if (and (.bSuccess (Utilities|Casting|CastToSkeletalMeshComponent _output_get)) (CallFunction|ShouldMeshComponentbeRendered _output_get))
    (Utilities|IsValid (Variables|/Script/Engine.SkinnedMeshComponent|GetSkeletalMesh (Utilities|Casting|CastToSkeletalMeshComponent _output_get))
      (:Is%20Valid
        (bind _output_get_2 (Variables|Default|SetCurrentSkeletalMeshComponent (Utilities|Casting|CastToSkeletalMeshComponent _output_get)))
        (Utilities|Array|AddUnique (Variables|Default|GetCurrentSkeletalMeshComponents) _output_get_2)
        (bind _output_get_3 (Variables|Default|SetCurrentSkeletalMeshObject (Variables|/Script/Engine.SkinnedMeshComponent|GetSkeletalMesh (Utilities|Casting|CastToSkeletalMeshComponent _output_get))))
        (Utilities|Array|AddUnique (Variables|Default|GetSkeletalMeshObjects) _output_get_3)
        (CallFunction|SetBoundsVariables (Utilities|Casting|CastToSkeletalMeshComponent _output_get))
        (Variables|BlueprintPreview|SetisSkelatalMeshComponent true)
        (Variables|BlueprintPreview|SetRenderComponent true)
        (CallFunction|GetBlueprintSkeletalMeshComponentfromPool)
        (return true))
      (:Is%20Not%20Valid
        _)))
  (Variables|BlueprintPreview|SetRenderComponent false)
  (return false))

(fn GetStaticMeshComponent
  (params (CurrentMeshComponent :type Mesh%20Component%20Object%20Reference))
  (Utilities|IsValid CurrentMeshComponent
    (:Is%20Valid
      (bind _output_get (Variables|BlueprintPreview|SetCurrentMeshComponent CurrentMeshComponent))
      (if (and (.bSuccess (Utilities|Casting|CastToStaticMeshComponent _output_get)) (CallFunction|ShouldMeshComponentbeRendered _output_get))
        (Utilities|IsValid (Variables|/Script/Engine.StaticMeshComponent|GetStaticMesh (Utilities|Casting|CastToStaticMeshComponent _output_get))
          (:Is%20Valid
            (if (not (Components|ComponentHasTag (Utilities|Casting|CastToStaticMeshComponent _output_get) "IgnoredByBlueprintPreviewer"))
              (bind _output_get_2 (Variables|BlueprintPreview|SetCurrentStaticMeshComponent (Utilities|Casting|CastToStaticMeshComponent _output_get)))
              (Utilities|Array|Add (Variables|Default|GetStaticMeshComponents) _output_get_2)
              (CallFunction|SetBoundsVariables (Utilities|Casting|CastToStaticMeshComponent _output_get) (Utilities|Casting|CastToStaticMeshComponent _output_get))
              (Variables|BlueprintPreview|SetisStaticMeshComponent true)
              (Variables|BlueprintPreview|SetRenderComponent true)
              (CallFunction|GetBlueprintStaticMeshComponentfromPool)
              (return true false)))
          (:Is%20Not%20Valid
            ;[0.6798 1 0.5604] CreativeEditOnlyMeshes may still be loading their mesh/material assets, so allow for a retry once that is complete
            (bind _ascreative_edit_only_mesh_component (Utilities|Casting|CastToCreativeEditOnlyMeshComponent (Utilities|Casting|CastToStaticMeshComponent _output_get))
              (:then
                (if (and (Class|CreativeEditOnlyMeshComponent|IsLoadingAssets _ascreative_edit_only_mesh_component) (Class|CreativeEditOnlyMeshComponent|ShouldShowinPreview _ascreative_edit_only_mesh_component))
                  (EventDispatchers|BindEventtoOnMeshandMaterialsLoaded _ascreative_edit_only_mesh_component (EventDispatchers|CreateEvent :!event OnEditOnlyMeshAssetsLoaded)) ; Retry when assets are loaded
                  (Variables|BlueprintPreview|SetRenderComponent false)
                  (return false true)
                  (else
                    _)))
              (:CastFailed
                _)))))
      (Variables|BlueprintPreview|SetRenderComponent false)
      (return false false))
    (:Is%20Not%20Valid
      (return false false))))

(fn GetTransitionTime
  (return (/ (Components|Timeline|GetTimelineLength (Variables|B_CreativeMoveTool_BlueprintPreviewer|GetTransitionTL)) (Components|Timeline|GetPlayRate (Variables|B_CreativeMoveTool_BlueprintPreviewer|GetTransitionTL)))))

(fn GetWidgetComponent
  (params (CurrentMeshComponent :type Mesh%20Component%20Object%20Reference))
  (bind (_aswidget_component _bsuccess) (Utilities|Casting|CastToWidgetComponent CurrentMeshComponent))
  (Utilities|IsValid CurrentMeshComponent
    (:Is%20Valid
      (Variables|BlueprintPreview|SetCurrentMeshComponent CurrentMeshComponent)
      (if (and _bsuccess (CallFunction|ShouldMeshComponentbeRendered CurrentMeshComponent))
        (if (not (Components|ComponentHasTag _aswidget_component "IgnoredByBlueprintPreviewer"))
          (Variables|BlueprintPreview|SetCurrentBillboardComponentDrawSize (UserInterface|GetDrawSize _aswidget_component))
          (Variables|BlueprintPreview|SetCurrentWidgetObject _aswidget_component)
          (CallFunction|SetBoundsVariables _aswidget_component)
          (Variables|BlueprintPreview|SetisBillboardComponent true)
          (Variables|BlueprintPreview|SetRenderComponent true)
          (CallFunction|GetBlueprintBillboardComponent)
          (return true))))
    (:Is%20Not%20Valid
      _))
  (Variables|BlueprintPreview|SetRenderComponent false)
  (return false))

(fn GravityGuideSetup)

(fn GridAxisGuideSetup)

(fn HandleAddInstancedMeshComponent
  (params (OriginalInstancedStaticMeshComp :type Instanced%20Static%20Mesh%20Component%20Object%20Reference) (Instances :type Integer))
  (for _index (range Instances)
    (Utilities|Array|Add (LocalVariables|GetInstanceTransforms) (Components|InstancedStaticMesh|GetInstanceTransform (Variables|GetOriginalInstancedStaticMeshComp) _index false)))
  (bind _returnvalue (AddComponent|NODE_AddInstancedStaticMeshComponent-1))
  (bind _output_get (LocalVariables|SetNewInstancedStaticMeshComp _returnvalue))
  (Components|InstancedStaticMesh|AddInstances _output_get (LocalVariables|GetInstanceTransforms) false false true)
  (return (LocalVariables|GetNewInstancedStaticMeshComp)))

(fn HandleAddSplineMeshComponent
  (params (OriginalSplineMeshComp :type Spline%20Mesh%20Component%20Object%20Reference))
  (bind _originalsplinemeshcomp (Variables|GetOriginalSplineMeshComp))
  (bind _returnvalue (AddComponent|NODE_AddSplineMeshComponent-0)) ; Collision must be off, uses absolute scale
  (SplineMesh|SetBoundaryMax _returnvalue (SplineMesh|GetBoundaryMax _originalsplinemeshcomp) false)
  (SplineMesh|SetBoundaryMin _returnvalue (SplineMesh|GetBoundaryMin _originalsplinemeshcomp) false)
  (SplineMesh|SetForwardAxis _returnvalue (SplineMesh|GetForwardAxis _originalsplinemeshcomp) false)
  (SplineMesh|SetSplineUpDir _returnvalue (SplineMesh|GetSplineUpDir _originalsplinemeshcomp) false)
  (SplineMesh|SetStartandEnd _returnvalue (SplineMesh|GetStartPosition _originalsplinemeshcomp) (SplineMesh|GetStartTangent _originalsplinemeshcomp) (SplineMesh|GetEndPosition _originalsplinemeshcomp) (SplineMesh|GetEndTangent _originalsplinemeshcomp) false :!class /Script/Engine.SplineMeshComponent)
  (SplineMesh|SetStartOffset _returnvalue (SplineMesh|GetStartOffset _originalsplinemeshcomp) false)
  (SplineMesh|SetStartRollRadians _returnvalue (SplineMesh|GetStartRoll _originalsplinemeshcomp) false)
  (SplineMesh|SetStartScale _returnvalue (SplineMesh|GetStartScale _originalsplinemeshcomp) false :!class /Script/Engine.SplineMeshComponent)
  (SplineMesh|SetEndOffset _returnvalue (SplineMesh|GetEndOffset _originalsplinemeshcomp) false)
  (SplineMesh|SetEndRollRadians _returnvalue (SplineMesh|GetEndRoll _originalsplinemeshcomp) false)
  (SplineMesh|SetEndScale _returnvalue (SplineMesh|GetEndScale _originalsplinemeshcomp) false :!class /Script/Engine.SplineMeshComponent)
  (SplineMesh|UpdateMesh _returnvalue)
  (return _returnvalue))

(fn HandleDeleteAnimationFinished
  (CallFunction|CheckForDestructionAnimationWhenTimelineCompleted))

(fn IgnoreCopyingMaterialParameters
  (bind _currentstaticmeshcomponent (Variables|BlueprintPreview|GetCurrentStaticMeshComponent))
  (Utilities|IsValid _currentstaticmeshcomponent
    (:Is%20Valid
      (return (Utilities|Array|ContainsItem (Variables|Default|GetStaticMeshComponentsToOmitFromCopyingMaterialParameters) (Utilities|ResolveSoftReference (Variables|/Script/Engine.StaticMeshComponent|GetStaticMesh _currentstaticmeshcomponent)))))
    (:Is%20Not%20Valid
      (return false))))

(fn InitializeIntermediateMaterials
  (Utilities|Array|Clear (Variables|Default|GetIntermediateMIDs))
  (for _array_element (Variables|Default|GetMeshComponents)
    (Utilities|IsValid _array_element
      (:Is%20Valid
        (if (Components|ComponentHasTag _array_element "MaterialChangeIgnoredByBlueprintPreviewer")
          _
          (elif (Variables|/Script/Engine.ActorComponent|GetbIsEditorOnly _array_element)
            _
            (else
              (for _index (range (Rendering|Material|GetNumMaterials _array_element))
                (bind _returnvalue (Rendering|Material|GetBaseMaterial (Rendering|Material|GetMaterial _array_element _index)))
                (switch Utilities|FlowControl|Switch|SwitchonEBlendMode (Variables|/Script/Engine.Material|GetBlendMode _returnvalue)
                  (:BLEND_Opaque
                    _)
                  (:BLEND_Masked
                    _)
                  (:BLEND_Translucent
                    ;[1 1 1] In rare instances, we want to render non-opaque materials
                    (if (not (Utilities|Array|ContainsItem (Variables|BlueprintPreview|GetNonOpaqueMaterialsToRender) (Utilities|ResolveSoftReference _returnvalue))) :!label _label
                      (Rendering|Material|SetMaterial _array_element _index "/Game/Weapons/Prototype/ObjectMover/Materials/M_MoveTool_Invisible.M_MoveTool_Invisible")))
                  (:BLEND_Additive
                    (goto _label))
                  (:BLEND_Modulate
                    (goto _label))
                  (:BLEND_AlphaComposite
                    (goto _label))
                  (:BLEND_AlphaHoldout
                    (goto _label)))
                (bind _returnvalue_2 (Rendering|Material|CreateDynamicMaterialInstance _array_element _index "/Game/Weapons/Prototype/ObjectMover/Materials/M_CreativeMoveTool_Intermediate.M_CreativeMoveTool_Intermediate" "None" :!class /Script/Engine.PrimitiveComponent))
                (Utilities|Array|AddUnique (Variables|Default|GetIntermediateMIDs) _returnvalue_2)
                (CallFunction|CopyMeshComponentMaterialParameterstoBlueprint _array_element _index _returnvalue_2)))))))))

(fn IsBuildingPiece
  (bind _returnvalue (Actor|GetOwner))
  (bind _returnvalue_2 (Game|GetInstigatorController _returnvalue))
  (Utilities|FlowControl|Sequence
    (:then_0
      (Utilities|IsValid _returnvalue
        (:Is%20Valid
          (Utilities|IsValid _returnvalue_2
            (:Is%20Valid
              (bind _asfort_player_controller (Utilities|Casting|CastToFortPlayerController _returnvalue_2))
              (Utilities|IsValid (Building|GetBuildPreviewMarker _asfort_player_controller)
                (:Is%20Valid
                  (return (Building|GetBuildingMeshComponent (Building|GetBuildPreviewMarker _asfort_player_controller)) (Building|GetBuildPreviewMarker _asfort_player_controller) (Creative|IsBuildToolOverridden _asfort_player_controller)))))))))
    (:then_1
      (return 0 0 false))))

(fn IsTransitionTimelineAtBeginning
  (return (Math|Float|NearlyEqual(Float) (Components|Timeline|GetPlaybackPosition (Variables|B_CreativeMoveTool_BlueprintPreviewer|GetTransitionTL)) 0.0 0.0001)))

(fn OnEditOnlyMeshAssetsLoaded
  (bind _result (Math|Integer|DecrementInt (Variables|Default|GetNumMeshesLoadingOnTarget))
    (:%20%20
      (if (<= _result 0)
        (CallFunction|CreateBlueprintPreview)))))

(fn OnMoveToolScalingToggled
  (params (bScalingActive :type Boolean))
  (bind _ascreative_move_tool_target (Utilities|Casting|CastToCreativeMoveToolTarget (Variables|Default|GetBaseActor)))
  (Class|CreativeMoveToolTarget|OnMoveToolScalingToggled _ascreative_move_tool_target bScalingActive))

(fn OnRep_BaseActor
  (bind _asbuilding_prop_creature_placer (Utilities|Casting|CastToBuildingProp_CreaturePlacer (Variables|Default|GetBaseActor))
    (:then
      (if (not (Class|CreativeAssetLoadingTrackingComponent|AreAssetsLoaded (Actor|GetComponentbyClass _asbuilding_prop_creature_placer "/Script/FortniteGame.CreativeAssetLoadingTrackingComponent")))
        (EventDispatchers|BindEventtoOnCreativeAssetsLoaded (Actor|GetComponentbyClass _asbuilding_prop_creature_placer "/Script/FortniteGame.CreativeAssetLoadingTrackingComponent") (EventDispatchers|CreateEvent :!event CreatureVisualsLoaded))))
    (:CastFailed
      _))
  (CallFunction|SetupPreview))

(fn OnRep_FXState
  (CallFunction|UpdateFXState))

(fn OnRep_MarkedToDestroy)

(fn ParticlePropHelperSetup
  (bind _baseactor (Variables|Default|GetBaseActor)
    (:then
      (if (Utilities|Array|ContainsItem (Variables|Actor|GetTags _baseactor) "CreativeParticleProp")
        (Rendering|SetActorHiddenInGame (Variables|Default|GetBaseActor) false)))
    (:else)))

(fn PlaySkeletalMeshAnimation
  (bind _blueprintskeletalmeshcomponents (Variables|Default|GetBlueprintSkeletalMeshComponents))
  (for (_array_element _array_index) (Variables|Default|GetCurrentSkeletalMeshComponents)
    (Utilities|IsValid _array_element
      (:Is%20Valid
        (if (.bSavedPlaying (Utilities|Struct|BreakSingleAnimationPlayData (Variables|/Script/Engine.SkeletalMeshComponent|GetAnimationData _array_element)))
          (Utilities|IsValid (Utilities|Struct|BreakSingleAnimationPlayData (Variables|/Script/Engine.SkeletalMeshComponent|GetAnimationData _array_element))
            (:Is%20Valid
              (Utilities|IsValid (Utilities|Array|Get(acopy) _blueprintskeletalmeshcomponents _array_index)
                (:Is%20Valid
                  (Components|Animation|PlayAnimation (Utilities|Array|Get(acopy) _blueprintskeletalmeshcomponents _array_index) (Utilities|Struct|BreakSingleAnimationPlayData (Variables|/Script/Engine.SkeletalMeshComponent|GetAnimationData _array_element)) (.bSavedLooping (Utilities|Struct|BreakSingleAnimationPlayData (Variables|/Script/Engine.SkeletalMeshComponent|GetAnimationData _array_element))))
                  (Components|Animation|SetPlayRate (Utilities|Array|Get(acopy) _blueprintskeletalmeshcomponents _array_index) (.SavedPlayRate (Utilities|Struct|BreakSingleAnimationPlayData (Variables|/Script/Engine.SkeletalMeshComponent|GetAnimationData _array_element))))
                  (Components|Animation|SetPosition (Utilities|Array|Get(acopy) _blueprintskeletalmeshcomponents _array_index) (.SavedPosition (Utilities|Struct|BreakSingleAnimationPlayData (Variables|/Script/Engine.SkeletalMeshComponent|GetAnimationData _array_element))) true))))))))))

(fn PlayTransitionBackwardWhenCurrentAnimationEnds
  (Networking|SwitchHasAuthority
    (:Authority
      (Variables|Default|SetPlayTransitionBackwardWhenAnimationEnds true))))

(fn QueryMIDBlendMode
  (bind _currentindex (Variables|Default|GetCurrentIndex))
  (bind _currentmeshcomponent (Variables|BlueprintPreview|GetCurrentMeshComponent))
  (bind _returnvalue (Rendering|Material|GetMaterial _currentmeshcomponent _currentindex))
  (bind (_asmaterial_instance_dynamic _bsuccess) (Utilities|Casting|CastToMaterialInstanceDynamic _returnvalue))
  (if _bsuccess
    (bind _returnvalue_2 (Rendering|Material|GetBaseMaterial _asmaterial_instance_dynamic))
    (switch Utilities|FlowControl|Switch|SwitchonEBlendMode (Variables|/Script/Engine.Material|GetBlendMode _returnvalue_2)
      (:BLEND_Opaque
        _)
      (:BLEND_Masked
        _)
      (:BLEND_Translucent
        (return false))
      (:BLEND_Additive
        (return false))
      (:BLEND_Modulate
        (return false))
      (:BLEND_AlphaComposite
        (return false)))
    (return true)))

(fn RestoreOriginalMeshComponentMaterials
  (bind _values (Utilities|Map|Values (Variables|Default|GetOriginalMeshComponentMaterials)))
  (if (Utilities|IsArrayNonEmpty _values)
    (bind _keys (Utilities|Map|Keys (Variables|Default|GetOriginalMeshComponentMaterials)))
    (for (_array_element _array_index) _keys
      (Utilities|IsValid _array_element
        (:Is%20Valid
          (Rendering|Material|SetMaterial (Utilities|Array|Get(acopy) _values _array_index) (Utilities|Array|Get(acopy) (Variables|Default|GetOriginalMeshComponentMaterialIDs) _array_index) _array_element))))))

(fn RotationGuideInitialSetup
  (Variables|Default|SetInitialRotationInverse (Math|Rotator|InvertRotator (Transformation|GetWorldRotation (Variables|Default|GetDefaultSceneRoot))))
  (Transformation|SetWorldRotation (Variables|Default|GetRotationGuide) (CallFunction|GetRotationGuideWorldRotation) false false)
  (Rendering|Material|SetVectorParameterValue (Variables|RotationGuide|GetRotationGuideMID) "AxisColor" (CallFunction|GetRotationGuideColor) :!class /Script/Engine.MaterialInstanceDynamic))

(fn ScaleGuideInitialSetup
  (Transformation|SetRelativeLocation (Variables|Default|GetScaleGuide) (CallFunction|GetActorCenter) false false)
  (Rendering|Material|SetVectorParameterValue (Variables|ScaleGuide|GetScaleGuideMID) "ActiveAxis" (CallFunction|GetScaleGuideColor) :!class /Script/Engine.MaterialInstanceDynamic))

(fn SetAbsoluteTransformsOnComponents
  (Variables|/Script/Engine.SceneComponent|SetbAbsoluteRotation true (Variables|Default|GetGridAxisGuide))
  (Variables|/Script/Engine.SceneComponent|SetbAbsoluteRotation true (Variables|Default|GetGravityGuide))
  (Variables|/Script/Engine.SceneComponent|SetbAbsoluteRotation true (Variables|Default|GetPivotHelper))
  (Variables|/Script/Engine.SceneComponent|SetbAbsoluteScale true (Variables|Default|GetRotationGuide))
  (Variables|/Script/Engine.SceneComponent|SetbAbsoluteScale true (Variables|Default|GetScaleGuide))
  (Variables|/Script/Engine.SceneComponent|SetbAbsoluteScale true (Variables|Default|GetPivotHelper))
  (Variables|/Script/Engine.SceneComponent|SetbAbsoluteScale true (Variables|Default|GetGridAxisGuide)))

(fn SetBlueprintMeshComponentsVisibility
  (params (bNewVisibility :type Boolean) (bPropagateToChildren :type Boolean))
  (for _array_element (Variables|Default|GetBlueprintMeshComponents)
    (Rendering|SetVisibility _array_element bNewVisibility bPropagateToChildren)))

(fn SetBoundsVariables
  (params (Scene%20Component :type Scene%20Component%20Object%20Reference) (Static%20Mesh%20Component :type Static%20Mesh%20Component%20Object%20Reference))
  (bind _scene_component (Variables|GetScene%20Component))
  (bind _relativescale3d (Variables|/Script/Engine.SceneComponent|GetRelativeScale3D _scene_component))
  (bind _z (.z _relativescale3d))
  (bind _returnvalue (Math|Float|Max(Float) _z 1.0))
  (bind _y (.y _relativescale3d))
  (bind _returnvalue_2 (Math|Float|Max(Float) _y 1.0))
  (bind _x (.x _relativescale3d))
  (bind _returnvalue_3 (Math|Float|Max(Float) _x 1.0))
  (bind _returnvalue_4 (Math|Vector|MakeVector _returnvalue_3 _returnvalue_2 _returnvalue))
  (bind _returnvalue_5 (/ "1.000000,1.000000,1.000000" _returnvalue_4))
  (bind _scene_component_2 (Variables|GetScene%20Component))
  (bind (_origin _boxextent _sphereradius) (Collision|GetComponentBounds _scene_component_2))
  (bind (_returnvalue_x _returnvalue_y _returnvalue_z) (Math|Vector|BreakVector (* _boxextent _returnvalue_5)))
  (Variables|Bounds|SetMaxComponentRadius (Math|Float|Max(Float) (Variables|Bounds|GetMaxComponentRadius) _sphereradius))
  (Variables|Bounds|SetMaxLocalBoundsZ (Math|Float|Max(Float) (* _returnvalue_z 2.0) (Variables|Bounds|GetMaxLocalBoundsZ)))
  (Variables|Bounds|SetMaxLocalBoundsXY (Math|Float|Max(Float) (Math|Float|Max(Float) _returnvalue_x _returnvalue_y) (Variables|Bounds|GetMaxLocalBoundsXY)))
  (Variables|Bounds|SetComponentOrigin (CallFunction|FindComponentOriginWS (Variables|GetScene%20Component)))
  (Variables|Bounds|SetMeshComponentTransform (CallFunction|GetSceneComponentRelativeTransform (Variables|GetScene%20Component)))
  (Utilities|IsValid (Variables|GetStatic%20Mesh%20Component)
    (:Is%20Valid
      (CallFunction|StaticMeshesComponentBoundsMinMax (Variables|GetStatic%20Mesh%20Component)))
    (:Is%20Not%20Valid
      (CallFunction|StaticMeshesComponentBoundsMinMax (Variables|GetScene%20Component))))
  (Variables|Bounds|SetComponentTrueCenter (/ (+ (Variables|Default|GetComponentBoundsMin) (Variables|Default|GetComponentBoundsMax)) 2)))

(fn SetCreativeInteractionTool
  (params (InCreativeInteractionTool :type Fort%20Creative%20Move%20Tool%20Object%20Reference))
  (bind _outputdelegate (EventDispatchers|CreateEvent :!event OnMoveToolScalingToggled))
  (bind _creativeinteractiontool (Variables|Default|GetCreativeInteractionTool))
  (if (!= InCreativeInteractionTool (Variables|Default|GetCreativeInteractionTool))
    (Utilities|FlowControl|Sequence
      (:then_0
        (Utilities|IsValid _creativeinteractiontool
          (:Is%20Valid
            (EventDispatchers|UnbindEventfromOnMoveToolScalingToggled _creativeinteractiontool _outputdelegate))))
      (:then_1
        (bind _output_get (Variables|Default|SetCreativeInteractionTool InCreativeInteractionTool))
        (Utilities|IsValid _output_get
          (:Is%20Valid
            (EventDispatchers|BindEventtoOnMoveToolScalingToggled InCreativeInteractionTool _outputdelegate)))))))

(fn SetEndSafetyParameters
  (bind _baseactor (Variables|Default|GetBaseActor)
    (:then
      (Rendering|SetActorHiddenInGame _baseactor (Utilities|Enum|Equal(Enum) (CallFunction|GetFXState) "NewEnumerator3")))
    (:else
      _))
  (CallFunction|SetMeshComponentsVisibility true)
  (CallFunction|DestroyBlueprintComponents)
  (for _array_element (Variables|Default|GetMeshComponentMIDs)
    (Utilities|IsValid _array_element
      (:Is%20Valid
        (Rendering|Material|SetScalarParameterValue _array_element "BlendToIntermediate" 0.0 :!class /Script/Engine.MaterialInstanceDynamic)))))

(fn SetHiddenActorCosmetics
  (params (Hidden :type Boolean) (ParticleComponents :type Boolean) (LightComponents :type Boolean :default true))
  (Variables|Default|GetBaseActor
    (:then
      (Utilities|FlowControl|Sequence
        (:then_0
          ;[1 1 1] Particles
          (if (Variables|GetParticleComponents)
            (for _array_element (Actor|GetComponentsByClass (Variables|Default|GetBaseActor) "/Script/CascadeRuntime.ParticleSystemComponent")
              (Components|Activation|SetActive _array_element (select (Variables|GetHidden) false (Variables|/Script/Engine.ActorComponent|GetbAutoActivate _array_element)) false))))
        (:then_1
          ;[1 1 1] Lights
          (if (Variables|GetLightComponents)
            (bind _ascreative_move_tool_light_preview_override (Utilities|Casting|CastToCreativeMoveToolLightPreviewOverride (Variables|Default|GetBaseActor))
              (:then
                (bind _returnvalue (Class|CreativeMoveToolLightPreviewOverride|ShouldOverrideMoveToolLightPreviewVisibility _ascreative_move_tool_light_preview_override))
                (if _returnvalue
                  _
                  (else
                    _)))
              (:CastFailed
                _))
            (for _array_element_2 (Actor|GetComponentsByClass (Variables|Default|GetBaseActor) "/Script/Engine.LightComponent")
              (Development|SetHiddeninGame _array_element_2 (select (Variables|GetHidden) true (not (Variables|/Script/Engine.SceneComponent|GetbVisible _array_element_2))) false))))
        (:then_2
          ;[1 1 1] Piggy back on Light Component shut off in order to hide GC components. Temp for 29.01 since GC's do not hide properlly when calling SetActorHiddenInGame
          (if (Variables|GetLightComponents)
            (for _array_element_3 (Actor|GetComponentsByClass (Variables|Default|GetBaseActor) "/Script/GeometryCollectionEngine.GeometryCollectionComponent")
              (Development|SetHiddeninGame _array_element_3 (Variables|GetHidden) false))))))
    (:else)))

(fn SetHiddenInGameAllVisualHelpers
  (params (NewHidden :type Boolean))
  (Development|SetHiddeninGame (Variables|Default|GetGridAxisGuide) NewHidden false))

(fn SetMeshComponentsVisibility
  (params (Visibility :type Boolean) (bPropagateToChildren :type Boolean))
  (bind (_buildingmeshcomponent _buildpreviewmarkeractor _bool) (CallFunction|IsBuildingPiece))
  (for _array_element (Variables|Default|GetMeshComponents)
    (Rendering|SetVisibility (select _bool _buildingmeshcomponent _array_element) Visibility bPropagateToChildren)))

(fn SetTextureParameter
  (params (Set%20Parameter%20Name :type Name) (TargetMID :type Material%20Instance%20Dynamic%20Object%20Reference) (Texture :type Texture%20Object%20Reference))
  (Utilities|IsValid Texture
    (:Is%20Valid
      (Rendering|Material|SetTextureParameterValue (Variables|GetTargetMID) (Variables|GetSet%20Parameter%20Name) (Variables|GetTexture)))))

(fn SetTimelinePlayrateByObjectHeight
  (params (Timeline :type Timeline%20Component%20Object%20Reference) (SpeedScale :type Float%20%28double-precision%29 :default 1.0))
  (Components|Timeline|SetPlayRate Timeline (Math|Float|Clamp(Float) (Utilities|Reciprocal (* SpeedScale (/ (Variables|Bounds|GetMaxLocalBoundsZ) (Variables|Default|GetTransitionPlayrateHeightDivisor)))) (Variables|Default|GetMinPlaybackSpeed) (Variables|Default|GetMaxPlaybackSpeed))))

(fn SetTransitionLightVisibility
  (params (bNewVisibility :type Boolean) (bPropagateToChildren :type Boolean) (SetRadius :type Boolean))
  (Rendering|SetVisibility (Variables|Default|GetTransitionLight) bNewVisibility bPropagateToChildren)
  (if SetRadius
    (Rendering|Lighting|SetAttenuationRadius (Variables|Default|GetTransitionLight) (Utilities|ClampMaxFloat (* (Variables|Bounds|GetMaxComponentRadius) 3.0) 1.5e+03))))

(fn SetupPreview
  (bind _asfort_creative_move_tool (Utilities|Casting|CastToFortCreativeMoveTool (Actor|GetOwner)))
  (CallFunction|EventCreateBlueprintPreview _asfort_creative_move_tool (Variables|Default|GetPropOwningActor)))

(fn SetupTransitionParticleParameters
  (params (bNewActive :type Boolean) (bReset :type Boolean) (Reverse :type Boolean))
  (bind _maxlocalboundsxyfudgefactor (Variables|Default|GetMaxLocalBoundsXYFudgeFactor))
  (bind _maxlocalboundsxy (Variables|Bounds|GetMaxLocalBoundsXY))
  (bind _returnvalue (* _maxlocalboundsxy _maxlocalboundsxyfudgefactor))
  (Components|Activation|SetActive (Variables|Default|GetTransitionParticles) bNewActive bReset)
  (if bNewActive
    (Effects|Components|ParticleSystem|SetFloatParameter (Variables|Default|GetTransitionParticles) "Radius" _returnvalue)
    (Effects|Components|ParticleSystem|SetVectorParameter (Variables|Default|GetTransitionParticles) "FlareWidth" (Math|Vector|MakeVector (* _returnvalue 9.0) 1.0 1.0))
    (Effects|Components|ParticleSystem|SetFloatParameter (Variables|Default|GetTransitionParticles) "FlareLifetime" (select Reverse (Variables|Default|GetTransitionLengthBackward) (Variables|Default|GetTransitionLengthForward)))
    (Effects|Components|ParticleSystem|SetFloatParameter (Variables|Default|GetTransitionParticles) "SpawnRateScale" (/ _returnvalue (Variables|Default|GetSpawnRateScaleBase)))
    (Effects|Components|ParticleSystem|SetVectorParameter (Variables|Default|GetTransitionParticles) "Velocity" (select Reverse (* (Variables|Default|GetTransitionParticleVelocity) -1.0) (Variables|Default|GetTransitionParticleVelocity)))
    (Transformation|SetWorldScale3D (Variables|Default|GetTransitionParticles) (Math|Vector|ClampVectorSize (Math|Vector|VectorGetAbs (Transformation|GetWorldScale (Variables|Default|GetDefaultSceneRoot))) 0.0 3.0))))

(fn ShadowPlaneSetup
  (Transformation|SetRelativeLocation (Variables|Default|GetShadowPlane) (+ (Variables|Bounds|GetComponentOrigin) (Variables|Default|GetShadowPlaneLocationOffset)) false false)
  (Transformation|SetRelativeScale3D (Variables|Default|GetShadowPlane) (Math|Conversions|ToVector(Float) (* (Variables|Bounds|GetMaxLocalBoundsXY) (Variables|GravityGuide|GetShadowPlaneScaleFudgeFactor)))))

(fn ShouldMeshComponentBeRendered
  (params (MeshComponent :type Mesh%20Component%20Object%20Reference))
  (Utilities|IsValid MeshComponent
    (:Is%20Valid
      (return (not (Math|Boolean|ORBoolean (Variables|/Script/Engine.ActorComponent|GetbIsEditorOnly MeshComponent) (Variables|/Script/Engine.SceneComponent|GetbHiddenInGame MeshComponent) (not (Variables|/Script/Engine.SceneComponent|GetbVisible MeshComponent))))))
    (:Is%20Not%20Valid
      (return false))))

(fn StaticMeshesComponentBoundMax
  (params (StaticMesh :type Static%20Mesh%20Component%20Object%20Reference))
  (bind (_min _max) (Components|StaticMesh|GetLocalBounds StaticMesh))
  (bind (_meshcomponenttransform_location _meshcomponenttransform_rotation _meshcomponenttransform_scale) (Math|Transform|BreakTransform (Variables|GetMeshComponentTransform_Location)))
  (bind (_meshcomponenttransform_location_2 _meshcomponenttransform_rotation_2 _meshcomponenttransform_scale_2) (Math|Transform|BreakTransform (Variables|GetMeshComponentTransform_Location)))
  (return (Math|Vector|VectorComponentMax (* _meshcomponenttransform_scale_2 (+ (Math|Transform|TransformDirection (Math|Transform|MakeTransform "0, 0, 0" _meshcomponenttransform_rotation _meshcomponenttransform_scale) _max) (Math|Transform|TransformDirection (Math|Transform|MakeTransform _meshcomponenttransform_location _meshcomponenttransform_rotation _meshcomponenttransform_scale) _meshcomponenttransform_location))) (Variables|Default|GetComponentBoundsMax))))

(fn StaticMeshesComponentBoundsMin
  (params (StaticMesh :type Static%20Mesh%20Component%20Object%20Reference))
  (bind (_meshcomponenttransform_location _meshcomponenttransform_rotation _meshcomponenttransform_scale) (Math|Transform|BreakTransform (Variables|GetMeshComponentTransform_Location)))
  (bind (_meshcomponenttransform_location_2 _meshcomponenttransform_rotation_2 _meshcomponenttransform_scale_2) (Math|Transform|BreakTransform (Variables|GetMeshComponentTransform_Location)))
  (return (Math|Vector|VectorComponentMin (* _meshcomponenttransform_scale_2 (+ (Math|Transform|TransformDirection (Math|Transform|MakeTransform "0, 0, 0" _meshcomponenttransform_rotation _meshcomponenttransform_scale) (Components|StaticMesh|GetLocalBounds StaticMesh)) (Math|Transform|TransformDirection (Math|Transform|MakeTransform _meshcomponenttransform_location _meshcomponenttransform_rotation _meshcomponenttransform_scale) _meshcomponenttransform_location))) (Variables|Default|GetComponentBoundsMin))))

(fn StaticMeshesComponentBoundsMinMax
  (params (StaticMesh :type Scene%20Component%20Object%20Reference))
  (bind _returnvalue (Fortnite|Utility|GetLocalBounds StaticMesh))
  (bind (_meshcomponenttransform_location _meshcomponenttransform_rotation _meshcomponenttransform_scale) (Math|Transform|BreakTransform (Variables|GetMeshComponentTransform_Location)))
  (bind _returnvalue_2 (Math|Vector|RotateVector _meshcomponenttransform_location (Transformation|GetActorRotation (Components|GetOwner StaticMesh))))
  (bind _returnvalue_3 (Components|GetOwner StaticMesh))
  (bind _returnvalue_4 (Transformation|GetActorRotation _returnvalue_3))
  (bind (_meshcomponenttransform_location_2 _meshcomponenttransform_rotation_2 _meshcomponenttransform_scale_2) (Math|Transform|BreakTransform (Variables|GetMeshComponentTransform_Location)))
  (bind _returnvalue_5 (Math|Vector|RotateVector _meshcomponenttransform_location_2 _returnvalue_4))
  (bind (_origin _boxextent) (Math|BoxSphereBounds|BreakBoxSphereBounds _returnvalue))
  (bind _returnvalue_6 (- _origin _boxextent))
  (bind _returnvalue_7 (* _meshcomponenttransform_scale _returnvalue_6))
  (bind _returnvalue_8 (Math|Vector|RotateVector _returnvalue_7 _meshcomponenttransform_rotation))
  (bind _returnvalue_9 (+ _origin _boxextent))
  (bind _returnvalue_10 (* _meshcomponenttransform_scale_2 _returnvalue_9))
  (bind _returnvalue_11 (Math|Vector|RotateVector _returnvalue_10 _meshcomponenttransform_rotation_2))
  (bind _returnvalue_12 (+ _returnvalue_5 _returnvalue_11))
  (bind _returnvalue_13 (+ _returnvalue_2 _returnvalue_8))
  (if (not (Components|ComponentHasTag StaticMesh "BoundsIgnoredByBlueprintPreviewer"))
    (Variables|Default|SetComponentBoundsMax (Math|Vector|VectorComponentMax (Math|Vector|VectorComponentMax _returnvalue_13 (Variables|Default|GetComponentBoundsMax)) (Math|Vector|VectorComponentMax (Variables|Default|GetComponentBoundsMax) _returnvalue_12)))
    (Variables|Default|SetComponentBoundsMin (Math|Vector|VectorComponentMin (Math|Vector|VectorComponentMin (Variables|Default|GetComponentBoundsMin) _returnvalue_13) (Math|Vector|VectorComponentMin (Variables|Default|GetComponentBoundsMin) _returnvalue_12)))))

(fn StopTransitionAnimation
  (Networking|SwitchHasAuthority
    (:Remote
      (Components|Timeline|Stop (Variables|B_CreativeMoveTool_BlueprintPreviewer|GetTransitionTL)))))

(fn StoreOriginalMeshComponentMaterials
  (params (SaveOriginalMaterials :type Boolean) (MaterialIndex :type Integer) (Mesh%20Component :type Mesh%20Component%20Object%20Reference))
  (if SaveOriginalMaterials
    (bind _foundpreview (CallFunction|GetFirstActiveBlueprintPreviewforActor (Components|GetOwner Mesh%20Component)))
    (Utilities|IsValid _foundpreview
      (:Is%20Valid
        ;[1 1 1] If previous blueprint preview with same previewed actor is found: copy its materials, materialsID, stop its transition animation, active the non-materials restoration flag and ask the server to destroy it. Now this blueprint preview instance is in charge of restoring the original material of the previewed actor.
        (Utilities|FlowControl|Sequence
          (:then_0
            (Variables|Default|SetOriginalMeshComponentMaterials (Variables|Default|GetOriginalMeshComponentMaterials _foundpreview)))
          (:then_1
            (Variables|Default|SetOriginalMeshComponentMaterialIDs (Variables|Default|GetOriginalMeshComponentMaterialIDs _foundpreview)))
          (:then_2
            (CallFunction|StopTransitionAnimation _foundpreview)
            (CallFunction|DoNotRestoreOriginalMaterialswhenDestroy _foundpreview)
            (CallFunction|ClientMakeDestroyonServerRequest _foundpreview))))
      (:Is%20Not%20Valid
        (Utilities|Map|Add (Variables|Default|GetOriginalMeshComponentMaterials) (Rendering|Material|GetMaterial Mesh%20Component MaterialIndex) Mesh%20Component)
        (Utilities|Array|Add (Variables|Default|GetOriginalMeshComponentMaterialIDs) MaterialIndex)))))

(fn UpdateGuidePositions
  (bind _relativelocationofpivotpoint (Variables|Default|GetRelativeLocationOfPivotPoint))
  (Transformation|SetRelativeLocation (Variables|Default|GetPivotHelper) _relativelocationofpivotpoint false false)
  (Transformation|SetRelativeLocation (Variables|Default|GetRotationGuide) _relativelocationofpivotpoint false false)
  (Transformation|SetRelativeLocation (Variables|Default|GetScaleGuide) _relativelocationofpivotpoint false false))

(fn UpdateRelativeLocationOfPivotPointandGuides)

(fn UpdateRotationLockedObjects
  (bind _lockedstaticmeshcomponents (Variables|Default|GetLockedStaticMeshComponents))
  (bind _keys (Utilities|Map|Keys _lockedstaticmeshcomponents))
  (for _array_element _keys
    (Transformation|SetWorldLocation _array_element (+ (* (.location (Utilities|Map|Find _lockedstaticmeshcomponents _array_element)) (Transformation|GetWorldScale (Variables|Default|GetDefaultSceneRoot))) (Transformation|GetWorldLocation (Variables|Default|GetDefaultSceneRoot))) false true)
    (Transformation|SetWorldRotation _array_element (.rotation (Utilities|Map|Find _lockedstaticmeshcomponents _array_element)) false true)))

(fn UpdateFXState
  (switch Utilities|FlowControl|Switch|SwitchonEnum_CreativeMoveTool_PreviewFX (Variables|Default|GetFXState)
    (:NewEnumerator1
      (CallFunction|InitiateMaterialTransition "Forward"))
    (:NewEnumerator2
      (CallFunction|InitiateMaterialTransition "Backward"))
    (:NewEnumerator3
      (CallFunction|DeleteObjectfromWorld))
    (:NewEnumerator4
      (CallFunction|DuplicateMaterialFlash)
      (CallFunction|ServerSetFXStatewithNotify "NewEnumerator0"))
    (:NewEnumerator5
      (CallFunction|InitiateMaterialTransition "Backward"))))

(fn ConstructionScript
  (Collision|SetActorEnableCollision false)
  (Networking|CanExecuteCosmeticEvents
    (:True
      (CallFunction|SetAbsoluteTransformsonComponents)
      (Transformation|DetachFromComponent (Variables|Default|GetShadowPlane) "KeepRelative" "KeepRelative" "KeepRelative" true)
      (bind _returnvalue (Rendering|Material|CreateDynamicMaterialInstance (Variables|Default|GetRotationGuide) 0 0 "None" :!class /Script/Engine.PrimitiveComponent))
      (Variables|RotationGuide|SetRotationGuideMID _returnvalue)
      (bind _returnvalue_2 (Rendering|Material|CreateDynamicMaterialInstance (Variables|Default|GetScaleGuide) 0 0 "None" :!class /Script/Engine.PrimitiveComponent))
      (Variables|ScaleGuide|SetScaleGuideMID _returnvalue_2)
      (bind _returnvalue_3 (Rendering|Material|CreateDynamicMaterialInstance (Variables|Default|GetShadowPlane) 0 0 "None" :!class /Script/Engine.PrimitiveComponent))
      (Variables|GravityGuide|SetShadowPlaneMID _returnvalue_3))))

(fn ValidateMeshComponents
  ;[0 1 0.1847 0.316] Iterate through our meshes and ensure that each mesh component wants to be controlled by the preview and is able to be controlled (has any materials)
  (CallFunction|ConfirmMeshComponents)
  (bind (_arrayindex _arrayelement) (Utilities|Array|ReverseforEachLoop (Variables|Default|GetMeshComponents))
    (:LoopBody
      (Utilities|IsValid _arrayelement
        (:Is%20Valid
          (if (Components|ComponentHasTag _arrayelement "MaterialChangeIgnoredByBlueprintPreviewer")
            _
            (elif (== (Rendering|Material|GetNumMaterials _arrayelement) 0)
              (bind _ascreative_edit_only_mesh_component (Utilities|Casting|CastToCreativeEditOnlyMeshComponent _arrayelement)
                (:then
                  (if (Class|CreativeEditOnlyMeshComponent|IsLoadingAssets _ascreative_edit_only_mesh_component)
                    _
                    (else
                      _)))
                (:CastFailed
                  _)))))
        (:Is%20Not%20Valid
          _))
      (Utilities|Array|RemoveIndex (Variables|Default|GetMeshComponents) _arrayindex))))
