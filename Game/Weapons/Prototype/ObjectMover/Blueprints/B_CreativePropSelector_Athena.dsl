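; Prop Hunt weapon, built on the Creative move tool, that lets a player possess props as a disguise.
; It spawns a PropSelector interaction mode to supply possession behavior and minigame settings, which govern cancelling and re-equipping the primary item.
; The pawn's possessed-prop event pushes or pops an input component and HUD hints for place, noise and cancel, driven by tag-based abilities.
; Unequip and DBNO cancel possession; controller changes remove leftover input.
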(event EventBeginPlay
  (|CallParentFunction)
  (Networking|CanExecuteCosmeticEvents
    (:False
      (CallFunction|EventInitializeObjectInteractionModeStack))))

(event FortWeapon|EventOnEquippedAndAttached
  ;[1 1 1] OnEquip - Bind event to OnPossessedProp
  (|CallParentFunction)
  (bind _asfort_player_pawn (Utilities|Casting|CastToFortPlayerPawn (Game|GetInstigator)))
  (EventDispatchers|AssignOnUnholsteredEvent _asfort_player_pawn (EventDispatchers|CreateEvent :!event Custom|OnUnholsteredEvent))
  (EventDispatchers|BindEventtoOnPlayerStartDBNO _asfort_player_pawn (EventDispatchers|CreateEvent :!event Custom|OnPlayerStartDBNO))
  (if (Pawn|IsLocallyControlled _asfort_player_pawn)
    (PropHunt|AssignOnPossessedProp _asfort_player_pawn (EventDispatchers|CreateEvent :!event Custom|OnPossessedProp))
    ;;[0.2447 0.5801 0.6198 0.5] If the controller changes it means this player died while disguised - Make sure to remove the disguise input component
    (Pawn|BindEventtoReceiveControllerChangedDelegate _asfort_player_pawn (delegate Custom|OnControlledPawnChanged))))

(event Custom|OnControlledPawnChanged
  (params (Pawn :type Pawn%20Object%20Reference) (OldController :type Controller%20Object%20Reference) (NewController :type Controller%20Object%20Reference))
  ;[1 1 1] OnEquip - Bind event to OnPossessedProp
  ;;[0.2447 0.5801 0.6198 0.5] If the controller changes it means this player died while disguised - Make sure to remove the disguise input component
  (Pawn|UnbindEventfromReceiveControllerChangedDelegate Pawn (delegate Custom|OnControlledPawnChanged))
  (bind _asplayer_controller (Utilities|Casting|CastToPlayerController OldController))
  (CallFunction|PopPossessedPropInput _asplayer_controller))

(event FortWeapon|EventOnLocalUnEquip
  ;[1 1 1] OnUnEquip - Unbind event from OnPossessedProp
  (|CallParentFunction)
  (Variables|/Script/Engine.PostProcessComponent|SetbEnabled false (Variables|Default|GetHighlightPostProcess))
  (CallFunction|CancelPropPossession)
  (bind _asfort_player_pawn (Utilities|Casting|CastToFortPlayerPawn (Game|GetInstigator)))
  (EventDispatchers|UnbindEventfromOnUnholsteredEvent _asfort_player_pawn (EventDispatchers|CreateEvent :!event Custom|OnUnholsteredEvent))
  (EventDispatchers|UnbindEventfromOnPlayerStartDBNO _asfort_player_pawn (EventDispatchers|CreateEvent :!event Custom|OnPlayerStartDBNO))
  (if (Pawn|IsLocallyControlled _asfort_player_pawn)
    (PropHunt|UnbindEventfromOnPossessedProp _asfort_player_pawn (EventDispatchers|CreateEvent :!event Custom|OnPossessedProp))))

(event Custom|OnPossessedProp
  (params (FortPlayerPawn :type Fort%20Player%20Pawn%20Object%20Reference) (PossessedProp :type Building%20Gameplay%20Actor%20Player%20Prop%20Attachment%20Object%20Reference))
  (bind _returnvalue (Fortnite|GetFortPlayerControllerfromActor FortPlayerPawn))
  ;[1 1 1] OnPossessedProp - Add / remove input component
  (Utilities|IsValid FortPlayerPawn
    (:Is%20Valid
      (CallFunction|SetPropPlaced false)
      (Utilities|IsValid PossessedProp
        (:Is%20Valid
          (Utilities|IsValid _returnvalue
            (:Is%20Valid
              (CallFunction|PushPossessedPropInput _returnvalue))))
        (:Is%20Not%20Valid
          (Utilities|IsValid _returnvalue
            (:Is%20Valid
              (CallFunction|PopPossessedPropInput _returnvalue))))))))

(event OnPossessedPropPlace(FortPossessedPropInput)
  ;[1 1 1] OnPossessedPropPlace - toggle "Placed" ability
  (CallFunction|SetPropPlaced (not (Variables|Default|GetPossessedPropPlaced))))

(event OnPossessedPropCancel(FortPossessedPropInput)
  ;[1 1 1] OnPossessedPropCancel - cancel possessed prop ability
  (if (CallFunction|AllowCancelPropPossession)
    (CallFunction|CancelPropPossession)
    (if (CallFunction|EquipPrimaryItemonCancelPropPossession)
      (bind _asfort_player_controller (Utilities|Casting|CastToFortPlayerController (Game|GetInstigatorController)))
      (Quickbars|ClientActivateSlot _asfort_player_controller "Primary" 0 0.0 true false))))

(event OnPossessedPropMakeNoise(FortPossessedPropInput)
  ;[1 1 1] OnPossessedPropMakeNoise - activate "make noise" ability
  (bind _asfort_player_pawn (Utilities|Casting|CastToFortPlayerPawn (Game|GetInstigator)))
  (Abilities|TryActivateAbilitiesbyTag (Ability|GetAbilitySystemComponent _asfort_player_pawn) (Variables|Default|GetPossessedPropMakeNoiseTag) true))

(event InputEvents|EventPushCreativeMoveToolEquippedInputComponent
  ;[1 1 1] Input
  (bind _asfort_player_controller (Utilities|Casting|CastToFortPlayerController (Game|GetInstigatorController)))
  (MoveTool|Equipped|PushPropPossessorEquippedInputComponent (CallFunction|GetCreativeMoveToolInputComponent) _asfort_player_controller)
  (CallFunction|SetLastBoundController _asfort_player_controller))

(event InputEvents|EventPopCreativeMoveToolEquippedInputComponent
  ;[1 1 1] Input
  (bind _asfort_player_controller (Utilities|Casting|CastToFortPlayerController (Game|GetInstigatorController)))
  (MoveTool|Equipped|PopPropPossessorEquippedInputComponent (CallFunction|GetCreativeMoveToolInputComponent) _asfort_player_controller)
  (CallFunction|SetLastBoundController _asfort_player_controller))

(event InputEvents|EventHandleExitPressed)

(event InputEvents|EventHandleExitReleased)

(event Custom|OnUnholsteredEvent
  ;[1 1 1] OnUnholsteredEvent - If we are unholstered while possessing a prop, make sure weapon visibility remains off
  (bind _asfort_player_pawn (Utilities|Casting|CastToFortPlayerPawn (Game|GetInstigator)))
  (Utilities|IsValid (PropHunt|GetPossessedProp _asfort_player_pawn)
    (:Is%20Valid
      (FortWeapon|SetWeaponVisibility false))))

(event InputEvents|EventOnCopyOrPickupFailed
  (|CallParentFunction)
  (CallFunction|PlayGameplayAbilityFail))

(event Custom|PlayGameplayAbilityFail
  (Class|ObjectInteractionBehavior|PlayGameplayAbility (Variables|/Game/Weapons/Prototype/ObjectMover/Blueprints/ObjectInteractionModes/PhoneToolActorTargets/CreativeInteractionMode_PropSelector.CreativeInteractionMode_PropSelector_C|GetCreativeInteraction_PossessProp (Variables|Default|GetPropSelectorInteractionMode)) "/Game/Weapons/Prototype/ObjectMover/Blueprints/ObjectInteractionBehaviors/GA_Failed_Click.GA_Failed_Click_C"))

(event Custom|OnPlayerStartDBNO
  ;[1 1 1] OnPlayerStartDBNO - Cancel Prop Possession when DBNO starts
  (CallFunction|CancelPropPossession))

(event Custom|EventInitializeObjectInteractionModeStack
  ;[1 1 1] Construct ObjectInteractionMode Stack (establishes the priority each mode will be queried when looking for MoveTool target)
  (bind _returnvalue (Game|SpawnActorfromClass (Variables|Default|GetCreativeInteractionMode_PropSelector) (Transformation|GetActorTransform (Actor|GetOwner)) "Undefined" "MultiplyWithRoot" self 0))
  (Variables|Default|SetPropSelectorInteractionMode _returnvalue))

(event EventEndPlay
  (params (EndPlayReason :type EEndPlayReason%20Enum :default "Destroyed"))
  ;[1 1 1] Cleanup
  (|CallParentFunction EndPlayReason)
  (Networking|CanExecuteCosmeticEvents
    (:False
      (Actor|DestroyActor (Variables|Default|GetPropSelectorInteractionMode)))))

(fn AllowCancelPropPossession
  (bind (_possessor_config_interface _minigame) (CallFunction|GetPropHuntSettings))
  (if (Utilities|IsValid _minigame :!class /Script/Engine.KismetSystemLibrary)
    (if (Minigame|IsinProgress _minigame)
      (bind (_show_props_remaining _prop_health_scaling _allow_change_of_prop _allow_cancel_prop_possession) (CreativePropPossession|GetPropDetails _possessor_config_interface))
      (return _allow_cancel_prop_possession)
      (else
        _))
    (else
      _))
  (return true))

(fn CancelPropPossession
  (bind _fpp (LocalVariables|GetFPP))
  (bind _possessedproptag (Variables|Default|GetPossessedPropTag))
  (CallFunction|SetPropPlaced false)
  (bind _asfort_player_pawn (Utilities|Casting|CastToFortPlayerPawn (Game|GetInstigator)))
  (LocalVariables|SetFPP _asfort_player_pawn)
  (bind _asfort_ability_system_component (Utilities|Casting|CastToFortAbilitySystemComponent (Ability|GetAbilitySystemComponent (LocalVariables|GetFPP))))
  (Tags|CancelAbilitiesWithTags _asfort_ability_system_component _possessedproptag)
  (if (Abilities|HasActiveAbilitywithAnyMatchingTag _asfort_ability_system_component _possessedproptag)
    _
    (elif (Pawn|IsLocallyControlled _fpp)
      (bind _asplayer_controller (Utilities|Casting|CastToPlayerController (Game|GetInstigatorController _fpp)))
      (CallFunction|PopPossessedPropInput _asplayer_controller))))

(fn CanUtilizeMoveToolSpawningFunctionality
  (bind _returnvalue (|CallParentFunction))
  (return _returnvalue))

(fn EquipPrimaryItemOnCancelPropPossession
  (bind (_possessor_config_interface _minigame) (CallFunction|GetPropHuntSettings))
  (if (Utilities|IsValid _minigame :!class /Script/Engine.KismetSystemLibrary)
    (if (Minigame|IsinProgress _minigame)
      (bind _true (CreativePropPossession|EquipPrimaryItemAfterCancelPossession _possessor_config_interface))
      (return _true)
      (else
        _))
    (else
      _))
  (return false))

(fn GetCreativeInteractionPossessPropComponent
  (return (Variables|/Game/Weapons/Prototype/ObjectMover/Blueprints/ObjectInteractionModes/PhoneToolActorTargets/CreativeInteractionMode_PropSelector.CreativeInteractionMode_PropSelector_C|GetCreativeInteraction_PossessProp (Variables|Default|GetPropSelectorInteractionMode))))

(fn GetHighlightSound
  (return "/Game/Sounds/Creative/Modes/PropHunt/PropHunt_Mouseover_Cue.PropHunt_Mouseover_Cue"))

(fn GetPossessedPropHealDrainAmount
  (return (Variables|Default|GetPossessedPropHealDrainAmount)))

(fn GetPropHuntSettings
  (bind _propselectorinteractionmode (Variables|Default|GetPropSelectorInteractionMode))
  (bind _creativeinteraction_possessprop (Variables|/Game/Weapons/Prototype/ObjectMover/Blueprints/ObjectInteractionModes/PhoneToolActorTargets/CreativeInteractionMode_PropSelector.CreativeInteractionMode_PropSelector_C|GetCreativeInteraction_PossessProp _propselectorinteractionmode))
  (bind _minigame (Variables|/Game/Weapons/Prototype/ObjectMover/Blueprints/ObjectInteractionBehaviors/PossessProp/CreativeInteraction_PossessProp.CreativeInteraction_PossessProp_C|GetMinigame _creativeinteraction_possessprop))
  (Utilities|IsValid _minigame
    (:Is%20Valid
      (bind _asprop_possessor_config_interface (Utilities|Casting|CastToPropPossessorConfigInterface (Variables|/Game/Athena/Playset/Minigames/Minigame_PlayerCreated.Minigame_PlayerCreated_C|GetPropHuntSettings _minigame))
        (:then
          (return _asprop_possessor_config_interface _minigame))
        (:CastFailed
          _)))
    (:Is%20Not%20Valid
      _))
  (return 0 0))

(fn PopPossessedPropInput
  (params (PlayerController :type Player%20Controller%20Object%20Reference))
  (Utilities|IsValid PlayerController
    (:Is%20Valid
      (bind _fortpossessedpropinput (Variables|Default|GetFortPossessedPropInput)
        (:then
          (PropHunt|PopPossessedPropInputMode _fortpossessedpropinput PlayerController))
        (:else
          _))
      (WeaponHUDData|SetHudKeyActionVisibility 0)
      (WeaponHUDData|SetHudKeyActionVisibility 1)
      (WeaponHUDData|SetHudKeyActionVisibility 2))
    (:Is%20Not%20Valid
      (Fortnite|FortShippingLog self "CreativePropSelector::PopPossessedPropInput has invalid PlayerController" true))))

(fn PushPossessedPropInput
  (params (PlayerController :type Player%20Controller%20Object%20Reference))
  (Utilities|IsValid PlayerController
    (:Is%20Valid
      (bind _fortpossessedpropinput (Variables|Default|GetFortPossessedPropInput)
        (:then
          (PropHunt|PushPossessedPropInputMode _fortpossessedpropinput PlayerController)
          (WeaponHUDData|SetHudKeyActionVisibility 0 true)
          (WeaponHUDData|SetHudKeyActionVisibility 1 true)
          (if (CallFunction|AllowCancelPropPossession)
            (WeaponHUDData|SetHudKeyActionVisibility 2 true)))
        (:else)))
    (:Is%20Not%20Valid
      (Fortnite|FortShippingLog self "CreativePropSelector::PushPossessedPropInput has invalid PlayerController" true))))

(fn SetPropPlaced
  (params (NewPropPlaced :type Boolean))
  (bind _returnvalue (Game|GetInstigator))
  (bind _returnvalue_2 (Ability|GetAbilitySystemComponent _returnvalue))
  (if NewPropPlaced
    (bind _returnvalue_3 (Abilities|TryActivateAbilitiesbyTag _returnvalue_2 (Variables|Default|GetPossessedPropPlacedTag) true))
    (Variables|Default|SetPossessedPropPlaced _returnvalue_3)
    (else
      (Variables|Default|SetPossessedPropPlaced false)
      (bind _asfort_ability_system_component (Utilities|Casting|CastToFortAbilitySystemComponent _returnvalue_2))
      (Tags|CancelAbilitiesWithTags _asfort_ability_system_component (Variables|Default|GetPossessedPropPlacedTag)))))

(fn ConstructionScript
  (|CallParentFunction))
