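; Blueprint interface for the prototype ObjectMover weapon's prop possession selector.
; It declares two getters: one for the creative interaction possess-prop component and one for the possessed prop's heal or drain amount.
; Both functions hold only empty default returns, so any behaviour comes from implementing classes.
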(fn GetCreativeInteractionPossessPropComponent
  (return 0))

(fn GetPossessedPropHealDrainAmount
  (return 0.0))
