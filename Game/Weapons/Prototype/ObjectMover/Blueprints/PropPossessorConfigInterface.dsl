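; Blueprint interface for configuring a prop possessor, apparently for prop disguise in the ObjectMover prototype weapon.
; Query functions expose disguise animation override, ping, prop and filtering details, plus whether to equip the primary item after cancelling possession.
; Three notification functions receive an instigating Controller when a player begins entering, finishes entering or exits a disguise.
; All functions are stubs with default returns, leaving behavior to implementers.
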(fn EquipPrimaryItemAfterCancelPossession
  (return false))

(fn GetDisguiseAnimationOverrideDetails
  (return false 0.0))

(fn GetPingDetails
  (return false 0.0))

(fn GetPropDetails
  (return false false false false))

(fn GetFilteringDetails
  (return 0))

(fn OnPlayerBeginEnteringDisguise
  (params (Instigator :type Controller%20Object%20Reference)))

(fn OnPlayerExitingDisguise
  (params (Instigator :type Controller%20Object%20Reference)))

(fn OnPlayerFinishEnteringDisguise
  (params (Instigator :type Controller%20Object%20Reference)))
