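# Translucent unlit particle material for the creative move tool prototype, lerped by "Use Block" between a sphere-masked particle and an animated block effect.
# Masked particles blend two colors using particle random.
# Blocks sample screen-space noise, panned in discrete steps via ceiled time, to drive three-color tinting and opacity within a texture-derived shape overlaid with scanlines.
# Dynamic parameters control noise offset, particle-color blending, block visibility and opacity.
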Material /Game/Weapons/Prototype/ObjectMover/Materials/M_CreativeMoveTool_Blocks
graph {
 properties {
  bEnableResponsiveAA=true
  bUsedWithInstancedStaticMeshes=true
  bUsedWithNiagaraMeshParticles=true
  bUsedWithNiagaraSprites=true
  bUsedWithParticleSprites=true
  blendMode=BLEND_Translucent
  pixelDepthOffsetMode=PDOM_Legacy
  shadingModel=MSM_Unlit
  usedShadingModels=Unlit
 }
 nodes [
  n1 = Multiply(A=ScreenPosition().ViewportUV, B=ScalarParameter() {defaultValue=10, parameterName=Block UV}) {constB=10}
  n2 = DynamicParameter() {defaultValue=<rgba 0 0 1 1>, paramNames [BlockUVRandom, BlockAlphaAdd, BlockSwitch, Opacity]}
  n3 = Multiply(A=Time(), B=ScalarParameter() {defaultValue=10.7749, group=Block, parameterName=Block U Speed}) {constB=0.5}
  n4 = TextureSample(UVs=Panner(Coordinate=Add(A=@n1, B=@n2.BlockUVRandom), Time=Ceil(None=@n3)) {desc="-0.33,-0.43", speedX=-0.33, speedY=-0.43}) {texture=/Game/Athena/UI/Challenges/Art/T_UI_DataChip_Noise}
  LinearInterpolate(Alpha=@n4.R)
  n5 = ParticleColor()
  n6 = Subtract(A=TextureCoordinate()) {constB=0.5}
  n7 = Constant() {r=0.1}
  n8 = SphereMask(A=Power(Base=Abs(None=@n6)) {constExponent=3}, B=Constant(), Radius=@n7, Hardness=@n7)
  n9 = Add(A=TextureCoordinate() {uTiling=0.5, vTiling=0.5}, B=Constant2Vector() {g=0.25, r=0.25})
  n10 = Saturate(None=Multiply(A=TextureSample(UVs=@n9) {texture=/Game/Effects/Fort_Effects/Textures/Husk/Poison/T_FakeBubble_Spec}.R) {constB=10})
  n11 = DynamicParameter() {defaultValue=<rgba 0 0 1 1>, paramNames [BlockUVRandom, BlockAlphaAdd, BlockSwitch, Opacity]}
  n12 = Multiply(A=@n10, B=Saturate(None=Add(A=@n4.R, B=@n11.BlockSwitch)))
  n13 = ParticleColor()
  n14 = MaterialFunctionCall!(
   Number of Scanlines=ScalarParameter() {defaultValue=150, parameterName=Scanline Amount}
   Scanline Speed
   "Scanlines Thickness (Bias)"=ScalarParameter() {defaultValue=-0.15, parameterName=ScanlineSize}
   Scanlines Dark Value=ScalarParameter() {defaultValue=0.5, parameterName=Scanlines Dark}
   Scanlines Light Value=ScalarParameter() {defaultValue=1, parameterName=Scanlines Light}
   Block Opacity
   UVs
   Time
   Block Repeat
  ) {materialFunction=/Game/Effects/Fort_Effects/Effects/Atlas/Materials/MF_Atlas_Scanlines_01}
  n15 = Multiply(A=@n13.A, B=@n14.Result)
  n16 = ScalarParameter() {group=Block, parameterName=Use Block}
  LinearInterpolate(
   A=Multiply(A=@n5.A, B=@n8)
   B=Multiply(
    A=Multiply(A=@n12, B=@n15)
    B=DynamicParameter() {defaultValue=<rgba 0 0 1 1>, paramNames [BlockUVRandom, BlockAlphaAdd, BlockSwitch, Opacity]}.Opacity
   )
   Alpha=@n16
  )
  n17 = LinearInterpolate(
   A=VectorParameter() {parameterName=ColorLow, defaultValue=<rgba 0.0627 0.1128 1 1>}.RGB
   B=VectorParameter() {parameterName=ColorHigh, defaultValue=<rgba 0.2798 1.5453 5 1>}.RGB
   Alpha=Multiply(A=ParticleRandom(), B=@n8)
  )
  n18 = Multiply(A=@n17, B=@n5.RGB)
  n19 = MaterialFunctionCall(
   A=VectorParameter() {group=Block, parameterName=BlockColor3, defaultValue=<rgba 0.5333 0 1 0>}.RGB
   B=VectorParameter() {group=Block, parameterName=BlockColor2, defaultValue=<rgba 0 0.3336 1 0>}.RGB
   C=VectorParameter() {group=Block, parameterName=BlockColor, defaultValue=<rgba 0.0333 1 0 0>}.RGB
   Alpha=@n4.R
  ) {materialFunction=&Lerp_3Color}
  n20 = LinearInterpolate(A=Desaturation(None=Multiply(A=@n10, B=@n19.Result), Fraction=Constant() {r=-1}), B=Multiply(A=@n10, B=@n13.RGB), Alpha=@n11.BlockAlphaAdd)
  MaterialFunctionCall(Emissive=LinearInterpolate(A=StaticSwitchParameter(True=@n18, False=@n17) {parameterName=UseParticleColorForEmissive}, B=Multiply(A=@n20, B=@n14.Result), Alpha=@n16)) {materialFunction=&MF_FX_EmissiveAdjustments}
  TextureCoordinate() {uTiling=0.2, vTiling=0.2}
 ]
}
