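# Translucent unlit particle sprite material for a simple blue flare on the creative-mode object move tool prototype.
# It builds the flare shape procedurally without textures, multiplying a radial UV sphere mask by a U-only band mask and sharpening the result with a power.
# That shape blends low and high color parameters, tinted by particle color and passed through MF_FX_EmissiveAdjustments.
# The same shape times particle alpha provides the opacity value.
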Material /Game/Weapons/Prototype/ObjectMover/Materials/M_CreativeMoveTool_FlareSimple
graph {
 properties {
  bUsedWithParticleSprites=true
  blendMode=BLEND_Translucent
  pixelDepthOffsetMode=PDOM_Legacy
  shadingModel=MSM_Unlit
  usedShadingModels=Unlit
 }
 nodes [
  n1 = VectorParameter() {parameterName=ColorLow, defaultValue=<rgba 0.0627 0.1128 1 1>}
  n2 = VectorParameter() {parameterName=ColorHigh, defaultValue=<rgba 0.1679 0.9272 3 1>}
  n3 = TextureCoordinate()
  n4 = Constant() {r=0.5}
  n5 = Constant() {r=0.5}
  n6 = Constant()
  n7 = ComponentMask(None=TextureCoordinate()) {r=true}
  n8 = Constant() {r=0.5}
  n9 = Constant() {r=0.05}
  n10 = Power(Base=Multiply(A=SphereMask(A=@n3, B=@n4, Radius=@n5, Hardness=@n6), B=SphereMask(A=@n7, B=@n8, Radius=@n8, Hardness=@n9))) {constExponent=6}
  n11 = ParticleColor()
  MaterialFunctionCall(Emissive=Multiply(A=LinearInterpolate(A=@n1.RGB, B=@n2.RGB, Alpha=@n10), B=@n11.RGB)) {materialFunction=&MF_FX_EmissiveAdjustments}
  Multiply(A=@n11.A, B=@n10)
 ]
}
