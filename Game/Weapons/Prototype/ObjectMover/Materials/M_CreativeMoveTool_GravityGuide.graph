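# Translucent unlit material for the Creative move tool's gravity guide, drawing a line with a pulse that repeatedly travels along it.
# A time-driven sphere mask, with V stretched by object X scale to offset mesh scaling, forms the pulse and blends ColorLow into ColorHigh.
# That pulse plus a center-line mask from Radius and Hardness is multiplied by a lengthwise fade.
# SplineThicken sets a uniform width from SplineWidth.
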Material /Game/Weapons/Prototype/ObjectMover/Materials/M_CreativeMoveTool_GravityGuide
graph {
 properties {
  bEnableResponsiveAA=true
  blendMode=BLEND_Translucent
  pixelDepthOffsetMode=PDOM_Legacy
  previewMesh=/Game/Effects/Fort_Effects/Mesh/Splines/SM_Straight_Spline_100
  shadingModel=MSM_Unlit
  usedShadingModels=Unlit
 }
 nodes [
  n1 = MaterialFunctionCall(Float2=TextureCoordinate()) {materialFunction=&BreakOutFloat2Components}
  n2 = Multiply(
   A=MaterialFunctionCall() {materialFunction=/Engine/Functions/Engine_MaterialFunctions02/WorldPositionOffset/ObjectScale}.Scale X
   B=Constant() {r=3}
  )
  n3 = OneMinus(None=Multiply(A=Time() {bOverride_Period=true, period=0.6667}) {constB=1.5})
  n4 = SphereMask(
   A=MaterialFunctionCall(X=@n1.R, Y=Multiply(A=@n1.G, B=@n2)) {materialFunction=&MakeFloat2}.Result
   B=AppendVector(A=Constant() {r=0.5}, B=Multiply(A=@n3, B=@n2))
   Radius=Constant() {r=0.5}
   Hardness=Constant()
  )
  LinearInterpolate(
   A=VectorParameter() {parameterName=ColorLow, defaultValue=<rgba 0.0668 0.0854 0.8 0>}.RGB
   B=VectorParameter() {parameterName=ColorHigh, defaultValue=<rgba 0.3748 0.8296 4 0>}.RGB
   Alpha=@n4
  )
  n5 = ScalarParameter() {defaultValue=8, parameterName=SplineWidth}
  MaterialFunctionCall(WidthBase=@n5, WidthTip=@n5) {materialFunction=&SplineThicken}
  n6 = MaterialFunctionCall(Float2=TextureCoordinate()) {materialFunction=&BreakOutFloat2Components}
  n7 = Constant() {r=0.5}
  n8 = ScalarParameter() {defaultValue=0.1, parameterName=Radius}
  n9 = ScalarParameter() {defaultValue=0.1, parameterName=Hardness}
  n10 = Constant() {r=0.5}
  Multiply(
   A=Add(A=@n4, B=SphereMask(A=@n6.R, B=@n7, Radius=@n8, Hardness=@n9))
   B=SphereMask(A=Power(Base=@n6.G), B=@n10, Radius=@n10, Hardness=Constant() {r=0.7})
  )
 ]
}
