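# Translucent, unlit, two-sided grid material for the Creative move tool, by its name used to show grid placement.
# Two world-aligned samples of one texture, at parameterized and fixed sizes on Z-stretched world position, form the grid, tinted red, green or blue by normal axis.
# The texture's blue channel blends in shadow color, and UV sphere masks draw boundary edges.
# A Fresnel-modulated grid value plus edges forms an opacity-like term, dimmed on back faces.
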Material /Game/Weapons/Prototype/ObjectMover/Materials/M_CreativeMoveTool_GridPlacement
graph {
 properties {
  allowTranslucentCustomDepthWrites=true
  bEnableResponsiveAA=true
  blendMode=BLEND_Translucent
  pixelDepthOffsetMode=PDOM_Legacy
  shadingModel=MSM_Unlit
  twoSided=true
  usedShadingModels=Unlit
 }
 nodes [
  n1 = TextureObjectParameter() {parameterName="TextureObject (T2d)", texture=/Game/Weapons/Prototype/ObjectMover/Textures/T_CreativeMoveTool_Grid}
  n2 = Multiply(A=WorldPosition().XYZ, B=Constant3Vector() {constant=<rgba 1 1 1.3333 0>})
  n3 = ScalarParameter() {defaultValue=100, parameterName=ProjectionTransitionContrast}
  n4 = MaterialFunctionCall(
   TextureObject=@n1
   TextureSize=ScalarParameter() {defaultValue=128, parameterName=GridSize}
   WorldPosition=@n2
   ProjectionTransitionContrast=@n3
  ) {materialFunction=&WorldAlignedTexture}
  n5 = MaterialFunctionCall(TextureObject=@n1, TextureSize=Constant() {r=64}, WorldPosition=@n2, ProjectionTransitionContrast=@n3) {materialFunction=&WorldAlignedTexture}
  n6 = Saturate(None=Add(A=ComponentMask(None=@n4.XYZ Texture) {g=true}, B=ComponentMask(None=@n5.XYZ Texture) {r=true}))
  n7 = Abs(None=ConstantBiasScale(None=TextureCoordinate()) {bias=-0.5, scale=2})
  n8 = ComponentMask(None=@n7) {r=true}
  n9 = Constant() {r=1}
  n10 = ScalarParameter() {defaultValue=0.015, parameterName=EdgeRadius}
  n11 = ScalarParameter() {defaultValue=0.5, parameterName=EdgeHardness}
  n12 = ComponentMask(None=@n7) {g=true}
  n13 = Saturate(None=Add(A=SphereMask(A=@n8, B=@n9, Radius=@n10, Hardness=@n11), B=SphereMask(A=@n12, B=@n9, Radius=@n10, Hardness=@n11)))
  Add(A=@n6, B=@n13)
  n14 = ScalarParameter() {defaultValue=0.15, parameterName=BackSideOpacity}
  n15 = Add(A=@n6, B=ScalarParameter() {defaultValue=0.07, parameterName=BG})
  n16 = MaterialFunctionCall(
   Brightness=ScalarParameter() {defaultValue=2, parameterName=Fresnel Brightness}
   Exponent=ScalarParameter() {defaultValue=3, parameterName=Fresnel Exponent}
  ) {materialFunction=&MF_Fresnel_FX}
  n17 = ComponentMask(None=@n4.XYZ Texture) {b=true}
  n18 = Add(A=LinearInterpolate(A=Multiply(A=@n15, B=@n16.Fresnel), Alpha=@n17) {constB=0.5}, B=@n13)
  LinearInterpolate(A=Add(A=Multiply(A=@n14, B=@n18), B=@n13), B=@n18, Alpha=Saturate(None=TwoSidedSign()))
  n19 = VectorParameter() {parameterName=ColorBG, defaultValue=<rgba 0.5 0.3167 0.1991 0>}
  n20 = MaterialFunctionCall(Float3=Abs(None=VertexNormalWS())) {materialFunction=&BreakOutFloat3Components}
  n21 = LinearInterpolate(
   A=LinearInterpolate(
    A=VectorParameter() {parameterName=X, defaultValue=<rgba 1 0.0573 0.0573 0>}.RGB
    B=VectorParameter() {parameterName=Y, defaultValue=<rgba 0.026 1 0.026 0>}.RGB
    Alpha=@n20.G
   )
   B=VectorParameter() {parameterName=Z, defaultValue=<rgba 0.0312 0.0312 1 0>}.RGB
   Alpha=@n20.B
  )
  LinearInterpolate(
   A=LinearInterpolate(
    A=LinearInterpolate(A=@n19.RGB, B=@n21, Alpha=@n6)
    B=VectorParameter() {parameterName=ShadowColor, defaultValue=<rgba 0.0677 0.0289 0.0111 0>}.RGB
    Alpha=@n17
   )
   B=VectorParameter() {parameterName=BoundaryColor, defaultValue=<rgba 0.3245 0.4281 0.6337 0>}.RGB
   Alpha=@n13
  )
  n22 = Subtract(A=TextureCoordinate()) {constB=0.5}
  SphereMask(
   A=Power(Base=Abs(None=@n22), Exp=ScalarParameter() {defaultValue=3, parameterName=MaskExp})
   B=Constant()
   Radius=ScalarParameter() {defaultValue=0.12, parameterName=MaskRadius}
   Hardness=ScalarParameter() {defaultValue=0.1, parameterName=MaskHardness}
  )
 ]
}
