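# Masked, two-sided lit material for the Creative move tool prototype, apparently an intermediate look that squashes a mesh away as WPOTransition changes.
# Normalized height from object bounds is compared with WPOTransition, and vertices at or below it get a Z offset onto that moving plane, scaled by RadiusScale.
# In the transition's upper half, a lerp adds an XY pull toward the object center, collapsing the shape.
# BaseColor, Normals and MaskTexture parameters supply surface inputs.
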Material /Game/Weapons/Prototype/ObjectMover/Materials/M_CreativeMoveTool_Intermediate
graph {
 properties {bUsedWithInstancedStaticMeshes=true, blendMode=BLEND_Masked, pixelDepthOffsetMode=PDOM_Legacy, twoSided=true, usedShadingModels=Default Lit}
 nodes [
  Constant()
  n1 = WorldPosition()
  n2 = ComponentMask(None=Subtract(A=@n1.XYZ, B=ObjectPositionWS())) {b=true}
  n3 = ComponentMask(None=ObjectBounds()) {b=true}
  n4 = ScalarParameter() {defaultValue=-100, group=CreativeMoveTool, parameterName=WPOTransition}
  n5 = ObjectPositionWS()
  n6 = Multiply(
   A=Subtract(A=@n4, B=Constant() {r=0.5})
   B=Multiply(A=@n3, B=ScalarParameter() {defaultValue=2, group=CreativeMoveTool, parameterName=RadiusScale}) {constB=1.75}
  )
  n7 = Multiply(A=Subtract(A=Add(A=@n5, B=@n6), B=ComponentMask(None=@n1.XYZ) {b=true}), B=Constant3Vector() {constant=<rgba 0 0 1 0>})
  n8 = If(A=ConstantBiasScale(None=Divide(A=@n2, B=@n3)), B=@n4, "A > B"=Constant(), "A == B"=@n7, "A < B"=@n7)
  n9 = Subtract(A=ObjectPositionWS(), B=WorldPosition().XYZ)
  n10 = Constant3Vector() {constant=<rgba 1 1 0 0>}
  LinearInterpolate(A=@n8, B=Add(A=@n8, B=Multiply(A=@n9, B=@n10)), Alpha=Saturate(None=ConstantBiasScale(None=@n4) {bias=-0.5, scale=2}))
  Sine(None=Time())
  TextureSampleParameter2D() {parameterName=Normals, samplerType=SAMPLERTYPE_Normal, texture=/Engine/EngineMaterials/FlatNormal}
  TextureSampleParameter2D() {parameterName=MaskTexture, texture=/Engine/EngineResources/WhiteSquareTexture}
  VectorParameter() {parameterName=BaseColor, defaultValue=<rgba 0.2344 0.2344 0.2344 0>}
 ]
}
