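# Unlit, two-sided alpha-composite lightbeam material for the Creative move tool, usable on mesh particles and Niagara effects.
# Fresnel-shaped panning noise, swept scanlines and optional cloud noise build color and opacity, with depth fade gated by quality level.
# Time of day, camera distance, view angle, gradient masks and a sine pulse then scale both.
# Parameter-blended world position offset handles camera alignment, spin, camera push and distance-based collapse toward the object position.
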Material /Game/Weapons/Prototype/ObjectMover/Materials/M_CreativeMoveTool_Lightbeam
graph {
 properties {
  bEnableResponsiveAA=true
  bUsedWithMeshParticles=true
  bUsedWithNiagaraMeshParticles=true
  bUsedWithNiagaraSprites=true
  blendMode=BLEND_AlphaComposite
  pixelDepthOffsetMode=PDOM_Legacy
  shadingModel=MSM_Unlit
  twoSided=true
  usedShadingModels=Unlit
 }
 nodes [
  CameraVectorWS()
  n1 = MaterialFunctionCall!(
   WorldPosition
   PivotLocation
   WS Normals
   Custom Object Basis 1=VectorParameter() {group=WPO, parameterName=Custom Obj Basis 1, defaultValue=<rgba 0 0 1 1>}.RGB
   Custom Object Basis 2=VectorParameter() {group=WPO, parameterName=Custom Obj Basis 2, defaultValue=<rgba 0 1 0 1>}.RGB
   Custom Object Basis 3=VectorParameter() {group=WPO, parameterName=Custom Obj Basis 3, defaultValue=<rgba 1 0 0 1>}.RGB
  ) {materialFunction=/Engine/Functions/Engine_MaterialFunctions02/WorldPositionOffset/AlignMeshToTheCamera}
  n2 = Transform(None=VectorParameter() {group=WPO, parameterName=WPO spin Axis, defaultValue=<rgba 0 0 1 1>}.RGB) {transformSourceType=TRANSFORMSOURCE_Local}
  n3 = Multiply(
   A=Add(A=Time(), B=ScalarParameter() {group=WPO, parameterName=WPO Time Spin Offset 1})
   B=ScalarParameter() {group=WPO, parameterName=WPO Spin Speed}
  )
  n4 = ScalarParameter() {group=WPO, parameterName=WPO Time Spin Offset 2}
  n5 = Transform(None=VectorParameter() {group=WPO, parameterName=WPO Spin Offset, defaultValue=<rgba 0 0 0 1>}.RGB) {transformSourceType=TRANSFORMSOURCE_Local}
  n6 = RotateAboutAxis(
   NormalizedRotationAxis=NamedRerouteDeclaration(None=Normalize(VectorInput=@n2)) {name=Rot Axis, nodeColor=<rgba 1 0.7031 0 1>}
   RotationAngle=NamedRerouteDeclaration(None=Add(A=@n3, B=@n4)) {name=Rot Angle, nodeColor=<rgba 0.3828 0 1 1>}
   PivotPoint=MaterialFunctionCall() {materialFunction=&ComponentPivotLocation}.Result
   Position=Add(A=WorldPosition().XYZ, B=@n5)
  )
  n7 = Add(
   A=Multiply(A=@n1.World Position Offset, B=ScalarParameter() {group=WPO, parameterName=AlignMeshToCamera Mul})
   B=Multiply(A=ScalarParameter() {group=WPO, parameterName=WPO Spin Multiplier, sliderMax=1}, B=Add(A=@n6, B=@n5))
  )
  n8 = ScalarParameter() {group=WPO, parameterName=Vertical Mask Camera offset Initial}
  n9 = ScalarParameter() {group=WPO, parameterName=Verical Mask Smoothsetp range Camera offset}
  n10 = SmoothStep(Min=Subtract(A=@n8, B=@n9), Max=Add(A=@n8, B=@n9), Value=NamedRerouteUsage())
  n11 = LinearInterpolate(
   A=ScalarParameter() {group=WPO, parameterName=Cam Offset Amt low}
   B=ScalarParameter() {group=WPO, parameterName=Cam Offset Amt High}
   Alpha=@n10
  )
  n12 = LinearInterpolate(
   A=ScalarParameter() {group=WPO, parameterName=Cam Offset Amt low Far}
   B=ScalarParameter() {group=WPO, parameterName=Cam Offset Amt High Far}
   Alpha=@n10
  )
  n13 = InverseLinearInterpolate(
   A=ScalarParameter() {parameterName=Cam offet distance min}
   B=ScalarParameter() {defaultValue=1, parameterName=Cam Offset distance Max}
   Value=NamedRerouteUsage()
  ) {bClampResult=true}
  n14 = WorldPosition()
  n15 = Add(
   A=@n7
   B=MaterialFunctionCall(
    Offset Amount=LinearInterpolate(A=@n11, B=@n12, Alpha=@n13)
    Clamp Padding=ScalarParameter() {group=WPO, parameterName=Clamp Padding}
    World Position=Add(A=@n7, B=@n14.XYZ)
   ) {materialFunction=/Engine/Functions/Engine_MaterialFunctions02/WorldPositionOffset/CameraOffset}.Clamped Camera Offset
  )
  n16 = ObjectPositionWS()
  n17 = WorldPosition()
  n18 = ScalarParameter() {defaultValue=1, parameterName=COllaspe size near}
  n19 = ScalarParameter() {defaultValue=1, parameterName=COllaspe size Far}
  n20 = CameraPositionWS()
  n21 = ObjectPositionWS()
  n22 = Abs(None=Length(None=Subtract(A=@n20, B=@n21)) {is=class MaterialExpressionLength})
  n23 = InverseLinearInterpolate(
   A=ScalarParameter() {parameterName=collaspe distance min}
   B=ScalarParameter() {defaultValue=1, parameterName=collaspe distance Max}
   Value=@n22
  ) {bClampResult=true}
  LinearInterpolate(
   A=@n15
   B=LinearInterpolate(A=@n15, B=Subtract(A=@n16, B=@n17.XYZ), Alpha=LinearInterpolate(A=@n18, B=@n19, Alpha=@n23))
   Alpha=ScalarParameter() {group=WPO, parameterName=Use Collapse WPO}
  )
  MaterialFunctionCall() {materialFunction=&CameraDirectionVector}
  MaterialFunctionCall() {materialFunction=&DebugTimeSine}
  NamedRerouteDeclaration(None=@n1.Rotated Normals) {name=Rotted normal, nodeColor=<rgba 0.25 0.25 1 1>}
  NamedRerouteDeclaration(None=@n22) {name=DistToCam, nodeColor=<rgba 0.0312 0 1 1>}
  n24 = Normalize(VectorInput=Transform(None=Constant3Vector() {constant=<rgba 0 0 1 0>}) {transformSourceType=TRANSFORMSOURCE_Local})
  NamedRerouteDeclaration(None=@n24) {name=Rot Axis_1, nodeColor=<rgba 1 0.7031 0 1>}
  n25 = Subtract(
   A=CameraPositionWS()
   B=MaterialFunctionCall() {materialFunction=&ComponentPivotLocation}.Result
  )
  NamedRerouteDeclaration(None=Normalize(VectorInput=@n25)) {name=Direction to camera, nodeColor=<rgba 0.5937 1 0 1>}
  n26 = Constant3Vector() {constant=<rgba 1 1 0 0>}
  n27 = Normalize(VectorInput=Multiply(A=@n25, B=@n26))
  n28 = Time()
  n29 = ScalarParameter() {group=WPO, parameterName=WPO Time Spin Offset 1 2}
  n30 = Add(
   A=Multiply(A=Add(A=@n28, B=@n29), B=ScalarParameter() {defaultValue=0.06, group=WPO, parameterName=WPO Spin Speed 2})
   B=ScalarParameter() {group=WPO, parameterName=WPO Time Spin Offset 2 2}
  )
  n31 = MaterialFunctionCall() {materialFunction=&ComponentPivotLocation}
  n32 = Transform(None=Constant3Vector() {constant=<rgba 1 0 0 0>}) {transformSourceType=TRANSFORMSOURCE_Local}
  n33 = Normalize(VectorInput=Multiply(A=RotateAboutAxis(NormalizedRotationAxis=@n24, RotationAngle=@n30, PivotPoint=@n31.Result, Position=@n32), B=@n26))
  n34 = OneMinus(None=Abs(None=DotProduct(A=@n27, B=@n33)))
  n35 = VectorParameter() {parameterName=Emissive mul options hor, channelNames {a=contrast, b=add, g=mul, r=pow}, defaultValue=<rgba 1 1 0 0>}
  n36 = Add(A=Multiply(A=Power(Base=@n34, Exp=@n35.pow), B=@n35.mul), B=@n35.add)
  NamedRerouteDeclaration(None=@n36) {name=Debug, nodeColor=<rgba 1 0.75 0 1>}
  NamedRerouteUsage()
  n37 = VectorParameter() {parameterName=TOD Scales, channelNames {a=Night, b=Evening, g=Day, r=Morning}, defaultValue=<rgba 1 1 1 1>}
  n38 = MaterialFunctionCall(Value_Morning=@n37.Morning, Value_Day=@n37.Morning, Value_Evening=@n37.Morning, Value_Night=@n37.Morning) {materialFunction=&Scalar_Time_of_Day}
  n39 = ParticleColor()
  n40 = ScalarParameter() {defaultValue=1, parameterName=Fade}
  n41 = TextureCoordinate()
  n42 = NamedRerouteDeclaration(None=StaticSwitchParameter(True=OneMinus(None=@n41), False=@n41) {parameterName=FlipUVs}) {name=Flipped UVs, nodeColor=<rgba 1 0 0.3984 1>}
  n43 = ScalarParameter() {defaultValue=0.9, parameterName=FalloffRadius}
  n44 = Divide(A=ScalarParameter() {defaultValue=300, parameterName=Length}, B=ScalarParameter() {defaultValue=400, parameterName=FalloffBaseLength})
  n45 = SphereMask(
   A=ComponentMask(None=@n42) {g=true}
   B=Constant() {r=1}
   Radius=Saturate(None=Divide(A=@n43, B=@n44))
   Hardness=ScalarParameter() {defaultValue=0.025, parameterName=FalloffHardness}
  )
  n46 = ScalarParameter() {defaultValue=4.5, parameterName=FresnelExponentHigh}
  n47 = ScalarParameter() {defaultValue=1, parameterName=FresnelExponentLow}
  n48 = Power(Base=@n45, Exp=ScalarParameter() {defaultValue=3.5, parameterName=MaskExponent})
  n49 = NamedRerouteUsage()
  n50 = NamedRerouteUsage()
  n51 = LinearInterpolate(A=VertexNormalWS(), B=NamedRerouteUsage(), Alpha=ScalarParameter() {group=WPO, parameterName=AlignMeshToCamera Mul})
  n52 = LinearInterpolate(
   A=ScalarParameter() {parameterName=fresnel Low}
   B=ScalarParameter() {defaultValue=1, parameterName=fresnel High}
   Alpha=MaterialFunctionCall(
    Brightness=ScalarParameter() {defaultValue=1, parameterName=FresnelBrightness}
    Exponent=LinearInterpolate(A=@n46, B=@n47, Alpha=@n48)
    Vertex / Pixel Normal=MaterialFunctionCall(Rotation Axis=@n49, Rotation Angle=@n50, World Space Vertex Normals=@n51) {materialFunction=&FixRotateAboutAxisNormals}.New World Space Vertex Normal
   ) {materialFunction=&MF_Fresnel_FX}.Fresnel Double Sided
  )
  n53 = OneMinus(None=@n48)
  n54 = ComponentMask(None=VectorParameter() {parameterName=NoiseUVTileA, defaultValue=<rgba 2 0.09 0 0>}.RGB) {g=true, r=true}
  n55 = VectorParameter() {parameterName=NoisePanSpeedA, defaultValue=<rgba 0.36 0.11 0 0>}
  n56 = TextureObjectParameter() {parameterName=Noise Tex RG, samplerType=SAMPLERTYPE_LinearColor, texture=&T_NoisePack_LinearColor}
  n57 = ScalarParameter() {parameterName=Noisue Tex Mip Bias}
  n58 = TextureSample!(
   UVs=Panner(Coordinate=Multiply(A=@n42, B=@n54), Speed=ComponentMask(None=@n55.RGB) {g=true, r=true}) {bFractionalPart=true}
   Tex=@n56
   Bias=@n57
   Apply View MipBias
  ) {mipValueMode=TMVM_MipBias, samplerType=SAMPLERTYPE_LinearColor, texture=&T_NoisePack_LinearColor}
  n59 = ComponentMask(None=VectorParameter() {parameterName=NoiseUVTileB, defaultValue=<rgba 3 0.11 0 0>}.RGB) {g=true, r=true}
  n60 = VectorParameter() {parameterName=NoisePanSpeedB, defaultValue=<rgba -0.63 0 0.1335 0>}
  n61 = TextureSample(
   UVs=Panner(Coordinate=Multiply(A=@n42, B=@n59), Speed=ComponentMask(None=@n60.RGB) {g=true, r=true}) {bFractionalPart=true, speedX=-0.026}
   Tex=@n56
   Bias=@n57
  ) {mipValueMode=TMVM_MipBias, samplerType=SAMPLERTYPE_LinearColor, texture=&T_NoisePack_LinearColor}
  n62 = LinearInterpolate(
   A=ScalarParameter() {parameterName=Texture Low}
   B=ScalarParameter() {defaultValue=1, parameterName=Texture High}
   Alpha=Add(A=Subtract(A=@n58.R) {constB=0.5}, B=Subtract(A=@n61.G) {constB=0.5})
  )
  n63 = Multiply(A=@n52, B=Multiply(A=Multiply(A=@n53, B=@n62), B=ScalarParameter() {defaultValue=1.5, parameterName=NoiseScale}))
  n64 = ScalarParameter() {defaultValue=0.6, parameterName=OpacityHigh}
  n65 = Multiply(
   A=ScalarParameter() {defaultValue=100, parameterName=Steps}
   B=Divide(A=ScalarParameter() {defaultValue=300, parameterName=Length}, B=ScalarParameter() {defaultValue=300, parameterName=ScanlineBaseLength})
  )
  n66 = ScalarParameter() {defaultValue=500, parameterName=MaxSteps}
  n67 = Frac(None=Multiply(A=ComponentMask(None=@n42) {g=true}, B=Clamp(None=@n65, Max=@n66) {clampMode=CMODE_ClampMax}))
  n68 = Constant() {r=0.5}
  n69 = ScalarParameter() {defaultValue=0.3, parameterName=ScanlinesRadius}
  n70 = ScalarParameter() {parameterName=ScanlinesHardness}
  n71 = Subtract(A=ComponentMask(None=@n42) {g=true}) {constB=0.5}
  n72 = ScalarParameter() {defaultValue=1.5, parameterName=Speed}
  n73 = Multiply(A=Time() {bOverride_Period=true, period=1.5}, B=@n72)
  n74 = Divide(B=@n72)
  n75 = Multiply(A=OneMinus(None=Multiply(A=@n73, B=@n74))) {constB=4}
  n76 = ScalarParameter() {defaultValue=1.5, parameterName=SweepRadius}
  n77 = ScalarParameter() {parameterName=SweepHardness}
  n78 = Multiply(
   A=Multiply(A=SphereMask(A=@n67, B=@n68, Radius=@n69, Hardness=@n70), B=SphereMask(A=@n71, B=@n75, Radius=@n76, Hardness=@n77))
   B=ScalarParameter() {defaultValue=0.55, parameterName=ScanlinesOpacity}
  )
  n79 = Constant()
  n80 = Multiply(
   A=@n52
   B=Add(A=LinearInterpolate(B=@n64, Alpha=@n48), B=StaticSwitchParameter(True=@n78, False=@n79) {defaultValue=true, parameterName=UseScanlines})
  )
  n81 = Multiply(A=@n45, B=Add(A=@n63, B=@n80))
  n82 = MaterialFunctionCall(Float2=TextureCoordinate()) {materialFunction=&BreakOutFloat2Components}
  n83 = Constant() {r=0.5}
  n84 = SphereMask(A=@n82.R, B=@n83, Radius=@n83, Hardness=Constant())
  n85 = ScalarParameter() {defaultValue=0.3, parameterName=CloudNoiseScaleU}
  n86 = ScalarParameter() {defaultValue=0.5, parameterName=CloudNoiseScaleV}
  n87 = Panner(
   Coordinate=Multiply(A=AppendVector(A=@n84, B=@n82.G), B=AppendVector(A=@n85, B=@n86))
   Time=Add(A=ScalarParameter() {bUseCustomPrimitiveData=true, parameterName=Random}, B=Time())
   Speed=AppendVector(A=ScalarParameter() {parameterName=CloudNoisePanSpeedU}, B=ScalarParameter() {defaultValue=-0.0122, parameterName=CloudNoisePanSpeedV})
  ) {bFractionalPart=true}
  n88 = TextureObjectParameter() {parameterName=Noise Tex 2 R, texture=&AtmosphericCloudNoise01}
  n89 = Multiply(
   A=Subtract(A=TextureSample!(UVs=@n87, Tex=@n88, Apply View MipBias) {texture=&AtmosphericCloudNoise01}.R) {constB=0.5}
   B=ScalarParameter() {defaultValue=0.25, parameterName=CloudNoiseScale}
  )
  n90 = StaticSwitchParameter(True=Add(A=@n81, B=@n89), False=@n81) {parameterName=UseCloudNoise}
  n91 = ScalarParameter() {defaultValue=30, parameterName=FadeDistance}
  n92 = DepthFade(Opacity=@n90, FadeDistance=@n91)
  n93 = ScalarParameter() {parameterName=Use Depth Fade On epic}
  n94 = LinearInterpolate(A=@n90, B=@n92, Alpha=@n93)
  n95 = VertexColor()
  n96 = VectorParameter() {parameterName=VertexColorMask, defaultValue=<rgba 0 0 0 1>}
  n97 = StaticSwitchParameter(True=DotProduct(A=AppendVector(A=@n95, B=@n95), B=AppendVector(A=@n96.RGB, B=@n96.RGB)), False=Constant() {r=1}) {parameterName=MaskVertexColor}
  n98 = Multiply(A=Multiply(A=@n40, B=QualitySwitch(Default=@n90, Low=@n90, High=@n92, Medium=@n92, Epic=@n94)), B=@n97)
  n99 = Multiply(A=@n39.A, B=@n98)
  n100 = StaticSwitchParameter(True=@n99, False=@n98) {parameterName=UseParticleColor}
  n101 = Constant() {r=1}
  n102 = MaterialFunctionCall(Distance=ScalarParameter() {defaultValue=100, parameterName=NearCameraFadeDistance}) {materialFunction=&Near_Camera_Fade}
  n103 = LinearInterpolate(
   A=QualitySwitch(Default=@n101, Low=@n101, High=@n102.Result, Medium=@n102.Result)
   B=MaterialFunctionCall(Fade Length=ScalarParameter() {parameterName=CameraDepthFade Lenght}, Fade Offset=ScalarParameter() {parameterName=CameraDepthFade Offset}) {materialFunction=&CameraDepthFade}.Result
   Alpha=ScalarParameter() {parameterName=Use Better Camera Fade}
  )
  n104 = Multiply(A=@n100, B=@n103)
  n105 = ScalarParameter() {defaultValue=28000, parameterName=Dist Remal Mul A}
  n106 = ScalarParameter() {defaultValue=35000, parameterName=Dist Remal Mul b}
  n107 = Abs(None=Distance(
    A=MaterialFunctionCall() {materialFunction=&ComponentPivotLocation}.Result
    B=CameraPositionWS()
   ))
  n108 = LinearInterpolate(
   B=LinearInterpolate(
    A=ScalarParameter() {parameterName=Dist Remal Mul Low}
    B=ScalarParameter() {defaultValue=1, parameterName=Dist Remal Mul High}
    Alpha=InverseLinearInterpolate(A=@n105, B=@n106, Value=@n107) {bClampResult=true}
   )
   Alpha=ScalarParameter() {parameterName=Use Distance Remap Multiplier}
  ) {constA=1}
  n109 = Multiply(A=Multiply(A=@n38.Result, B=StaticSwitchParameter(True=@n104, False=@n100) {parameterName=UseNearCameraFade}), B=@n108)
  n110 = ScalarParameter() {parameterName=Emissive low Hor}
  n111 = ScalarParameter() {defaultValue=1, parameterName=Emissive High Hor}
  n112 = Saturate(None=MaterialFunctionCall(In=@n36, Contrast=@n35.contrast) {materialFunction=&CheapContrast}.Result)
  n113 = Abs(None=DotProduct(A=Normalize(VectorInput=@n25), B=Constant3Vector() {constant=<rgba 0 0 1 0>}))
  n114 = VectorParameter() {parameterName=Emissive mul options 2 ver, channelNames {a=contrast, b=add, g=mul, r=pow}, defaultValue=<rgba 1 1 0 0>}
  n115 = Power(Base=LinearInterpolate(Alpha=OneMinus(None=@n113)), Exp=@n114.pow)
  n116 = MaterialFunctionCall(In=Add(A=Multiply(A=@n115, B=@n114.mul), B=@n114.add), Contrast=@n114.contrast) {materialFunction=&CheapContrast}
  n117 = LinearInterpolate(
   B=LinearInterpolate(
    A=ScalarParameter() {defaultValue=0.05, parameterName=Emissive low Hor 2}
    B=LinearInterpolate(A=@n110, B=@n111, Alpha=@n112)
    Alpha=Saturate(None=@n116.Result)
   )
   Alpha=ScalarParameter() {parameterName=Use LH emissive mul}
  ) {constA=1}
  n118 = ScalarParameter() {defaultValue=0.25, parameterName=Vertical Mask Initial}
  n119 = ScalarParameter() {defaultValue=0.1, parameterName=Verical Mask Smoothsetp range}
  n120 = Subtract(A=@n118, B=@n119)
  n121 = Add(A=@n118, B=@n119)
  n122 = DotProduct(A=NamedRerouteUsage(), B=Normalize(VectorInput=VectorParameter() {parameterName=Mask, defaultValue=<rgba 0 0 1 1>}.RGB))
  n123 = NamedRerouteDeclaration(None=OneMinus(None=Abs(None=@n122))) {name=VMask, nodeColor=<rgba 1 0 0 1>}
  n124 = LinearInterpolate(
   B=LinearInterpolate(
    A=ScalarParameter() {defaultValue=1, parameterName=Vertical Mask Opacity Low}
    B=ScalarParameter() {defaultValue=1, parameterName=Vertical Mask Opacity High}
    Alpha=SmoothStep(Min=@n120, Max=@n121, Value=@n123) {constMin=0.4}
   )
   Alpha=ScalarParameter() {parameterName=Use Vertical Mask}
  ) {constA=1}
  n125 = Multiply(A=Time(), B=ScalarParameter() {defaultValue=0.2, parameterName=Pulse Speed})
  n126 = MaterialFunctionCall(In=ConstantBiasScale(None=Sine(None=@n125)), Contrast=ScalarParameter() {defaultValue=0.2, parameterName=Pulse contrast}) {materialFunction=&CheapContrast}
  n127 = LinearInterpolate(
   A=ScalarParameter() {defaultValue=1, parameterName=Pulse Low, sliderMax=1}
   B=ScalarParameter() {defaultValue=1, parameterName=Pulse High, sliderMax=1}
   Alpha=NamedRerouteDeclaration(None=Saturate(None=@n126.Result)) {name=PulseTime, nodeColor=<rgba 1 0.3984 0 1>}
  )
  n128 = Multiply(A=Multiply(A=Multiply(A=@n109, B=@n117), B=@n124), B=@n127)
  n129 = ScalarParameter() {defaultValue=0.25, parameterName=LH Mask Vertical Mask Initial}
  n130 = ScalarParameter() {defaultValue=0.1, parameterName=LH Mask Verical Mask Smoothsetp range}
  n131 = Subtract(A=@n129, B=@n130)
  n132 = Add(A=@n129, B=@n130)
  n133 = NamedRerouteUsage()
  n134 = Add(
   A=MaterialFunctionCall(UV Channel=NamedRerouteUsage()) {materialFunction=&LinearGradient}.VGradient
   B=LinearInterpolate(
    A=ScalarParameter() {defaultValue=-0.2, parameterName=LH Mask Small Offset Low}
    B=ScalarParameter() {defaultValue=-0.2, parameterName=LH Mask Small Offset High}
    Alpha=SmoothStep(Min=@n131, Max=@n132, Value=@n133) {constMin=0.4}
   )
  ) {constB=0}
  n135 = ScalarParameter() {defaultValue=0.5, parameterName=LH Mask Small Contrast}
  n136 = MaterialFunctionCall(UV Channel=NamedRerouteUsage()) {materialFunction=&LinearGradient}
  n137 = LinearInterpolate(
   A=ScalarParameter() {parameterName=Beams pulse Low}
   B=ScalarParameter() {defaultValue=-0.5, parameterName=Beams pulse high}
   Alpha=NamedRerouteUsage()
  )
  n138 = Saturate(None=MaterialFunctionCall(In=Add(A=@n136.VGradient, B=@n137) {constB=0}, Contrast=ScalarParameter() {defaultValue=0.5, parameterName=pulse Beams Contrast}) {materialFunction=&CheapContrast}.Result)
  n139 = MaterialFunctionCall!(
   Width=LinearInterpolate(
    A=ScalarParameter() {defaultValue=0.2, parameterName=LH Mask Width}
    B=ScalarParameter() {defaultValue=0.3, parameterName=LH Mask Width bottpm}
    Alpha=MaterialFunctionCall() {materialFunction=&LinearGradient}.VGradient
   )
   Sharpness=ScalarParameter() {defaultValue=1, parameterName=LH Mask Sharpness}
   Offset=ScalarParameter() {parameterName=LH Mask Offset}
   Direction Switch=StaticBool() {value=true}
   Compare
   Input Coordinates=Multiply(A=TextureCoordinate()) {constB=2}
  ) {materialFunction=/Engine/Functions/Engine_MaterialFunctions03/Procedurals/GeneratedBand}
  n140 = LinearInterpolate(
   B=LinearInterpolate(
    A=MaterialFunctionCall(In=@n134, Contrast=@n135) {materialFunction=&CheapContrast}.Result
    Alpha=Multiply(A=@n138, B=@n139.Result)
   )
   Alpha=ScalarParameter() {parameterName=Use LH Mask}
  ) {constA=1}
  n141 = VectorParameter() {parameterName=Bottom Mask options, channelNames {a=constrast, b=add, g=mul, r=pow}, defaultValue=<rgba 1 1 0 0>}
  n142 = OneMinus(None=MaterialFunctionCall(UV Channel=NamedRerouteUsage()) {materialFunction=&LinearGradient}.VGradient)
  n143 = Add(A=Multiply(A=@n141.mul, B=Power(Base=@n142, Exp=@n141.pow) {constExponent=5}), B=@n141.add)
  n144 = LinearInterpolate(
   B=Saturate(None=MaterialFunctionCall(In=@n143, Contrast=@n141.constrast) {materialFunction=&CheapContrast}.Result)
   Alpha=ScalarParameter() {parameterName=UseBottom Mask}
  ) {constA=1}
  Reroute(None=Multiply(A=Multiply(A=@n128, B=@n140), B=@n144)) {desc=opac}
  n145 = VectorParameter() {bUseCustomPrimitiveData=true, parameterName=CustomPrimitiveColor, primitiveDataIndex=28}
  n146 = Multiply(A=Add(A=@n80, B=@n63), B=@n45)
  n147 = StaticSwitchParameter(True=Add(A=@n89, B=@n146), False=@n146) {parameterName=UseCloudNoise}
  n148 = LinearInterpolate(
   A=VectorParameter() {parameterName=ColorLow, defaultValue=<rgba 0.0833 0.0247 0.8 1>}.RGB
   B=VectorParameter() {parameterName=ColorHigh, defaultValue=<rgba 0.1438 0.6079 2 1>}.RGB
   Alpha=Saturate(None=@n147)
  )
  n149 = DepthFade(Opacity=@n147, FadeDistance=@n91)
  n150 = QualitySwitch(Default=@n147, Low=@n147, High=@n149, Medium=@n149, Epic=LinearInterpolate(A=@n147, B=@n149, Alpha=@n93))
  n151 = Multiply(
   A=Multiply(A=StaticSwitchParameter(True=@n145.RGB, False=@n148) {parameterName=ColorUsesCustomPrimitiveVec4}, B=Multiply(A=@n150, B=@n40))
   B=@n97
  )
  n152 = StaticSwitchParameter(True=Multiply(A=Multiply(A=@n151, B=@n39.RGB), B=@n99), False=@n151) {parameterName=UseParticleColor}
  n153 = StaticSwitchParameter(True=Multiply(A=@n152, B=@n103), False=@n152) {parameterName=UseNearCameraFade}
  n154 = Multiply(A=Multiply(A=Multiply(A=@n38.Result, B=@n153), B=@n108), B=@n117)
  n155 = Multiply(A=Multiply(A=Multiply(A=@n154, B=@n124), B=@n127), B=@n140)
  Reroute(None=MaterialFunctionCall(Emissive=Multiply(A=@n155, B=@n144)) {materialFunction=&MF_FX_EmissiveAdjustments}.Result) {desc=emiss}
  ScalarParameter() {parameterName=SmoothsetpRange}
  ScalarParameter() {defaultValue=5, parameterName=Bottom Mask Pow}
  VectorParameter() {parameterName=Camera offset Mask Options}
 ]
}
