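# Skinned-mesh material for the Creative move tool's phone prop, pairing phone body textures with an animated, tinted screen.
# Time drives a sine over the tiled transition texture's green channel for pulsing blocks, added to a red-channel grid.
# A screen mask confines the pattern, ScreenColor tints it, and a crossfade of two icon textures (IconLERP) is added on top.
# Icon UVs are offset and center-scaled by parameters.
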Material /Game/Weapons/Prototype/ObjectMover/Materials/M_CreativeMoveTool_Phone
graph {
 properties {bUsedWithMeshDeformer=true, bUsedWithSkeletalMesh=true, pixelDepthOffsetMode=PDOM_Legacy, usedShadingModels=Default Lit}
 nodes [
  n1 = ConstantBiasScale(None=Time() {bOverride_Period=true, period=1}) {bias=-0.5, scale=2}
  n2 = TextureSample(UVs=TextureCoordinate() {uTiling=3, vTiling=3}) {texture=/Game/Weapons/Prototype/ObjectMover/Textures/T_Movetool_Transition}
  n3 = ConstantBiasScale(None=@n2.G) {bias=-0.5, scale=2}
  n4 = Sine(None=Add(A=Multiply(A=@n1, B=@n3), B=@n3))
  n5 = ScalarParameter() {defaultValue=0.25, parameterName=BlocksScale}
  n6 = ScalarParameter() {defaultValue=0.5, parameterName=GridIntensity}
  n7 = TextureSample() {samplerType=SAMPLERTYPE_LinearColor, texture=/Game/Weapons/Prototype/ObjectMover/Textures/T_CellPhone_CreativeToolMasks}
  n8 = Multiply(A=Add(A=Multiply(A=@n4, B=@n5), B=Multiply(A=@n2.R, B=@n6)), B=@n7.R)
  n9 = VectorParameter() {parameterName=IconPositionUVs, defaultValue=<rgba -0.27 0.07 0 0>}
  n10 = MaterialFunctionCall(
   UVs=Add(A=TextureCoordinate(), B=ComponentMask(None=@n9.RGB) {g=true, r=true})
   Texture Scale=ComponentMask(None=VectorParameter() {parameterName=IconScaleUVs, defaultValue=<rgba 0.4 0.4 0 0>}.RGB) {g=true, r=true}
  ) {materialFunction=&ScaleUVsByCenter}
  Add(
   A=Multiply(A=Add(A=@n8, B=@n7.R), B=VectorParameter() {parameterName=ScreenColor}.RGB)
   B=LinearInterpolate(
    A=TextureSampleParameter2D(UVs=@n10.UVs) {parameterName=IconA, texture=&T_BlacksRGB}.R
    B=TextureSampleParameter2D(UVs=@n10.UVs) {parameterName=IconB, texture=&T_BlacksRGB}.R
    Alpha=ScalarParameter() {parameterName=IconLERP}
   )
  )
  TextureSample() {texture=/Game/Weapons/Prototype/ObjectMover/Textures/T_Cell_Phone_NoScreen}
  TextureSample() {samplerType=SAMPLERTYPE_Normal, texture=/Game/Animation/Props/Cell_Phone/Textures/T_Cell_Phone_N}
  TextureSample() {samplerType=SAMPLERTYPE_Masks, texture=/Game/Animation/Props/Cell_Phone/Textures/T_Cell_Phone_S}
 ]
}
