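# Unlit translucent material with responsive AA, for the pivot helper of the creative object mover tool, judging by its name and path.
# A world-space vertex offset adds a timed sine pulse along the direction to the object center, scaled by PulseScaleWS.
# It also scales the mesh about its pivot by camera-to-pivot distance, keeping it readable from afar.
# A Fresnel function blends ColorLow to ColorHigh, alongside an Opacity parameter.
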Material /Game/Weapons/Prototype/ObjectMover/Materials/M_CreativeMoveTool_PivotHelper
graph {
 properties {bEnableResponsiveAA=true, blendMode=BLEND_Translucent, pixelDepthOffsetMode=PDOM_Legacy, shadingModel=MSM_Unlit, usedShadingModels=Unlit}
 nodes [
  n1 = Normalize(VectorInput=Subtract(A=ObjectPositionWS(), B=WorldPosition().XYZ))
  n2 = Sine(None=Time()) {period=0.33}
  n3 = WorldPosition()
  n4 = MaterialFunctionCall() {materialFunction=&ObjectPivotPoint}
  n5 = Distance(
   A=MaterialFunctionCall() {materialFunction=&ObjectPivotPoint}.Object Pivot Location
   B=CameraPositionWS()
  )
  Add(
   A=Multiply(A=Multiply(A=@n1, B=@n2), B=ScalarParameter() {defaultValue=0.5, parameterName=PulseScaleWS})
   B=Multiply(A=Subtract(A=@n3.XYZ, B=@n4.Object Pivot Location), B=Divide(A=@n5) {constB=1000})
  )
  LinearInterpolate(
   A=VectorParameter() {parameterName=ColorLow, defaultValue=<rgba 0.021 0 0.4601 0>}.RGB
   B=VectorParameter() {parameterName=ColorHigh, defaultValue=<rgba 0.2443 0.7632 1.5 0>}.RGB
   Alpha=MaterialFunctionCall(
    Brightness=ScalarParameter() {defaultValue=1, parameterName="Brightness (S)"}
    Exponent=ScalarParameter() {defaultValue=2, parameterName="Exponent (S)"}
   ) {materialFunction=&MF_Fresnel_FX}.Fresnel
  )
  ScalarParameter() {defaultValue=1, parameterName=Opacity}
 ]
}
