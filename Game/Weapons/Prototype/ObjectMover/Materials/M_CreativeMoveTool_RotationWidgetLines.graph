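# Translucent unlit material for the line rings of the creative move tool's rotation widget.
# SplineThicken and a pivot-relative offset, both driven by camera distance, enlarge line width and widget as the camera moves away.
# A soft line mask plus a looping, reversible sweep on red and blue vertex-color channels is faded by vertex alpha and UVs.
# That combined mask blends low and high colors, and the sweep tints toward the axis color.
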Material /Game/Weapons/Prototype/ObjectMover/Materials/M_CreativeMoveTool_RotationWidgetLines
graph {
 properties {
  bEnableResponsiveAA=true
  blendMode=BLEND_Translucent
  pixelDepthOffsetMode=PDOM_Legacy
  previewMesh=/Game/Weapons/Prototype/ObjectMover/Meshes/Sm_PhoneTool_RotationWidget
  shadingModel=MSM_Unlit
  translucencyLightingMode=TLM_VolumetricDirectional
  usedShadingModels=Unlit
 }
 nodes [
  n1 = Distance(
   A=MaterialFunctionCall() {materialFunction=&ObjectPivotPoint}.Object Pivot Location
   B=CameraPositionWS()
  )
  n2 = Add(
   A=LinearInterpolate(
    A=ScalarParameter() {defaultValue=10, parameterName="WidthBase (S)"}
    B=ScalarParameter() {defaultValue=3, parameterName=WidthBaseTopRing}
    Alpha=VertexColor().B
   )
   B=Divide(A=@n1) {constB=100}
  )
  n3 = WorldPosition()
  n4 = MaterialFunctionCall() {materialFunction=&ObjectPivotPoint}
  Add(
   A=MaterialFunctionCall(WidthBase=@n2, WidthTip=@n2) {materialFunction=&SplineThicken}.WorldPositionOffset
   B=Multiply(A=Subtract(A=@n3.XYZ, B=@n4.Object Pivot Location), B=Divide(A=@n1) {constB=1000})
  )
  n5 = TextureCoordinate()
  n6 = ScalarParameter() {defaultValue=8, parameterName=SweepScale}
  n7 = AppendVector(A=Constant() {r=1}, B=@n6)
  n8 = Constant() {r=0.5}
  n9 = Time() {bOverride_Period=true, period=1}
  n10 = OneMinus(None=@n9)
  n11 = ScalarParameter() {parameterName=FlipSweepDirection}
  n12 = Multiply(A=ConstantBiasScale(None=LinearInterpolate(A=@n10, B=@n9, Alpha=@n11)) {bias=-0.2, scale=1.4}, B=@n6) {constB=4}
  n13 = Multiply(
   A=SphereMask(
    A=Multiply(A=@n5, B=@n7)
    B=AppendVector(A=@n8, B=@n12)
    Radius=Constant() {r=0.3}
    Hardness=ScalarParameter() {defaultValue=0.6, parameterName=SweepHardness}
   )
   B=DotProduct(A=VertexColor(), B=Constant3Vector() {constant=<rgba 1 0 1 0>})
  )
  n14 = SphereMask(
   A=MaterialFunctionCall(Float2=TextureCoordinate()) {materialFunction=&BreakOutFloat2Components}.R
   B=Constant() {r=0.5}
   Radius=ScalarParameter() {defaultValue=0.2, parameterName=RadiusRing}
   Hardness=ScalarParameter() {defaultValue=0.1, parameterName=Hardness}
  )
  n15 = VertexColor()
  n16 = TextureCoordinate()
  n17 = Saturate(None=Multiply(A=MaterialFunctionCall(Float2=@n16) {materialFunction=&BreakOutFloat2Components}.G) {constB=2.5})
  n18 = TextureCoordinate() {coordinateIndex=1}
  n19 = ScalarParameter() {defaultValue=1, parameterName=Opacity}
  n20 = Multiply(
   A=Multiply(A=Add(A=@n13, B=@n14), B=Multiply(A=@n15.A, B=@n17))
   B=SphereMask(A=ComponentMask(None=@n18) {g=true}, B=Constant() {r=1}, Radius=@n19, Hardness=@n19)
  )
  LinearInterpolate(
   A=LinearInterpolate(
    A=VectorParameter() {parameterName=ColorLow, defaultValue=<rgba 0.0558 0.0714 0.6684 0>}.RGB
    B=VectorParameter() {parameterName=ColorHigh, defaultValue=<rgba 0.3748 0.8296 4 0>}.RGB
    Alpha=Power(Base=@n20) {constExponent=10}
   )
   B=VectorParameter() {parameterName=AxisColor, defaultValue=<rgba 6 0.7283 0.1103 0>}.RGB
   Alpha=@n13
  )
 ]
}
