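# Translucent unlit material for the Creative move tool's scale widget, drawing colored axis lines and a pivot.
# Vertex color separates pivot from axes and, dotted with ActiveAxis, selects which axes get wider glowing lines, axis tints and an animated sweep, reversible via ScaleUp.
# World position offset uses SplineThicken plus camera-distance scaling around the object pivot to grow the widget with distance.
# Opacity blends pivot fresnel with line masks, then applies a UV1-based Opacity mask.
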Material /Game/Weapons/Prototype/ObjectMover/Materials/M_CreativeMoveTool_ScaleWidget
graph {
 properties {
  bEnableResponsiveAA=true
  blendMode=BLEND_Translucent
  pixelDepthOffsetMode=PDOM_Legacy
  shadingModel=MSM_Unlit
  translucencyLightingMode=TLM_VolumetricDirectional
  usedShadingModels=Unlit
 }
 nodes [
  n1 = Distance(
   A=MaterialFunctionCall() {materialFunction=&ObjectPivotPoint}.Object Pivot Location
   B=CameraPositionWS()
  )
  n2 = Add(A=ScalarParameter() {defaultValue=8, parameterName=WidthBaseInactive}, B=Divide(A=@n1) {constB=100})
  n3 = VertexColor()
  n4 = Constant() {r=1}
  n5 = WorldPosition()
  n6 = MaterialFunctionCall() {materialFunction=&ObjectPivotPoint}
  Add(
   A=LinearInterpolate(
    B=MaterialFunctionCall(WidthBase=@n2, WidthTip=@n2) {materialFunction=&SplineThicken}.WorldPositionOffset
    Alpha=DotProduct(A=@n3, B=@n4)
   ) {constB=0}
   B=Multiply(A=Subtract(A=@n5.XYZ, B=@n6.Object Pivot Location), B=Divide(A=@n1) {constB=1000})
  )
  n7 = VertexColor()
  n8 = VectorParameter() {parameterName=ActiveAxis, defaultValue=<rgba 1 1 1 0>}
  n9 = SphereMask(
   A=MaterialFunctionCall(Float2=TextureCoordinate()) {materialFunction=&BreakOutFloat2Components}.R
   B=Constant() {r=0.5}
   Radius=LinearInterpolate(
    A=ScalarParameter() {defaultValue=0.07, parameterName=RadiusSmall}
    B=ScalarParameter() {defaultValue=0.25, parameterName=RadiusLarge}
    Alpha=DotProduct(A=@n7, B=@n8.RGB)
   )
   Hardness=ScalarParameter() {defaultValue=0.1, parameterName=Hardness}
  )
  n10 = LinearInterpolate(
   A=VectorParameter() {parameterName=ColorLow, defaultValue=<rgba 0.0558 0.0714 0.6684 0>}.RGB
   B=VectorParameter() {parameterName=ColorHigh, defaultValue=<rgba 0.3748 0.8296 4 0>}.RGB
   Alpha=Power(Base=@n9) {constExponent=10}
  )
  n11 = VertexColor()
  n12 = LinearInterpolate(
   A=LinearInterpolate(
    A=VectorParameter() {parameterName=X, defaultValue=<rgba 6 0.7283 0.1103 0>}.RGB
    B=VectorParameter() {parameterName=Y, defaultValue=<rgba 1.1958 4.8 0.475 0>}.RGB
    Alpha=@n11.G
   )
   B=VectorParameter() {parameterName=Z, defaultValue=<rgba 5 3.731 0.346 1>}.RGB
   Alpha=@n11.B
  )
  n13 = TextureCoordinate()
  n14 = ScalarParameter() {defaultValue=6, parameterName=SweepScale}
  n15 = AppendVector(A=Constant() {r=1}, B=@n14)
  n16 = Constant() {r=0.5}
  n17 = Multiply(A=Time() {bOverride_Period=true, period=0.5}) {constB=2}
  n18 = OneMinus(None=@n17)
  n19 = ScalarParameter() {defaultValue=1, parameterName=ScaleUp}
  n20 = Multiply(A=ConstantBiasScale(None=LinearInterpolate(A=@n17, B=@n18, Alpha=@n19)) {bias=-0.2, scale=1.4}, B=@n14) {constB=4}
  n21 = Multiply(
   A=DotProduct(A=VertexColor(), B=VectorParameter() {parameterName=ActiveAxis, defaultValue=<rgba 1 1 1 0>}.RGB)
   B=SphereMask(A=Multiply(A=@n13, B=@n15), B=AppendVector(A=@n16, B=@n20), Radius=Constant() {r=0.5}, Hardness=Constant() {r=0.3})
  )
  n22 = VertexColor()
  n23 = VectorParameter() {parameterName=ActiveAxis, defaultValue=<rgba 1 1 1 0>}
  LinearInterpolate(
   A=VectorParameter() {parameterName=PivotColor, defaultValue=<rgba 0.342 0.342 0.342 0>}.RGB
   B=LinearInterpolate(
    A=VectorParameter() {parameterName=InactiveColor, defaultValue=<rgba 0.0399 0.0399 0.0399 0>}.RGB
    B=LinearInterpolate(A=@n10, B=@n12, Alpha=@n21)
    Alpha=DotProduct(A=@n22, B=@n23.RGB)
   )
   Alpha=DotProduct(A=VertexColor(), B=Constant() {r=1})
  )
  n24 = ScalarParameter() {defaultValue=1.2, parameterName="Brightness (S)"}
  n25 = ScalarParameter() {defaultValue=6, parameterName="Exponent (S)"}
  n26 = VertexColor()
  n27 = Constant() {r=1}
  n28 = MaterialFunctionCall(Float2=TextureCoordinate()) {materialFunction=&BreakOutFloat2Components}
  n29 = Multiply(A=LinearInterpolate(B=Add(A=@n21, B=@n9), Alpha=DotProduct(A=@n26, B=@n27)) {constA=1}, B=Saturate(None=Multiply(A=@n28.G) {constB=2}))
  n30 = VertexColor()
  n31 = VertexColor()
  n32 = Constant() {r=1}
  n33 = TextureCoordinate() {coordinateIndex=1}
  n34 = ScalarParameter() {defaultValue=1, parameterName=Opacity}
  Multiply(
   A=LinearInterpolate(
    A=MaterialFunctionCall(Brightness=@n24, Exponent=@n25) {materialFunction=&MF_Fresnel_FX}.Fresnel
    B=Multiply(A=@n29, B=@n30.A)
    Alpha=DotProduct(A=@n31, B=@n32)
   ) {constA=1}
   B=SphereMask(A=ComponentMask(None=@n33) {g=true}, B=Constant() {r=1}, Radius=@n34, Hardness=@n34)
  )
 ]
}
