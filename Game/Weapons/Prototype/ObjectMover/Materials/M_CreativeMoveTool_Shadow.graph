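# Unlit alpha-composite material for the shadow marker of the prototype creative move tool (ObjectMover), drawing a dark circular blob with a colored ring.
# A UV sphere mask forms the soft disc, and a second sphere mask on that scalar falloff picks out a thin band that blends shadow color into ring color.
# A Fresnel using object orientation as normal fades the disc by view angle, then depth fade and an Opacity parameter scale the premultiplied color.
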Material /Game/Weapons/Prototype/ObjectMover/Materials/M_CreativeMoveTool_Shadow
graph {
 properties {bEnableResponsiveAA=true, blendMode=BLEND_AlphaComposite, pixelDepthOffsetMode=PDOM_Legacy, shadingModel=MSM_Unlit, usedShadingModels=Unlit}
 nodes [
  n1 = Constant() {r=0.5}
  n2 = SphereMask(A=TextureCoordinate(), B=@n1, Radius=@n1, Hardness=ScalarParameter() {parameterName=Hardness})
  n3 = Constant() {r=0.2}
  n4 = ScalarParameter() {defaultValue=0.01, parameterName=RingRadius}
  n5 = Constant() {r=0.8}
  n6 = MaterialFunctionCall() {materialFunction=&CameraDirectionVector}
  n7 = Multiply(
   A=@n2
   B=MaterialFunctionCall(
    Brightness=ScalarParameter() {defaultValue=4, parameterName="Brightness (S)"}
    Exponent=ScalarParameter() {defaultValue=2, parameterName="Exponent (S)"}
    Camera Vector=Multiply(A=@n6.Result) {constB=-1}
    Vertex / Pixel Normal=ObjectOrientation()
   ) {materialFunction=&MF_Fresnel_FX}.Fresnel
  )
  n8 = ScalarParameter() {defaultValue=10, parameterName=FadeDistance}
  Multiply(
   A=LinearInterpolate(
    A=VectorParameter() {parameterName=ShadowColor, defaultValue=<rgba 0.0078 0.0139 0.0417 1>}.RGB
    B=VectorParameter() {parameterName=RingColor, defaultValue=<rgba 0.5629 1.0032 3 1>}.RGB
    Alpha=SphereMask(A=@n2, B=@n3, Radius=@n4, Hardness=@n5)
   )
   B=Multiply(A=DepthFade(Opacity=@n7, FadeDistance=@n8), B=ScalarParameter() {defaultValue=1, parameterName=Opacity})
  )
 ]
}
