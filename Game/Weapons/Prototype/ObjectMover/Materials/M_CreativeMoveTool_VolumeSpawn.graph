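# Unlit, two-sided, alpha-composite material for a glowing volume-spawn effect on the Creative move tool prototype.
# An inverted double-sided Fresnel plus a powered sphere-mask falloff along UV V is scaled by an inverted object-height gradient with a fade-in bias.
# That mask blends the low and high colors and also multiplies the result.
# A quality switch depth-fades the multiplier on Medium and High only.
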Material /Game/Weapons/Prototype/ObjectMover/Materials/M_CreativeMoveTool_VolumeSpawn
graph {
 properties {
  bEnableResponsiveAA=true
  blendMode=BLEND_AlphaComposite
  pixelDepthOffsetMode=PDOM_Legacy
  previewMesh=/Game/Athena/BuildingActors/BoxWithComplexCollision
  shadingModel=MSM_Unlit
  twoSided=true
  usedShadingModels=Unlit
 }
 nodes [
  n1 = MaterialFunctionCall(Exponent=ScalarParameter() {defaultValue=0.5, parameterName=FresnelExponent}) {materialFunction=&MF_Fresnel_FX}
  n2 = SphereMask(
   A=ComponentMask(None=TextureCoordinate()) {g=true}
   B=Constant() {r=1}
   Radius=ScalarParameter() {defaultValue=0.9, parameterName=FalloffRadius}
   Hardness=ScalarParameter() {parameterName=FalloffHardness}
  )
  n3 = ScalarParameter() {defaultValue=20, parameterName=MaskExponent}
  n4 = MaterialFunctionCall!(Center, Height) {materialFunction=/Game/Effects/Fort_Effects/Materials/_Functions/MF_Z_GradientFromObjectHeight_01}
  n5 = Multiply(
   A=Add(A=OneMinus(None=@n1.Fresnel Double Sided), B=Power(Base=@n2, Exp=@n3))
   B=Add(A=OneMinus(None=@n4.Result), B=ScalarParameter() {defaultValue=1, parameterName=FadeInBias})
  )
  n6 = DepthFade(Opacity=@n5, FadeDistance=ScalarParameter() {defaultValue=30, parameterName=FadeDistance})
  Multiply(
   A=LinearInterpolate(
    A=VectorParameter() {parameterName=ColorLow, defaultValue=<rgba 0.0642 0.019 0.6163 1>}.RGB
    B=VectorParameter() {parameterName=ColorHigh, defaultValue=<rgba 0.1438 0.6079 2 1>}.RGB
    Alpha=Saturate(None=@n5)
   )
   B=QualitySwitch(Default=@n5, Low=@n5, High=@n6, Medium=@n6)
  )
 ]
}
