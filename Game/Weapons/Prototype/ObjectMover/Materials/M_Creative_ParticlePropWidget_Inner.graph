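# M_Creative_ParticlePropWidget_Inner is a masked, lit material, apparently for the inner part of the Creative move tool's prop widget.
# Time-scrolled UVs, warped by a panning swirl flow map pulsed by a sine, drive both a hue shift and a texture lookup.
# These combine with inverted Fresnel into emissive, while a dithered distance fade scaled by object radius drives opacity mask.
# The resulting attributes feed MF_CreativeMoveTool.
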Material /Game/Weapons/Prototype/ObjectMover/Materials/M_Creative_ParticlePropWidget_Inner
graph {
 properties {bUseMaterialAttributes=true, blendMode=BLEND_Masked, ditherOpacityMask=true, pixelDepthOffsetMode=PDOM_Legacy, usedShadingModels=Default Lit}
 nodes [
  n1 = Sine(None=Time()) {period=8.5}
  Add(A=@n1)
  n2 = Time() {bOverride_Period=true, period=4}
  n3 = Divide() {constB=4}
  n4 = TextureSample(UVs=Panner() {speedY=0.1}) {samplerType=SAMPLERTYPE_LinearColor, texture=/Game/Global/Textures/Default/Flowmaps/T_SwirlFlow_02}
  n5 = Add(A=Add(A=Multiply(A=@n2, B=@n3), B=TextureCoordinate()), B=ComponentMask(None=Multiply(A=@n4.RGB, B=@n1)) {g=true, r=true})
  n6 = Frac(None=@n5)
  n7 = Multiply(
   A=MaterialFunctionCall(
    Hue Shift Percentage=MaterialFunctionCall(Float2=@n6) {materialFunction=&BreakOutFloat2Components}.G
    Texture=VectorParameter() {parameterName=Base, defaultValue=<rgba 0.3 0.1005 0.1005 1>}.RGB
   ) {materialFunction=&HueShift}.Result
   B=ScalarParameter() {defaultValue=15, parameterName=EmissiveScale}
  )
  n8 = MaterialFunctionCall(Exponent=ScalarParameter() {defaultValue=1.8, parameterName="Exponent (S)"}) {materialFunction=&MF_Fresnel_FX}
  n9 = ScalarParameter() {defaultValue=20, parameterName="Fade_Start_Distance (S)"}
  n10 = Add(A=ObjectRadius(), B=ScalarParameter() {defaultValue=100, parameterName=RadiusBias})
  n11 = SetMaterialAttributes!(
   MaterialAttributes
   Base Color=VectorParameter() {parameterName=ColorHigh, defaultValue=<rgba 0.1303 0.3083 0.5399 0>}.RGB
   Metallic=ScalarParameter() {defaultValue=0.3, parameterName=Metallic}
   Specular=ScalarParameter() {defaultValue=0.35, parameterName=Specular}
   Roughness=ScalarParameter() {defaultValue=0.2, parameterName=Roughness}
   Emissive Color=Multiply(
    A=Multiply(A=@n7, B=@n8.Fresnel Inverted)
    B=TextureSample(UVs=@n5) {texture=&2ndInfectedMapLinear}.RGB
   )
   Opacity Mask=MaterialFunctionCall(Position=ObjectPositionWS(), Fade_Start_Distance=Add(A=@n9, B=@n10), Fade_End_Distance=@n10) {materialFunction=&MF_Distance_SyncedFade_FX}.Result
  ) {attributeSetTypes [
    #1
    #2
    #3
    #4
    #5
    #6
   ]}
  MaterialFunctionCall(MA In=@n11) {materialFunction=&MF_CreativeMoveTool}
 ]
}
