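# Material function that supplies the grid pattern for the Creative move tool (ObjectMover prototype weapon).
# It feeds T_Movetool_Transition through WorldAlignedTexture for triplanar projection, with tile size defaulting to a GridTile parameter.
# Position and normal inputs default to world values transformed into local space, so the pattern follows the object.
# The projected red, green and blue channels become the Grid, Blocks and Unused outputs.
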MaterialFunction /Game/Weapons/Prototype/ObjectMover/Materials/MF_CreativeMoveTool_Grid
graph {nodes [
  n1 = WorldPosition() {worldPositionShaderOffset=WPT_ExcludeAllShaderOffsets}
  n2 = VertexNormalWS()
  n3 = FunctionInput() {
   bUsePreviewValueAsDefault=true
   inputName=ProjectionTransitionContrast
   inputType=FunctionInput_Scalar
   sortPriority=1
   previewValue=<1 0 0 1>
  }
  n4 = MaterialFunctionCall(
   TextureObject=TextureObject() {texture=/Game/Weapons/Prototype/ObjectMover/Textures/T_Movetool_Transition}
   TextureSize=FunctionInput(Preview=ScalarParameter() {defaultValue=160, parameterName=GridTile}) {bUsePreviewValueAsDefault=true, inputName=TileSize, inputType=FunctionInput_Scalar}
   WorldPosition=FunctionInput(Preview=TransformPosition(None=@n1.XYZ) {transformSourceType=TRANSFORMPOSSOURCE_World}) {bUsePreviewValueAsDefault=true, inputName=WorldPosition, sortPriority=2}
   World Space Normal=FunctionInput(Preview=Transform(None=@n2) {transformSourceType=TRANSFORMSOURCE_World, transformType=TRANSFORM_Local}) {bUsePreviewValueAsDefault=true, inputName=WorldNormal, sortPriority=3}
   ProjectionTransitionContrast=@n3
  ) {materialFunction=&WorldAlignedTexture}
  FunctionOutput(None=ComponentMask(None=@n4.XYZ Texture) {r=true}) {outputName=Grid}
  FunctionOutput(None=ComponentMask(None=@n4.XYZ Texture) {g=true}) {outputName=Blocks, sortPriority=1}
  FunctionOutput(None=ComponentMask(None=@n4.XYZ Texture) {b=true}) {outputName=Unused, sortPriority=2}
 ]}
