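# Translucent, depth-test-free preview material for objects controlled by the prototype move tool, tinted through DiffuseTint and DraftColor.
# Opacity and color combine a local-bounds height gradient, projected grid, Fresnel and depth-based intersection line, feeding named reroutes.
# Pixels behind scene geometry dim and desaturate, while custom stencil and depth checks handle overlap with highlighted objects.
# Block transition effects are largely zeroed by constants marked TEMP DISCONNECT.
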Material /Game/Weapons/Prototype/ObjectMover/Materials/M_MoveTool_Controlled
graph {
 properties {
  bDisableDepthTest=true
  bEnableResponsiveAA=true
  bUsedWithInstancedStaticMeshes=true
  bUsedWithMeshDeformer=true
  bUsedWithSkeletalMesh=true
  bUsedWithSplineMeshes=true
  blendMode=BLEND_AlphaComposite
  pixelDepthOffsetMode=PDOM_Legacy
  translucencyLightingMode=TLM_VolumetricPerVertexDirectional
  twoSided=true
  usedShadingModels=Default Lit
 }
 nodes [
  ConstantBiasScale(None=Sine(None=Time()) {period=4})
  Constant()
  Constant()
  n1 = Subtract(A=CameraPositionWS(), B=ObjectPositionWS())
  NamedRerouteDeclaration(None=Multiply(A=Normalize(VectorInput=@n1)) {constB=2}) {name=WPO, nodeColor=<rgba 0.3751 1 0.2804 1>}
  n2 = Constant() {r=1}
  n3 = ComponentMask(None=TextureCoordinate() {coordinateIndex=1}) {r=true}
  n4 = Constant()
  n5 = Constant() {r=1}
  NamedRerouteDeclaration(None=LinearInterpolate(
    A=TextureSampleParameter2D() {parameterName=Normals, samplerType=SAMPLERTYPE_Normal, texture=/Engine/EngineMaterials/FlatNormal}.RGB
    B=Constant3Vector() {constant=<rgba 0.5 0.5 1 0>}
    Alpha=If(A=@n2, B=@n3, "A > B"=@n4, "A == B"=@n4, "A < B"=@n5) {desc=0}
   )) {name=Normals, nodeColor=<rgba 0.2371 0.3006 1 1>}
  n6 = Subtract(A=PixelDepth().R, B=SceneDepth().R)
  n7 = Abs(None=DotProduct(A=VertexNormalWS(), B=CameraVectorWS()))
  n8 = Multiply(A=Divide(A=@n6, B=ScalarParameter() {defaultValue=50, parameterName=DistanceBehingTranslucencyDivisor}), B=@n7)
  n9 = Saturate(None=Ceil(None=Saturate(None=@n8)))
  n10 = LinearInterpolate(B=ScalarParameter() {defaultValue=0.3, parameterName=OpacityBehindTranslucency}, Alpha=@n9) {constA=1}
  n11 = Custom() {code="GetPrimitiveData(Parameters.PrimitiveId).LocalObjectBoundsMin.xyz", description=Local Bounds Minimum, inputs []}
  n12 = Subtract(
   A=TransformPosition(None=WorldPosition() {worldPositionShaderOffset=WPT_ExcludeAllShaderOffsets}.XYZ) {transformSourceType=TRANSFORMPOSSOURCE_World}
   B=@n11
  )
  n13 = Max(A=Subtract(A=Custom() {code="GetPrimitiveData(Parameters.PrimitiveId).LocalObjectBoundsMax.xyz", description=Local Bounds Max, inputs []}, B=@n11)) {constB=0.01}
  n14 = Saturate(None=MaterialFunctionCall(Float3=Divide(A=@n12, B=@n13)) {materialFunction=&BreakOutFloat3Components}.B)
  n15 = LinearInterpolate(A=ScalarParameter() {defaultValue=0.45, parameterName=OpacityLow}, Alpha=@n14)
  n16 = MaterialFunctionCall!(TileSize, ProjectionTransitionContrast, WorldPosition, WorldNormal) {materialFunction=&/MF_CreativeMoveTool_Grid}
  n17 = ScalarParameter() {defaultValue=0.1, parameterName=GridIntensity}
  n18 = SceneDepth()
  n19 = PixelDepth()
  n20 = Multiply(A=Max(A=Subtract(A=@n18.R, B=@n19.R), B=@n6), B=@n7)
  n21 = ScalarParameter() {defaultValue=12, parameterName=IntersectingLineDivisor}
  n22 = OneMinus(None=Saturate(None=Divide(A=@n20, B=@n21)))
  n23 = ScalarParameter() {defaultValue=10, parameterName=IntersectingLineExponent}
  n24 = Saturate(None=Add(A=Multiply(A=@n16.Grid, B=@n17), B=Power(Base=@n22, Exp=@n23)))
  n25 = Constant() {desc=TEMP DISCONNECT}
  n26 = Saturate(None=Add(A=Add(A=@n15, B=@n24), B=@n25))
  n27 = TextureSampleParameter2D() {parameterName=MaskTexture, texture=/Engine/EngineResources/WhiteSquareTexture}
  n28 = MaterialFunctionCall() {materialFunction=&ObjectLocalBounds}
  n29 = Add(A=@n28.Local Bounds Minimum, B=@n28.Local Bounds Minimum)
  n30 = Subtract(A=WorldPosition().XYZ, B=TransformPosition(None=Multiply(A=@n29) {constB=0.5}) {transformType=TRANSFORMPOSSOURCE_World})
  n31 = ComponentMask(None=ObjectBounds()) {b=true}
  n32 = ConstantBiasScale(None=Divide(A=ComponentMask(None=@n30) {b=true}, B=Max(A=@n31)))
  n33 = ScalarParameter() {defaultValue=-100, parameterName=Transition}
  n34 = SphereMask(A=@n32, B=Constant(), Radius=@n33, Hardness=Constant() {r=1})
  n35 = Constant() {desc=TEMP DISCONNECT}
  n36 = Add(A=Multiply(A=Multiply(A=@n26, B=@n27.R), B=@n34), B=@n35)
  n37 = Saturate(None=TwoSidedSign())
  n38 = Constant() {r=1}
  n39 = ComponentMask(None=SceneTexture() {bFiltered=true, sceneTextureId=PPI_CustomStencil}.Color) {r=true}
  n40 = SceneTexture() {bFiltered=true, sceneTextureId=PPI_CustomDepth}
  n41 = SceneDepth()
  n42 = Saturate(None=Subtract(
    A=ComponentMask(None=@n40.Color) {r=true}
    B=Min(A=@n41.R) {constB=100000, desc=FORT-308709 - Clamp- Scene Depth to 1 KM to prevent preview from disappearing in the sky.}
   ))
  n43 = If(A=@n39, B=Constant() {r=2}, "A > B"=@n42, "A == B"=Constant() {r=1}, "A < B"=@n42) {desc=Check against move tool hightlight Stencil to prevent them from blocking the preview}
  n44 = ShadingPathSwitch(Default=@n38, Deferred=If(A=@n39, B=Constant() {r=20}, "A > B"=Constant() {r=1}, "A == B"=@n43, "A < B"=@n43), Forward=@n38, Mobile=@n38)
  n45 = Multiply(A=Multiply(A=Multiply(A=@n36, B=@n37), B=@n10), B=@n44)
  NamedRerouteDeclaration(None=@n45) {name=Opacity, nodeColor=<rgba 1 0.5859 0 1>}
  n46 = TextureCoordinate() {coordinateIndex=1}
  n47 = Constant()
  n48 = LinearInterpolate(
   A=TextureSampleParameter2D() {parameterName=Diffuse, texture=&WhiteSquareTexture}.RGB
   B=Constant() {r=1}
   Alpha=If(A=Constant() {r=1}, B=ComponentMask(None=@n46) {r=true}, "A > B"=@n47, "A == B"=@n47, "A < B"=Constant() {r=1}) {desc=0}
  )
  n49 = LinearInterpolate(
   A=VectorParameter() {parameterName=DiffuseLow, defaultValue=<rgba 0.6042 0.6042 0.6042 1>}.RGB
   B=VectorParameter() {parameterName=DiffuseHigh, defaultValue=<rgba 2.5 2.5 2.5 1>}.RGB
   Alpha=@n48
  )
  n50 = Multiply(A=@n48, B=ScalarParameter() {defaultValue=30.4264, parameterName=DiffuseScale})
  n51 = ScalarParameter() {defaultValue=22.9948, parameterName=DiffuseExponent}
  n52 = LinearInterpolate(A=Saturate(None=Power(Base=@n50, Exp=@n51)), B=ScalarParameter() {defaultValue=0.25, parameterName=UpperDiffuse}, Alpha=@n14) {constB=0.6222}
  n53 = Desaturation(None=Saturate(None=Subtract(A=@n49, B=@n52)))
  n54 = Add(A=Add(A=Multiply(A=@n14, B=@n53), B=ScalarParameter() {defaultValue=0.15, parameterName=DiffuseAdd}), B=@n24)
  n55 = VectorParameter() {parameterName=DiffuseTint, defaultValue=<rgba 1.2 0.2438 0.2438 0.8>}
  n56 = MaterialFunctionCall(
   Brightness=ScalarParameter() {defaultValue=1, parameterName=FresnelBrightness}
   Exponent=ScalarParameter() {defaultValue=2, parameterName=FresnelExponent}
  ) {materialFunction=&MF_Fresnel_FX}
  n57 = MaterialFunctionCall(ProjectionTransitionContrast=ScalarParameter() {defaultValue=10, parameterName=ProjectionTransitionContrast}) {materialFunction=&MF_CreativeMoveTool_Grid}
  n58 = ScalarParameter() {defaultValue=1.5, parameterName=BlockSineScale}
  n59 = Sine(None=Add(A=ConstantBiasScale(None=@n57.Blocks) {bias=-0.5, scale=2}, B=Multiply(A=@n58, B=@n33))) {period=0.5}
  n60 = Multiply(A=OneMinus(None=@n57.Grid), B=Multiply(A=ConstantBiasScale(None=@n59) {bias=1.6, scale=0.8}, B=@n25))
  n61 = Add(A=Saturate(None=Subtract(A=@n56.Fresnel Inverted, B=@n24)), B=@n60)
  n62 = Desaturation(
   None=LinearInterpolate(
    A=Multiply(A=@n54, B=@n55.RGB)
    B=VectorParameter() {parameterName=DraftColor, defaultValue=<rgba 0.0112 0.0258 0.1128 1>}.RGB
    Alpha=Saturate(None=@n61)
   )
   Fraction=@n9
  )
  n63 = Multiply(A=@n55.RGB, B=ScalarParameter() {defaultValue=20, parameterName=TransitionEdgeScale})
  n64 = Multiply(A=Multiply(A=LinearInterpolate(A=@n62, B=@n63, Alpha=@n35), B=@n55.A), B=@n36)
  n65 = ScalarParameter() {defaultValue=0.45, parameterName=BlocksTrailingTransitionScale}
  n66 = Multiply(A=@n55.RGB, B=Multiply(A=@n60, B=Multiply(A=@n65, B=@n34)))
  n67 = Multiply(A=Multiply(A=Add(A=@n64, B=@n66), B=@n10), B=@n44)
  NamedRerouteDeclaration(None=@n67) {name=Emissive, nodeColor=<rgba 1 0 0.9375 1>}
  n68 = Multiply(A=Multiply(A=Multiply(A=@n62, B=@n36), B=@n10), B=@n44)
  NamedRerouteDeclaration(None=@n68) {name=Base Color, nodeColor=<rgba 0.2271 1 0.1754 1>}
  NamedRerouteUsage()
  NamedRerouteUsage()
  NamedRerouteUsage()
  NamedRerouteUsage()
  NamedRerouteUsage()
  ObjectPositionWS()
  Reroute(None=@n67)
  Reroute(None=@n68)
  Reroute(None=@n45)
  n69 = ScalarParameter() {defaultValue=0.04, parameterName=NormalScale}
  n70 = ComponentMask(None=TextureSampleParameter2D() {parameterName=Normals, samplerType=SAMPLERTYPE_Normal, texture=&FlatNormal}.RGB) {g=true, r=true}
  SphereMask(
   A=Add(A=@n32, B=Multiply(A=@n69, B=@n70))
   B=@n33
   Radius=ScalarParameter() {defaultValue=0.02, parameterName=TransitionHardlineRadius}
   Hardness=ScalarParameter() {defaultValue=0.6, parameterName=TransitionHardlineHardness}
  ) {desc=Hard line}
  SphereMask(
   A=@n32
   B=ConstantBiasScale(None=@n33) {bias=-0.25, scale=2}
   Radius=ScalarParameter() {defaultValue=0.5, parameterName=TransitionBlocksRadius}
   Hardness=ScalarParameter() {defaultValue=0.2, parameterName=TransitionBlockHardness}
  )
 ]
}
