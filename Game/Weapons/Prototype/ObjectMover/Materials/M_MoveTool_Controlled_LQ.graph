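# Translucent low-quality material for the prototype ObjectMover move tool, drawing controlled objects as tinted holograms with a height-based transition.
# Desaturated, remapped diffuse and local-height gradient mix with a grid and inverted fresnel between DiffuseTint and DraftColor.
# Transition drives sphere masks along bounds-normalized world Z, producing a glowing hard line and animated blocks that shape emissive and opacity.
# Screen-space scanlines modulate opacity, and a separate vector offsets toward the camera.
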Material /Game/Weapons/Prototype/ObjectMover/Materials/M_MoveTool_Controlled_LQ
graph {
 properties {
  bEnableResponsiveAA=true
  bUsedWithMeshDeformer=true
  bUsedWithNanite=true
  bUsedWithSkeletalMesh=true
  bUsedWithSplineMeshes=true
  blendMode=BLEND_Translucent
  pixelDepthOffsetMode=PDOM_Legacy
  translucencyLightingMode=TLM_VolumetricPerVertexDirectional
  usedShadingModels=Default Lit
 }
 nodes [
  ConstantBiasScale(None=Sine(None=Time()) {period=4})
  n1 = TextureCoordinate() {coordinateIndex=1}
  n2 = Constant()
  LinearInterpolate(
   A=TextureSampleParameter2D() {parameterName=Normals, samplerType=SAMPLERTYPE_Normal, texture=/Engine/EngineMaterials/FlatNormal}.RGB
   B=Constant3Vector() {constant=<rgba 0.5 0.5 1 0>}
   Alpha=If(A=Constant() {r=1}, B=ComponentMask(None=@n1) {r=true}, "A > B"=@n2, "A == B"=@n2, "A < B"=Constant() {r=1}) {desc=0}
  )
  n3 = Custom() {code="GetPrimitiveData(Parameters.PrimitiveId).LocalObjectBoundsMin.xyz", description=Local Bounds Minimum, inputs []}
  n4 = Subtract(
   A=TransformPosition(None=WorldPosition() {worldPositionShaderOffset=WPT_ExcludeAllShaderOffsets}.XYZ) {transformSourceType=TRANSFORMPOSSOURCE_World}
   B=@n3
  )
  n5 = Max(A=Subtract(A=Custom() {code="GetPrimitiveData(Parameters.PrimitiveId).LocalObjectBoundsMax.xyz", description=Local Bounds Max, inputs []}, B=@n3)) {constB=0.01}
  n6 = Saturate(None=MaterialFunctionCall(Float3=Divide(A=@n4, B=@n5)) {materialFunction=&BreakOutFloat3Components}.B)
  n7 = VectorParameter() {parameterName=DiffuseLow, defaultValue=<rgba 0.6042 0.6042 0.6042 1>}
  n8 = VectorParameter() {parameterName=DiffuseHigh, defaultValue=<rgba 2.5 2.5 2.5 1>}
  n9 = TextureCoordinate() {coordinateIndex=1}
  n10 = Constant()
  n11 = LinearInterpolate(
   A=TextureSampleParameter2D() {parameterName=Diffuse, texture=/Engine/EngineResources/WhiteSquareTexture}.RGB
   B=Constant() {r=1}
   Alpha=If(A=Constant() {r=1}, B=ComponentMask(None=@n9) {r=true}, "A > B"=@n10, "A == B"=@n10, "A < B"=Constant() {r=1}) {desc=0}
  )
  n12 = ScalarParameter() {defaultValue=30.4264, parameterName=DiffuseScale}
  n13 = Saturate(None=Power(Base=Multiply(A=@n11, B=@n12), Exp=ScalarParameter() {defaultValue=22.9948, parameterName=DiffuseExponent}))
  n14 = ScalarParameter() {defaultValue=0.25, parameterName=UpperDiffuse}
  n15 = Saturate(None=Subtract(A=LinearInterpolate(A=@n7.RGB, B=@n8.RGB, Alpha=@n11), B=LinearInterpolate(A=@n13, B=@n14, Alpha=@n6) {constB=0.6222}))
  n16 = Add(A=Multiply(A=@n6, B=Desaturation(None=@n15)), B=ScalarParameter() {defaultValue=0.15, parameterName=DiffuseAdd})
  n17 = Saturate(None=Multiply(
    A=MaterialFunctionCall!(TileSize, ProjectionTransitionContrast, WorldPosition, WorldNormal) {materialFunction=&/MF_CreativeMoveTool_Grid}.Grid
    B=ScalarParameter() {defaultValue=0.1, parameterName=GridIntensity}
   ))
  n18 = VectorParameter() {parameterName=DiffuseTint, defaultValue=<rgba 0.0937 0.1893 1.2 0.8>}
  n19 = ScalarParameter() {defaultValue=2, parameterName=Fresnel Exponent}
  n20 = Saturate(None=Subtract(
    A=MaterialFunctionCall(Exponent=@n19) {materialFunction=&MF_Fresnel_FX}.Fresnel Inverted
    B=@n17
   ))
  n21 = MaterialFunctionCall(ProjectionTransitionContrast=ScalarParameter() {defaultValue=10, parameterName=ProjectionTransitionContrast}) {materialFunction=&MF_CreativeMoveTool_Grid}
  n22 = ScalarParameter() {defaultValue=1.5, parameterName=BlockSineScale}
  n23 = ScalarParameter() {defaultValue=1.25, parameterName=Transition}
  n24 = Sine(None=Add(A=ConstantBiasScale(None=@n21.Blocks) {bias=-0.5, scale=2}, B=Multiply(A=@n22, B=@n23))) {period=0.5}
  n25 = Subtract(A=WorldPosition() {worldPositionShaderOffset=WPT_ExcludeAllShaderOffsets}.XYZ, B=ObjectPositionWS())
  n26 = Max(A=ComponentMask(None=ObjectBounds()) {b=true}) {desc=Planes that are flat on the Z axis will evaluate to 0 here and then cause undefined errors. Adding this max to fix the undefined errors}
  n27 = ConstantBiasScale(None=Divide(A=ComponentMask(None=@n25) {b=true}, B=@n26))
  n28 = SphereMask(
   A=@n27
   B=ConstantBiasScale(None=@n23) {bias=-0.25, scale=2}
   Radius=ScalarParameter() {defaultValue=0.5, parameterName=TransitionBlocksRadius}
   Hardness=ScalarParameter() {defaultValue=0.2, parameterName=TransitionBlockHardness}
  )
  n29 = Multiply(A=OneMinus(None=@n21.Grid), B=Multiply(A=ConstantBiasScale(None=@n24) {bias=1.6, scale=0.8}, B=@n28))
  n30 = LinearInterpolate(
   A=Multiply(A=Add(A=@n16, B=@n17), B=@n18.RGB)
   B=VectorParameter() {parameterName=DraftColor, defaultValue=<rgba 0.0112 0.0258 0.1128 1>}.RGB
   Alpha=Saturate(None=Add(A=@n20, B=@n29))
  )
  n31 = ScalarParameter() {defaultValue=10, parameterName=TransitionEdgeScale}
  n32 = ScalarParameter() {defaultValue=0.04, parameterName=NormalScale}
  n33 = ComponentMask(None=TextureSampleParameter2D() {parameterName=Normals, samplerType=SAMPLERTYPE_Normal, texture=&FlatNormal}.RGB) {g=true, r=true}
  n34 = SphereMask(
   A=Add(A=@n27, B=Multiply(A=@n32, B=@n33))
   B=@n23
   Radius=ScalarParameter() {defaultValue=0.02, parameterName=TransitionHardlineRadius}
   Hardness=ScalarParameter() {defaultValue=0.6, parameterName=TransitionHardlineHardness}
  ) {desc=Hard line}
  n35 = Multiply(A=LinearInterpolate(A=@n30, B=Multiply(A=@n18.RGB, B=@n31), Alpha=@n34), B=@n18.A)
  n36 = ScalarParameter() {defaultValue=0.45, parameterName=BlocksTrailingTransitionScale}
  n37 = SphereMask(A=@n27, B=Constant(), Radius=@n23, Hardness=Constant() {r=1})
  n38 = Multiply(A=@n18.RGB, B=Multiply(A=@n29, B=Multiply(A=@n36, B=@n37)))
  n39 = Add(A=@n34, B=@n37)
  MaterialFunctionCall(Emissive=Multiply(A=Add(A=@n35, B=@n38), B=@n39)) {materialFunction=&MF_FX_EmissiveAdjustments}
  Multiply(A=@n30, B=@n39)
  n40 = Add(A=LinearInterpolate(A=ScalarParameter() {defaultValue=0.45, parameterName=OpacityLow}, Alpha=@n6), B=@n17)
  n41 = Multiply(
   A=Saturate(None=Add(A=@n40, B=@n28))
   B=TextureSampleParameter2D() {parameterName=MaskTexture, texture=&WhiteSquareTexture}.R
  )
  n42 = Add(A=ScreenPosition().ViewportUV, B=Multiply(A=Time(), B=ScalarParameter() {parameterName=ScanlinePanSpeed}))
  n43 = Frac(None=Multiply(A=ComponentMask(None=@n42) {g=true}, B=ScalarParameter() {defaultValue=200, parameterName=ScanlineRepeat}))
  n44 = Constant() {r=0.5}
  n45 = ScalarParameter() {defaultValue=0.4, parameterName=ScanlineRadius}
  n46 = ScalarParameter() {defaultValue=0.1, parameterName=ScanlineHardness}
  Multiply(
   A=Add(A=Multiply(A=@n41, B=@n37), B=@n34)
   B=LinearInterpolate(A=ScalarParameter() {defaultValue=1, parameterName=ScanlinesLow}, Alpha=SphereMask(A=@n43, B=@n44, Radius=@n45, Hardness=@n46))
  )
  n47 = CameraPositionWS()
  n48 = ObjectPositionWS()
  Multiply(A=Normalize(VectorInput=Subtract(A=@n47, B=@n48)), B=VectorParameter() {parameterName=CameraOffset, defaultValue=<rgba 20 20 20 0>}.RGB)
 ]
}
