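# Translucent unlit material for the prototype ObjectMover weapon, named as the invisible material of its move tool.
# The graph holds only a default zero Constant and a Subtract of world position from object position, with no connections to material outputs shown.
# Usage with Nanite is flagged, although the blend mode is translucent.
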Material /Game/Weapons/Prototype/ObjectMover/Materials/M_MoveTool_Invisible
graph {
 properties {bUsedWithNanite=true, blendMode=BLEND_Translucent, pixelDepthOffsetMode=PDOM_Legacy, shadingModel=MSM_Unlit, usedShadingModels=Unlit}
 nodes [
  Constant()
  Subtract(A=ObjectPositionWS(), B=WorldPosition().XYZ)
 ]
}
