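# M_MoveTool_SaveZone is a translucent, unlit, two-sided material for what appears to be the Object Mover prototype's zone volume.
# It layers rotated UV scanlines on back faces over a world-aligned grid of squares pulsed by a time-driven sine.
# Opacity fades near and far from the camera, and front-face tops and bottoms fade out.
# Beyond a distance threshold based on object scale, vertices are offset far along their normals, apparently to cull the mesh.
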Material /Game/Weapons/Prototype/ObjectMover/Materials/M_MoveTool_SaveZone
graph {
 properties {
  bEnableResponsiveAA=true
  bUseTranslucencyVertexFog=false
  blendMode=BLEND_Translucent
  pixelDepthOffsetMode=PDOM_Legacy
  shadingModel=MSM_Unlit
  twoSided=true
  usedShadingModels=Unlit
 }
 nodes [
  Constant3Vector() {constant=<rgba 0 0 -999999 0>}
  n1 = MaterialFunctionCall(Vector 3=WorldPosition() {worldPositionShaderOffset=WPT_CameraRelative}.XYZ) {materialFunction=/Engine/Functions/Engine_MaterialFunctions02/Utility/VectorLength}
  n2 = ScalarParameter() {defaultValue=1350, parameterName=FarFade}
  n3 = MaterialFunctionCall() {materialFunction=/Engine/Functions/Engine_MaterialFunctions02/WorldPositionOffset/ObjectScale}
  n4 = Multiply(
   A=ScalarParameter() {defaultValue=7.5, parameterName=Buffer}
   B=Divide(A=Max(A=@n3.Scale X, B=@n3.Scale X)) {constB=36, desc=Original values based off a scale of 36 on XY}
  )
  n5 = Constant() {r=9999999}
  n6 = VertexNormalWS()
  n7 = Constant()
  Reroute(None=If(A=@n1.V3 Length, B=Multiply(A=@n2, B=@n4) {constB=1.36, desc=add some buffer}, "A > B"=Multiply(A=@n5, B=@n6), "A == B"=@n7, "A < B"=@n7))
  n8 = VectorParameter() {parameterName=BGColor, defaultValue=<rgba 0.8 0.3276 0.0417 1>}
  n9 = Divide(
   A=Rotator(Coordinate=TextureCoordinate(), Time=ScalarParameter() {defaultValue=3.12, parameterName=Rotation})
   B=ScalarParameter() {defaultValue=0.008, parameterName=ScanlineRepeat}
  )
  n10 = SphereMask(
   A=Frac(None=ComponentMask(None=@n9) {r=true})
   B=Constant() {r=0.5}
   Radius=ScalarParameter() {defaultValue=0.15, parameterName=ScanlineRadius}
   Hardness=ScalarParameter() {defaultValue=0.95, parameterName=ScanlineHardness}
  )
  n11 = TwoSidedSign()
  n12 = Multiply(A=@n10, B=OneMinus(None=Saturate(None=@n11)))
  n13 = Divide(A=PixelDepth().R, B=Constant() {r=3500})
  n14 = MaterialFunctionCall!(
   TextureObject=TextureObject() {samplerType=SAMPLERTYPE_LinearColor, texture=/Game/Gadgets/Assets/Teleporter/Textures/T_Teleporter_GridSquares_01}
   TextureSize=Constant() {r=512}
   WorldPosition
   Export Float 4=StaticBool()
   In
   World Space Normal
   ProjectionTransitionContrast=Constant() {r=13}
  ) {materialFunction=/Game/Athena/SafeZone/WorldAlignedTexture_ScalePan}
  n15 = ComponentMask(None=@n14.XYZ Texture) {g=true}
  n16 = Time() {bOverride_Period=true, period=3.8314}
  n17 = ScalarParameter() {defaultValue=0.261, parameterName=Sine Speed Grid}
  n18 = Sine(None=Add(A=ComponentMask(None=@n14.XYZ Texture) {r=true}, B=Multiply(A=@n16, B=@n17)))
  n19 = Multiply(
   A=OneMinus(None=Saturate(None=@n13))
   B=Power(Base=Multiply(A=@n15, B=@n18), Exp=ScalarParameter() {defaultValue=5, parameterName=Grid Power})
  )
  n20 = Multiply(A=OneMinus(None=@n12), B=@n19)
  n21 = Multiply(A=@n20, B=VectorParameter() {parameterName=BlocksColor, defaultValue=<rgba 0.6615 0.1564 0.0439 0>}.RGB)
  n22 = VectorParameter() {parameterName=scanlines Color, defaultValue=<rgba 0.7865 0.285 0.0375 1>}
  Reroute(None=LinearInterpolate(A=LinearInterpolate(A=@n8.RGB, B=@n21, Alpha=@n19), B=Multiply(A=@n10, B=@n22.RGB), Alpha=@n12))
  n23 = Add(A=PixelDepth().R, B=Constant() {r=-100})
  n24 = ScalarParameter() {defaultValue=300, parameterName=NearFade}
  n25 = Add(
   A=Multiply(A=@n12, B=ScalarParameter() {defaultValue=0.5, parameterName=LinesScale})
   B=Multiply(A=@n20, B=ScalarParameter() {defaultValue=0.15, parameterName=BlocksOpacity})
  )
  n26 = Constant() {r=0.7}
  n27 = Constant() {r=-100}
  n28 = Saturate(None=Divide(A=Add(A=@n1.V3 Length, B=@n27), B=@n2))
  n29 = Multiply(A=Saturate(None=Divide(A=@n23, B=@n24)), B=Multiply(A=Add(A=@n25, B=@n26), B=OneMinus(None=@n28)))
  n30 = Abs(None=Normalize(VectorInput=VertexNormalWS()))
  n31 = Constant3Vector() {constant=<rgba 0 0 1 0>}
  n32 = VertexInterpolator(VS=OneMinus(None=DotProduct(A=@n30, B=@n31)))
  n33 = TwoSidedSign()
  Reroute(None=LinearInterpolate(A=@n29, B=Multiply(A=@n29, B=@n32.PS), Alpha=Saturate(None=@n33)))
 ]
}
