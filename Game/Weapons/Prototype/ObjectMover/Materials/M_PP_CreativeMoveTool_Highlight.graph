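# Post-process material that highlights Creative move tool props marked by custom stencil 200, 201 (multi-select) or 202 (particle).
# Builds custom-depth outline, AO tint, inverted Fresnel and grid aligned to phone vectors, with a depth-tested stencil mask.
# Colors shift for delete, via a material parameter collection value, and for multi-select.
# Mobile, forward and low/medium quality use a cheaper branch with a world-space vertex-normal grid multiplied into the tinted scene.
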Material /Game/Weapons/Prototype/ObjectMover/Materials/M_PP_CreativeMoveTool_Highlight
graph {
 properties {
  blendableLocation=BL_SceneColorBeforeDOF
  blendablePriority=255
  materialDomain=MD_PostProcess
  pixelDepthOffsetMode=PDOM_Legacy
  usedShadingModels=Default Lit
 }
 nodes [
  n1 = ComponentMask(None=ScreenPosition().ViewportUV) {g=true, r=true}
  n2 = Subtract(A=Divide(A=ViewSize(), B=Constant() {r=320}), B=Constant() {r=2}) {constB=0}
  n3 = Multiply(A=Clamp(None=Ceil(None=@n2)) {maxDefault=2, minDefault=1}, B=SceneTexelSize()) {constA=2}
  n4 = Constant()
  n5 = AppendVector(A=ComponentMask(None=@n3) {desc=x, r=true}, B=@n4)
  n6 = AppendVector(A=@n4, B=ComponentMask(None=@n3) {desc=y, g=true}) {desc=y+}
  n7 = Constant() {r=-1}
  n8 = Multiply(A=@n5, B=@n7)
  n9 = Multiply(A=@n7, B=@n6)
  n10 = MaterialFunctionCall!(
   Base=SceneTexture() {sceneTextureId=PPI_CustomDepth}.Color
   BaseSampleBias
   Scale
   Sample1=SceneTexture(UVs=Add(A=@n1, B=@n5) {desc=x+}) {sceneTextureId=PPI_CustomDepth}.Color
   Sample2=SceneTexture(UVs=Add(A=@n1, B=@n6) {desc=y+}) {sceneTextureId=PPI_CustomDepth}.Color
   Sample3=SceneTexture(UVs=Add(A=@n1, B=@n8) {desc=x-}) {sceneTextureId=PPI_CustomDepth}.Color
   Sample4=SceneTexture(UVs=Add(A=@n1, B=@n9) {desc=y-}) {sceneTextureId=PPI_CustomDepth}.Color
   Divisor=ScalarParameter() {defaultValue=-16, parameterName=Divisor}
   UseCenteredLines
   UseRGB
  ) {materialFunction=/Game/Packages/jl_tests/combine4Samples}
  n11 = CollectionParameter() {
   collection=&FortniteMaterialParameters
   parameterName=CreativeDeleteHighlight
  }
  n12 = LinearInterpolate(
   A=VectorParameter() {parameterName=OutlineColor, defaultValue=<rgba 0.45 0.4871 0.8906 1>}.RGB
   B=VectorParameter() {parameterName=OutlineColorDelete, defaultValue=<rgba 0.8906 0.1818 0.0543 1>}.RGB
   Alpha=@n11
  )
  n13 = VectorParameter() {parameterName=OutlineColorMultiSelect, defaultValue=<rgba 0.0287 0.8906 0.6725 1>}
  n14 = NamedRerouteUsage()
  NamedRerouteDeclaration(None=Multiply(A=ComponentMask(None=@n10.Result) {r=true}, B=LinearInterpolate(A=@n12, B=@n13.RGB, Alpha=@n14))) {name=Cached Outline Out, nodeColor=<rgba 0 1 0.625 1>}
  NamedRerouteDeclaration(None=@n11) {name=Cache_MPC_'CreativeDeleteHighlight', nodeColor=<rgba 1 0.3047 0 1>}
  n15 = Constant()
  n16 = If(
   A=ComponentMask(None=SceneTexture() {sceneTextureId=PPI_CustomStencil}.Color) {r=true}
   B=Constant() {r=202}
   "A > B"=@n15
   "A == B"=Constant() {r=1}
   "A < B"=@n15
  )
  NamedRerouteDeclaration(None=OneMinus(None=@n16)) {name=CachedParticlePropOneMinus, nodeColor=<rgba 0 0.0156 1 1>}
  NamedRerouteDeclaration(None=@n16) {name=CachedParticleProp, nodeColor=<rgba 0.3359 1 0 1>}
  n17 = MaterialFunctionCall(
   Brightness=ScalarParameter() {defaultValue=1, parameterName=FresnelBrightness}
   Exponent=ScalarParameter() {defaultValue=4, parameterName=FresnelExponent}
   Vertex / Pixel Normal=ComponentMask(None=SceneTexture() {bFiltered=true, sceneTextureId=PPI_WorldNormal}.Color) {b=true, g=true, r=true}
  ) {materialFunction=&MF_Fresnel_FX}
  n18 = ScalarParameter() {defaultValue=2, parameterName=FresnelScale}
  NamedRerouteDeclaration(None=Multiply(A=Multiply(A=@n17.Fresnel Inverted, B=@n18), B=NamedRerouteUsage())) {name=CachedFresnelOut, nodeColor=<rgba 0 1 0.0625 1>}
  n19 = SceneTexture() {sceneTextureId=PPI_CustomStencil}
  n20 = Constant()
  NamedRerouteDeclaration(None=If(A=ComponentMask(None=@n19.Color) {r=true}, B=Constant() {r=201}, "A > B"=@n20, "A == B"=Constant() {r=1}, "A < B"=@n20)) {name=Cached MultiSelectedProp Out, nodeColor=<rgba 1 0.1406 0 1>}
  n21 = WorldPosition()
  n22 = CollectionParameter() {
   collection=&FortniteMaterialParameters
   parameterName=CreativeModePhoneRightVector
  }
  n23 = ComponentMask(None=@n22) {b=true, g=true, r=true}
  n24 = CollectionParameter() {
   collection=&FortniteMaterialParameters
   parameterName=CreativeModePhoneForwardVector
  }
  n25 = ComponentMask(None=@n24) {b=true, g=true, r=true}
  n26 = CollectionParameter() {
   collection=&FortniteMaterialParameters
   parameterName=CreativeModePhoneUpVector
  }
  n27 = ComponentMask(None=@n26) {b=true, g=true, r=true}
  n28 = AppendVector(A=ComponentMask(None=SceneTexture() {sceneTextureId=PPI_WorldNormal}.Color) {b=true, g=true, r=true}, B=Constant() {r=1})
  n29 = Multiply(
   A=NamedRerouteUsage()
   B=MaterialFunctionCall!(
    TileSize
    ProjectionTransitionContrast=ScalarParameter() {defaultValue=10, parameterName=ProjectionTransitionContrast}
    WorldPosition=MaterialFunctionCall(VectorToTransform=@n21.XYZ, BasisX=@n23, BasisY=@n25, BasisZ=@n27) {materialFunction=&Transform3x3Matrix}.Transformed Vector
    WorldNormal=MaterialFunctionCall(VectorToTransform=@n28, BasisX=@n23, BasisY=@n25, BasisZ=@n27) {materialFunction=&Transform3x3Matrix}.Transformed Vector
   ) {materialFunction=&/MF_CreativeMoveTool_Grid}.Grid
  )
  n30 = NamedRerouteUsage()
  n31 = CollectionParameter() {
   collection=&FortniteMaterialParameters
   parameterName=CreativeDeleteHighlight
  }
  n32 = LinearInterpolate(
   A=VectorParameter() {parameterName=GridColorHighlight, defaultValue=<rgba 0.0455 0.0823 0.5344 1>}.RGB
   B=VectorParameter() {parameterName=GridColorHighlightDelete, defaultValue=<rgba 0.5344 0.1706 0.0578 1>}.RGB
   Alpha=@n31
  ) {constAlpha=1}
  n33 = VectorParameter() {parameterName=GridColorHighlightMultiSelect, defaultValue=<rgba 0.1476 0.5344 0.0997 1>}
  n34 = NamedRerouteUsage()
  NamedRerouteDeclaration(None=Multiply(A=Add(A=@n29, B=@n30), B=LinearInterpolate(A=@n32, B=@n33.RGB, Alpha=@n34))) {name=Cached_PC_HighEnd_WorldGridOut, nodeColor=<rgba 0.1016 1 0 1>}
  n35 = ComponentMask(None=SceneTexture() {sceneTextureId=PPI_SceneDepth}.Color) {r=true}
  n36 = ComponentMask(None=SceneTexture() {sceneTextureId=PPI_CustomDepth}.Color) {r=true}
  n37 = Constant() {r=1}
  n38 = Constant()
  n39 = Constant()
  n40 = If(
   A=ComponentMask(None=SceneTexture() {sceneTextureId=PPI_CustomStencil}.Color) {r=true}
   B=Constant() {r=200}
   "A > B"=@n39
   "A == B"=Constant() {r=1}
   "A < B"=@n39
  )
  n41 = NamedRerouteUsage()
  n42 = Saturate(None=Add(A=Add(A=@n40, B=@n41), B=NamedRerouteUsage()))
  NamedRerouteDeclaration(None=Multiply(
    A=If(A=@n35, B=@n36, "A > B"=@n37, "A == B"=@n37, "A < B"=@n38)
    B=NamedRerouteDeclaration(None=@n42) {name=Cached_AllStencils, nodeColor=<rgba 0 1 0.6953 1>}
   )) {name="Cached_NormalProp&&Stencil", nodeColor=<rgba 1 0 0.4219 1>}
  n43 = LinearInterpolate(
   A=LinearInterpolate(
    A=VectorParameter() {parameterName=AOColor, defaultValue=<rgba 0.1177 0.1637 0.8646 0>}.RGB
    B=VectorParameter() {parameterName=AOColorDelete, defaultValue=<rgba 0.3941 0.0902 0.0285 0>}.RGB
    Alpha=NamedRerouteUsage()
   )
   B=VectorParameter() {parameterName=AOColorMultiSelect, defaultValue=<rgba 0.0205 0.6962 0.0994 0>}.RGB
   Alpha=NamedRerouteUsage()
  )
  n44 = ComponentMask(None=SceneTexture() {sceneTextureId=PPI_AmbientOcclusion}.Color) {b=true, g=true, r=true}
  n45 = ComponentMask(None=SceneTexture() {sceneTextureId=PPI_PostProcessInput0}.Color) {b=true, g=true, r=true}
  n46 = Clamp(None=@n45) {maxDefault=0.5}
  n47 = Constant() {r=1}
  n48 = Saturate(None=Floor(None=DotProduct(A=@n45, B=@n47)))
  n49 = LinearInterpolate(A=@n45, B=Multiply(A=LinearInterpolate(A=@n43, Alpha=@n44), B=LinearInterpolate(A=@n45, B=@n46, Alpha=@n48)), Alpha=NamedRerouteUsage())
  n50 = Multiply(A=NamedRerouteUsage(), B=NamedRerouteUsage())
  n51 = NamedRerouteUsage()
  n52 = NamedRerouteUsage()
  n53 = NamedRerouteUsage()
  n54 = MaterialFunctionCall(
   ProjectionTransitionContrast=ScalarParameter() {defaultValue=10, parameterName=MobileSwitchProjectionTransitionContrast}
   WorldPosition=TransformPosition(None=WorldPosition() {worldPositionShaderOffset=WPT_ExcludeAllShaderOffsets}.XYZ) {transformSourceType=TRANSFORMPOSSOURCE_World, transformType=TRANSFORMPOSSOURCE_World}
   WorldNormal=Transform(None=VertexNormalWS()) {transformSourceType=TRANSFORMSOURCE_World}
  ) {materialFunction=&MF_CreativeMoveTool_Grid}
  n55 = VectorParameter() {parameterName=GridColorHighlight, defaultValue=<rgba 0.0455 0.0823 0.5344 1>}
  n56 = VectorParameter() {parameterName=GridColorHighlightDelete, defaultValue=<rgba 0.5344 0.1706 0.0578 1>}
  n57 = CollectionParameter() {
   collection=&FortniteMaterialParameters
   parameterName=CreativeDeleteHighlight
  }
  n58 = Multiply(
   A=Add(A=Multiply(A=@n53, B=@n54.Grid), B=NamedRerouteUsage())
   B=LinearInterpolate(
    A=LinearInterpolate(A=@n55.RGB, B=@n56.RGB, Alpha=@n57) {constAlpha=1}
    B=VectorParameter() {parameterName=GridColorHighlightMultiSelect, defaultValue=<rgba 0.1476 0.5344 0.0997 1>}.RGB
    Alpha=NamedRerouteUsage()
   )
  )
  n59 = Add(A=@n49, B=Multiply(A=Multiply(A=@n49, B=@n58), B=NamedRerouteUsage()))
  n60 = ShadingPathSwitch(Default=Add(A=Add(A=@n49, B=@n50), B=Multiply(A=@n51, B=@n52)), Mobile=@n59) {desc=Migrated from FeatureLevelSwitch}
  n61 = QualitySwitch(Default=@n60, Low=@n59, High=@n60, Medium=@n59, Epic=@n60)
  ShadingPathSwitch(Default=@n61, Deferred=@n61, Forward=@n59, Mobile=@n59)
 ]
}
