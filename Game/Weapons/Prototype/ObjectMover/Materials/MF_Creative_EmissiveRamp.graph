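# Material function for the Object Mover prototype weapon that adds a pulsing ramp to the emissive color in a material attributes input.
# Scaled, looping time offsets the ramp texture's red channel into a SphereMask, creating a moving band that lifts emissive from a low value to full.
# The texture's green channel masks where the pulse multiplies the color, leaving black areas static.
# A static switch passes the original emissive through when the ramp is off.
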MaterialFunction /Game/Weapons/Prototype/ObjectMover/Materials/MF_Creative_EmissiveRamp
graph {
 properties {description=This function adds a pulsing emissive ramp to a material's emissive color. Note: Use mask for the ramp texture. Black contains static emissive color. White contains the emissive ramp.}
 nodes [
  n1 = GetMaterialAttributes(None=FunctionInput() {inputName=MA In, inputType=FunctionInput_MaterialAttributes}) {attributeGetTypes [#1]}
  n2 = ScalarParameter() {defaultValue=0.05, group=EmissiveRamp, parameterName=EmissiveRampLowValue}
  n3 = TextureSampleParameter2D() {group=EmissiveRamp, parameterName=EmissiveRampTexture}
  n4 = Frac(None=Multiply(A=Time() {period=1}, B=ScalarParameter() {defaultValue=0.25, group=EmissiveRamp, parameterName=EmissiveRampTimeScale}))
  n5 = SphereMask(
   A=Add(A=@n3.R, B=ConstantBiasScale(None=@n4) {bias=-0.5, scale=4})
   B=Constant()
   Radius=ScalarParameter() {defaultValue=1, group=EmissiveRamp, parameterName=EmissiveRampRadius}
   Hardness=ScalarParameter() {group=EmissiveRamp, parameterName=EmissiveRampHardness}
  )
  n6 = Multiply(A=@n1.EmissiveColor, B=LinearInterpolate(B=LinearInterpolate(A=@n2, Alpha=@n5), Alpha=@n3.G) {constA=1})
  FunctionOutput(None=SetMaterialAttributes!(
    MaterialAttributes=@n1.MaterialAttributes
    EmissiveColor=StaticSwitchParameter(True=@n6, False=@n1.EmissiveColor) {group=EmissiveRamp, parameterName=UseEmissiveRamp}
   ) {attributeSetTypes [#1]}) {outputName=MA Out}
 ]
}
