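# Material function in the ObjectMover weapon prototype that computes a curvature-compensated Fresnel mask from a world normal, defaulting to the pixel normal.
# Screen-space normal derivatives feed cross products, whose absolute sum, weighted by Thickness, becomes the exponent on saturated camera-dot-normal.
# This scales falloff with on-screen normal change, apparently normalizing rim width.
# Outputs provide hard rounded masks and soft masks sharpened by PowerExponent, each with an inverted variant.
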MaterialFunction /Game/Weapons/Prototype/ObjectMover/Materials/MF_NormalizedFresnel
graph {
 properties {bExposeToLibrary=true}
 nodes [
  n1 = CameraVectorWS()
  n2 = FunctionInput(Preview=PixelNormalWS()) {bUsePreviewValueAsDefault=true, inputName=NormalWS, sortPriority=1}
  n3 = DDX(Value=@n2)
  n4 = Abs(None=CrossProduct(A=Subtract(A=@n2, B=@n3), B=Add(A=@n3, B=@n2)))
  n5 = DDY(Value=@n2)
  n6 = Abs(None=CrossProduct(A=Subtract(A=@n2, B=@n5), B=Add(A=@n5, B=@n2)))
  n7 = Constant() {r=3}
  n8 = Power(
   Base=Saturate(None=DotProduct(A=@n1, B=@n2))
   Exp=DotProduct(A=Add(A=@n4, B=@n6), B=FunctionInput(Preview=@n7) {bUsePreviewValueAsDefault=true, inputName=Thickness, inputType=FunctionInput_Scalar})
  )
  n9 = Round(None=@n8)
  FunctionOutput(None=OneMinus(None=@n9)) {outputName=RoundInverted}
  FunctionOutput(None=@n9) {outputName=Round}
  n10 = FunctionInput(Preview=Constant() {r=4}) {bUsePreviewValueAsDefault=true, inputName=PowerExponent, inputType=FunctionInput_Scalar, sortPriority=2, previewValue=<8 0 0 1>}
  n11 = Power(Base=@n8, Exp=@n10) {constExponent=8}
  FunctionOutput(None=@n11) {outputName=Power, sortPriority=2}
  FunctionOutput(None=Power(Base=OneMinus(None=@n11), Exp=@n10) {constExponent=8}) {outputName=PowerInverted, sortPriority=3}
 ]
}
